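# Control Rig for the female SugarBubble player body that procedurally corrects sleeve, collar, rope, blouse and bag-belt bones as the body poses.
# Parent constraints and spherical pose readers on hands, clavicles, arms and legs weight additive transform offsets, with Kalman smoothing on some weights.
# Spine bend and calf separation drive further correctives, while baseball-slide and DBNO flags switch in extra blouse offsets.
# A shared DeformRig function runs last for body deformation.
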ControlRigBlueprint /Game/Characters/Player/Female/Medium/Bodies/F_MED_SugarBubble/Meshes/F_MED_SugarBubble_CtrlRig
graphs [
 RigUnit_SequenceAggregate_1_ContainedGraph {nodes [
   n1 = RigVMFunctionEntryNode()
   RigUnit_SequenceAggregate = &RigVMFunction_Sequence(ExecuteContext=@n1.ExecuteContext)
   RigUnit_SequenceAggregate_1 = &RigVMFunction_Sequence(ExecuteContext=@RigUnit_SequenceAggregate.B)
   RigUnit_SequenceAggregate_2 = &RigVMFunction_Sequence(ExecuteContext=@RigUnit_SequenceAggregate_1.B)
   RigVMFunctionReturnNode(A=@RigUnit_SequenceAggregate.A, B=@RigUnit_SequenceAggregate_1.A, C=@RigUnit_SequenceAggregate_2.A, D=@RigUnit_SequenceAggregate_2.B)
  ]}
 RigVMFunctionLibrary {links [], nodes []}
 RigVMModel {nodes [
   RigUnit_BeginExecution = &RigUnit_BeginExecution()
   RigUnit_SequenceAggregate = RigVMAggregateNode(ExecuteContext=@RigUnit_BeginExecution.ExecutePin)
   &RigUnit_ParentConstraint(
    AdvancedSettings=(InterpolationType=Average,RotationOrderForFilter=XZY)
    Child=(Type=Bone,Name="_hand_l_Sleeves3")
    Filter=(TranslationFilter=(bX=false,bY=false,bZ=false),RotationFilter=(bX=true,bY=true,bZ=true),ScaleFilter=(bX=True,bY=True,bZ=True))
    Parents=((Item=(Type=Bone,Name="hand_l"),Weight=1))
    Weight=0.8
    ExecutePin=@RigUnit_SequenceAggregate.A
   )
   > &RigUnit_ParentConstraint(
    AdvancedSettings=(InterpolationType=Average,RotationOrderForFilter=XZY)
    Child=(Type=Bone,Name="_hand_r_Sleeves3")
    Filter=(TranslationFilter=(bX=false,bY=false,bZ=false),RotationFilter=(bX=True,bY=True,bZ=True),ScaleFilter=(bX=True,bY=True,bZ=True))
    Parents=((Item=(Type=Bone,Name="hand_r"),Weight=1))
    Weight=0.6
   )
   > n1 = &RigUnit_SphericalPoseReader(
    ActiveRegionScaleFactors=(PositiveWidth=3,NegativeWidth=3,PositiveHeight=1,NegativeHeight=1)
    ActiveRegionSize=0.1
    Debug=(bDrawDebug=false,DebugScale=25,DebugSegments=20,DebugThickness=0.25)
    DriverAxis=(X=1,Y=0,Z=0)
    DriverItem=(Type=Bone,Name="hand_l")
    FalloffRegionScaleFactors=(PositiveWidth=3,NegativeWidth=3,PositiveHeight=2,NegativeHeight=1)
    FalloffSize=0.4
    OptionalParentItem=(Type=Bone)
    RotationOffset=(X=50,Y=-30,Z=-60)
   )
   > &RigUnit_ModifyTransforms(
    ItemToModify=((Item=(Type=Bone,Name="_hand_l_Sleeves3"),Transform=(Rotation=(X=0,Y=0,Z=0.1736,W=0.9848),Translation=(X=-4,Y=1,Z=0),Scale3D=(X=1,Y=1,Z=1))))
    Weight=@n1.OutputParam
   )
   > &RigUnit_ModifyTransforms(
    ItemToModify=((Item=(Type=Bone,Name="_hand_l_Sleeves_01"),Transform=(Rotation=(X=0.0151,Y=-0.0858,Z=0.173,W=0.9811),Translation=(X=0,Y=0,Z=0),Scale3D=(X=1,Y=1,Z=1))))
    Weight=@n1.OutputParam
   )
   > &RigUnit_ModifyTransforms(
    ItemToModify=((Item=(Type=Bone,Name="_hand_l_Sleeves_03"),Transform=(Rotation=(X=0,Y=0,Z=-0.0872,W=0.9962),Translation=(X=0,Y=-5,Z=0),Scale3D=(X=1,Y=1,Z=1))))
    Weight=@n1.OutputParam
   )
   > n2 = &RigUnit_SphericalPoseReader(
    ActiveRegionScaleFactors=(PositiveWidth=1,NegativeWidth=1,PositiveHeight=1,NegativeHeight=1)
    ActiveRegionSize=0.1
    Debug=(bDrawDebug=false,DebugScale=25,DebugSegments=20,DebugThickness=0.25)
    DriverAxis=(X=1,Y=0,Z=0)
    DriverItem=(Type=Bone,Name="hand_l")
    FalloffRegionScaleFactors=(PositiveWidth=3,NegativeWidth=1,PositiveHeight=1,NegativeHeight=1)
    FalloffSize=0.3
    OptionalParentItem=(Type=Bone)
    RotationOffset=(X=0,Y=0,Z=40)
   )
   > &RigUnit_ModifyTransforms(
    ItemToModify=((Item=(Type=Bone,Name="_hand_l_Sleeves3"),Transform=(Rotation=(X=0,Y=0,Z=-0.3007,W=0.9537),Translation=(X=0,Y=0,Z=0),Scale3D=(X=1,Y=1,Z=1))))
    Weight=@n2.OutputParam
   )
   > n3 = &RigUnit_SphericalPoseReader(
    ActiveRegionScaleFactors=(PositiveWidth=1,NegativeWidth=1,PositiveHeight=1,NegativeHeight=3)
    ActiveRegionSize=0.1
    Debug=(bDrawDebug=false,DebugScale=25,DebugSegments=20,DebugThickness=0.25)
    DriverAxis=(X=-1,Y=0,Z=0)
    DriverItem=(Type=Bone,Name="hand_r")
    FalloffRegionScaleFactors=(PositiveWidth=1,NegativeWidth=1,PositiveHeight=0.3,NegativeHeight=2)
    FalloffSize=0.4
    OptionalParentItem=(Type=Bone)
    RotationOffset=(X=20,Y=120,Z=50)
   )
   > &RigUnit_ModifyTransforms(
    ItemToModify=((Item=(Type=Bone,Name="_hand_r_Sleeves3"),Transform=(Rotation=(X=0.0151,Y=-0.173,Z=0.0858,W=0.9811),Translation=(X=5,Y=0,Z=0),Scale3D=(X=1,Y=1,Z=1))))
    Weight=@n3.OutputParam
   )
   > &RigUnit_ModifyTransforms(
    ItemToModify=((Item=(Type=Bone,Name="_hand_r_Sreeves_01"),Transform=(Rotation=(X=0,Y=0,Z=0,W=1),Translation=(X=0,Y=0,Z=0),Scale3D=(X=1,Y=1,Z=1))))
    Weight=@n3.OutputParam
   )
   > n4 = &RigUnit_SphericalPoseReader(
    ActiveRegionScaleFactors=(PositiveWidth=1,NegativeWidth=1,PositiveHeight=1,NegativeHeight=1)
    ActiveRegionSize=0.1
    Debug=(bDrawDebug=false,DebugScale=25,DebugSegments=20,DebugThickness=0.25)
    DriverAxis=(X=-1,Y=0,Z=0)
    DriverItem=(Type=Bone,Name="hand_r")
    FalloffRegionScaleFactors=(PositiveWidth=3,NegativeWidth=1,PositiveHeight=1,NegativeHeight=1)
    FalloffSize=0.3
    OptionalParentItem=(Type=Bone)
    RotationOffset=(X=20,Y=200,Z=40)
   )
   > &RigUnit_ModifyTransforms(
    ItemToModify=((Item=(Type=Bone,Name="_hand_r_Sleeves3"),Transform=(Rotation=(X=0,Y=0,Z=-0.1736,W=0.9848),Translation=(X=0,Y=0,Z=0),Scale3D=(X=1,Y=1,Z=1))))
    Weight=@n4.OutputParam
   )
   > &RigUnit_ParentConstraint(
    AdvancedSettings=(InterpolationType=Average,RotationOrderForFilter=XZY)
    Child=(Type=Bone,Name="dyn_collar_02")
    Parents=((Item=(Type=Bone,Name="neck_01"),Weight=1))
    Weight=0.5
   )
   n5 = &RigUnit_SphericalPoseReader(
    ActiveRegionScaleFactors=(PositiveWidth=1,NegativeWidth=1,PositiveHeight=1.5,NegativeHeight=1.5)
    ActiveRegionSize=0.1
    Debug=(bDrawDebug=false,DebugScale=25,DebugSegments=20,DebugThickness=0.25)
    DriverAxis=(X=1,Y=0,Z=0)
    DriverItem=(Type=Bone,Name="clavicle_l")
    FalloffRegionScaleFactors=(PositiveWidth=1.7,NegativeWidth=2,PositiveHeight=2,NegativeHeight=2)
    FalloffSize=0.35
    OptionalParentItem=(Type=Bone)
    RotationOffset=(X=20,Y=0,Z=10)
    ExecutePin=@RigUnit_SequenceAggregate.B
   )
   > &RigUnit_ModifyTransforms(
    ItemToModify=((Item=(Type=Bone,Name="dyn_collar_01_l"),Transform=(Rotation=(X=0,Y=0,Z=0,W=1),Translation=(X=0,Y=0,Z=0),Scale3D=(X=1,Y=1,Z=1))))
    ItemToModify.0.Transform=/Script/RigVM.RigVMFunction_MathTransformFromSRT(Location=(X=4,Y=0,Z=7), Rotation=(X=0,Y=40,Z=0), Scale=(X=1,Y=1,Z=1)) {is=struct RigVMUnitNode}.Transform
    Weight=@n5.OutputParam
   )
   > &RigUnit_ModifyTransforms(
    ItemToModify=((Item=(Type=Bone,Name="dyn_collar_02_l"),Transform=(Rotation=(X=0,Y=0,Z=0,W=1),Translation=(X=0,Y=0,Z=0),Scale3D=(X=1,Y=1,Z=1))))
    ItemToModify.0.Transform=&RigVMFunction_MathTransformFromSRT(Location=(X=2,Y=0,Z=7), Rotation=(X=20,Y=-20,Z=0), Scale=(X=1,Y=1,Z=1)).Transform
    Weight=@n5.OutputParam
   )
   > n6 = &RigUnit_SphericalPoseReader(
    ActiveRegionScaleFactors=(PositiveWidth=1.5,NegativeWidth=1.5,PositiveHeight=1,NegativeHeight=1)
    ActiveRegionSize=0.1
    Debug=(bDrawDebug=false,DebugScale=25,DebugSegments=20,DebugThickness=0.25)
    DriverAxis=(X=-1,Y=0,Z=0)
    DriverItem=(Type=Bone,Name="clavicle_r")
    FalloffRegionScaleFactors=(PositiveWidth=1,NegativeWidth=1,PositiveHeight=1,NegativeHeight=1)
    FalloffSize=0.42
    OptionalParentItem=(Type=Bone)
    RotationOffset=(X=-180,Y=20,Z=-110)
   )
   > &RigUnit_ModifyTransforms(
    ItemToModify=((Item=(Type=Bone,Name="dyn_collar_01_r"),Transform=(Rotation=(X=0,Y=0,Z=0,W=1),Translation=(X=0,Y=0,Z=0),Scale3D=(X=1,Y=1,Z=1))))
    ItemToModify.0.Transform=&RigVMFunction_MathTransformFromSRT(Location=(X=5,Y=0,Z=-6), Rotation=(X=0,Y=10,Z=0), Scale=(X=1,Y=1,Z=1)).Transform
    Weight=@n6.OutputParam
   )
   > &RigUnit_ModifyTransforms(
    ItemToModify=((Item=(Type=Bone,Name="dyn_collar_02_r"),Transform=(Rotation=(X=0,Y=0,Z=0,W=1),Translation=(X=0,Y=0,Z=0),Scale3D=(X=1,Y=1,Z=1))))
    ItemToModify.0.Transform=&RigVMFunction_MathTransformFromSRT(Location=(X=3,Y=0,Z=-3), Rotation=(X=0,Y=0,Z=0), Scale=(X=1,Y=1,Z=1)).Transform
    Weight=@n6.OutputParam
   )
   > n7 = &RigUnit_SphericalPoseReader(
    ActiveRegionScaleFactors=(PositiveWidth=1,NegativeWidth=1,PositiveHeight=3,NegativeHeight=3)
    ActiveRegionSize=0.05
    Debug=(bDrawDebug=false,bDrawLocalAxes=false,DebugScale=25,DebugSegments=20,DebugThickness=0.25)
    DriverAxis=(X=1,Y=0,Z=0)
    DriverItem=(Type=Bone,Name="upperarm_l")
    FalloffRegionScaleFactors=(PositiveWidth=1.8,NegativeWidth=2,PositiveHeight=2,NegativeHeight=2)
    FalloffSize=0.3
    OptionalParentItem=(Type=Bone)
    RotationOffset=(X=90,Y=-5,Z=0)
   )
   > &RigUnit_ModifyTransforms(
    ItemToModify=((Item=(Type=Bone,Name="dyn_collar_l"),Transform=(Rotation=(X=0,Y=0,Z=0,W=1),Translation=(X=0,Y=0,Z=0),Scale3D=(X=1,Y=1,Z=1))))
    ItemToModify.0.Transform.Rotation=&RigVMFunction_MathQuaternionFromEuler(Euler=(X=0,Y=0,Z=70)).Result
    Weight=@n7.OutputParam
   )
   > &RigUnit_ModifyTransforms(
    ItemToModify=((Item=(Type=Bone,Name="dyn_collar_l"),Transform=(Rotation=(X=0,Y=0,Z=0,W=1),Translation=(X=0,Y=0,Z=0),Scale3D=(X=1,Y=1,Z=1))))
    ItemToModify.0.Transform.Rotation=&RigVMFunction_MathQuaternionFromEuler(Euler=(X=0,Y=0,Z=20)).Result
    Weight=@n7.OutputParam
   )
   > n8 = &RigUnit_SphericalPoseReader(
    ActiveRegionScaleFactors=(PositiveWidth=2.5,NegativeWidth=3,PositiveHeight=1,NegativeHeight=1)
    ActiveRegionSize=0.05
    Debug=(bDrawDebug=false,DebugScale=25,DebugSegments=20,DebugThickness=0.25)
    DriverAxis=(X=-1,Y=0,Z=0)
    DriverItem=(Type=Bone,Name="upperarm_r")
    FalloffRegionScaleFactors=(PositiveWidth=0.7,NegativeWidth=2,PositiveHeight=1,NegativeHeight=2)
    FalloffSize=0.3
    FlipWidthScaling=true
    OptionalParentItem=(Type=Bone)
    RotationOffset=(X=130,Y=235,Z=170)
   )
   > &RigUnit_ModifyTransforms(
    ItemToModify=((Item=(Type=Bone,Name="dyn_collar_r"),Transform=(Rotation=(X=0,Y=0,Z=0,W=1),Translation=(X=-1,Y=0,Z=0),Scale3D=(X=1,Y=1,Z=1))))
    ItemToModify.0.Transform.Rotation=&RigVMFunction_MathQuaternionFromEuler(Euler=(X=0,Y=90,Z=-30)).Result
    Weight=@n8.OutputParam
   )
   > n9 = &RigUnit_SphericalPoseReader(
    ActiveRegionScaleFactors=(PositiveWidth=3.5,NegativeWidth=3,PositiveHeight=1,NegativeHeight=1)
    ActiveRegionSize=0.05
    Debug=(bDrawDebug=false,DebugScale=25,DebugSegments=20,DebugThickness=0.25)
    DriverAxis=(X=-1,Y=0,Z=0)
    DriverItem=(Type=Bone,Name="upperarm_r")
    FalloffRegionScaleFactors=(PositiveWidth=2,NegativeWidth=2,PositiveHeight=1,NegativeHeight=2)
    FalloffSize=0.3
    FlipWidthScaling=false
    OptionalParentItem=(Type=Bone)
    RotationOffset=(X=130,Y=235,Z=170)
   )
   > n10 = &RigUnit_SphericalPoseReader(
    ActiveRegionScaleFactors=(PositiveWidth=1,NegativeWidth=1,PositiveHeight=1,NegativeHeight=1)
    ActiveRegionSize=0.05
    Debug=(bDrawDebug=false,DebugScale=25,DebugSegments=20,DebugThickness=0.25)
    DriverAxis=(X=-1,Y=0,Z=0)
    DriverItem=(Type=Bone,Name="upperarm_r")
    FalloffRegionScaleFactors=(PositiveWidth=1,NegativeWidth=1,PositiveHeight=1,NegativeHeight=1)
    FalloffSize=0.1
    FlipWidthScaling=true
    OptionalParentItem=(Type=Bone)
    RotationOffset=(X=140,Y=160,Z=170)
   )
   Greater = /Script/RigVM.RigVMFunction_MathDoubleGreater(B=@n9.OutputParam, A=@n10.OutputParam) {is=struct RigVMUnitNode, template="Greater::Execute(in A,in B,out Result)"}
   DISPATCH_RigVMDispatch_If = RigVMDispatchNode(Condition=@Greater.Result, False=@n9.OutputParam, True=@n10.OutputParam) {template="DISPATCH_RigVMDispatch_If(in Condition,in True,in False,out Result)"}
   > &RigUnit_ModifyTransforms(
    ItemToModify=((Item=(Type=Bone,Name="dyn_rope_r_base"),Transform=(Rotation=(X=0.0076,Y=-0.0868,Z=0.0868,W=0.9924),Translation=(X=0,Y=-5,Z=0),Scale3D=(X=1,Y=1,Z=1))))
    Weight=@DISPATCH_RigVMDispatch_If.Result
   )
   > &RigUnit_ParentConstraint(
    AdvancedSettings=(RotationOrderForFilter=XZY)
    Child=(Type=Bone,Name="dyn_rope_l_base")
    Parents="((Item=(Type=Bone,Name=\"spine_04\"),Weight=0.0983),(Item=(Type=Bone,Name=\"spine_05\"),Weight=0.3366),(Item=(Type=Bone,Name=\"clavicle_l\"),Weight=0.4387),(Item=(Type=Bone,Name=\"upperarm_twist_01_l\"),Weight=0.1028),(Item=(Type=Bone,Name=\"deform_upperarm_fr_l\"), ...(truncated)"
    Weight=1
   )
   > n11 = &RigUnit_SphericalPoseReader(
    ActiveRegionScaleFactors=(PositiveWidth=4,NegativeWidth=3,PositiveHeight=1,NegativeHeight=1)
    ActiveRegionSize=0.05
    Debug=(bDrawDebug=false,bDrawLocalAxes=false,DebugScale=25,DebugSegments=20,DebugThickness=0.25)
    DriverAxis=(X=1,Y=0,Z=0)
    DriverItem=(Type=Bone,Name="upperarm_l")
    FalloffRegionScaleFactors=(PositiveWidth=4,NegativeWidth=2,PositiveHeight=2,NegativeHeight=2)
    FalloffSize=0.3
    OptionalParentItem=(Type=Bone)
    RotationOffset=(X=0,Y=-5,Z=0)
   )
   > &RigUnit_ModifyTransforms(
    ItemToModify=((Item=(Type=Bone,Name="dyn_rope_l_base"),Transform=(Rotation=(X=0,Y=0,Z=0,W=1),Translation=(X=5,Y=0,Z=0),Scale3D=(X=1,Y=1,Z=1))))
    ItemToModify.0.Transform.Rotation=&RigVMFunction_MathQuaternionFromEuler(Euler=(X=0,Y=20,Z=10)).Result
    Weight=@n11.OutputParam
   )
   > &RigUnit_ParentConstraint(
    AdvancedSettings=(RotationOrderForFilter=XZY)
    Child=(Type=Bone,Name="dyn_rope_r_base")
    Parents="((Item=(Type=Bone,Name=\"spine_04\"),Weight=0.0963),(Item=(Type=Bone,Name=\"spine_05\"),Weight=0.2849),(Item=(Type=Bone,Name=\"clavicle_r\"),Weight=0.4534),(Item=(Type=Bone,Name=\"upperarm_twist_01_r\"),Weight=0.1292),(Item=(Type=Bone,Name=\"deform_upperarm_fr_r\"), ...(truncated)"
    Weight=1
   )
   > n12 = &RigUnit_SphericalPoseReader(
    ActiveRegionScaleFactors=(PositiveWidth=2.5,NegativeWidth=4,PositiveHeight=1,NegativeHeight=1)
    ActiveRegionSize=0.05
    Debug=(bDrawDebug=false,DebugScale=25,DebugSegments=20,DebugThickness=0.25)
    DriverAxis=(X=-1,Y=0,Z=0)
    DriverItem=(Type=Bone,Name="upperarm_r")
    FalloffRegionScaleFactors=(PositiveWidth=1,NegativeWidth=4,PositiveHeight=2,NegativeHeight=2)
    FalloffSize=0.3
    FlipWidthScaling=true
    OptionalParentItem=(Type=Bone)
    RotationOffset=(X=20,Y=190,Z=0)
   )
   > &RigUnit_ModifyTransforms(
    ItemToModify=((Item=(Type=Bone,Name="dyn_rope_r_base"),Transform=(Rotation=(X=0.05,Y=-0.0714,Z=0.5714,W=0.816),Translation=(X=0,Y=0,Z=0),Scale3D=(X=1,Y=1,Z=1))))
    Weight=@n12.OutputParam
   )
   > n13 = &RigUnit_SphericalPoseReader(
    ActiveRegionScaleFactors=(PositiveWidth=1,NegativeWidth=1,PositiveHeight=1,NegativeHeight=1)
    ActiveRegionSize=0.05
    Debug=(bDrawDebug=true,DebugScale=25,DebugSegments=20,DebugThickness=0.25)
    DriverAxis=(X=-1,Y=0,Z=0)
    DriverItem=(Type=Bone,Name="upperarm_r")
    FalloffRegionScaleFactors=(PositiveWidth=1,NegativeWidth=1,PositiveHeight=2,NegativeHeight=2)
    FalloffSize=0.1
    FlipWidthScaling=true
    OptionalParentItem=(Type=Bone)
    RotationOffset=(X=-85,Y=20,Z=10)
   )
   > &RigUnit_ModifyTransforms(
    ItemToModify=((Item=(Type=Bone,Name="dyn_rope_r_base"),Transform=(Rotation=(X=0,Y=-0.0175,Z=0,W=0.9998),Translation=(X=0,Y=-2,Z=-1),Scale3D=(X=1,Y=1,Z=1))))
    Weight=@n13.OutputParam
   )
   SequenceAggregate_1_1_1 = RigVMAggregateNode(ExecuteContext=@RigUnit_SequenceAggregate.C)
   n14 = &RigUnit_SphericalPoseReader(
    ActiveRegionScaleFactors=(PositiveWidth=1,NegativeWidth=0,PositiveHeight=1,NegativeHeight=1)
    ActiveRegionSize=0.3
    Debug=(bDrawDebug=false,bDraw2D=false,bDrawLocalAxes=false,DebugScale=30,DebugSegments=30,DebugThickness=0.2)
    DriverAxis=(X=1,Y=0,Z=0)
    DriverItem=(Type=Bone,Name="thigh_r")
    FalloffRegionScaleFactors=(PositiveWidth=1,NegativeWidth=1,PositiveHeight=1,NegativeHeight=1)
    FalloffSize=0.5
    FlipWidthScaling=false
    OptionalParentItem=(Type=Bone,Name="pelvis")
    RotationOffset=(X=-90,Y=180,Z=-20)
    ExecutePin=@SequenceAggregate_1_1_1.A
   )
   > &RigUnit_ModifyTransforms(
    ItemToModify=((Item=(Type=Bone,Name="dyn_blouse_base_fr_r"),Transform=(Rotation=(X=-0.7236,Y=-0.2409,Z=-0.0211,W=0.6465),Translation=(X=0,Y=0,Z=0),Scale3D=(X=1,Y=1,Z=1))))
    ItemToModify.0.Transform=&RigVMFunction_MathTransformFromSRT(Location=(X=-4,Y=5,Z=5), Rotation=(X=-50,Y=-60,Z=10), Scale=(X=1,Y=1,Z=1)).Transform
    Weight=@n14.OutputParam
   )
   > &RigUnit_ModifyTransforms(
    ItemToModify=((Item=(Type=Bone,Name="dyn_blouse_base_side_r"),Transform=(Rotation=(X=0,Y=0,Z=0,W=1),Translation=(X=0,Y=0,Z=0),Scale3D=(X=1,Y=1,Z=1))))
    Weight=1
   )
   > n15 = &RigUnit_SphericalPoseReader(
    ActiveRegionScaleFactors=(PositiveWidth=1,NegativeWidth=0,PositiveHeight=1,NegativeHeight=1)
    ActiveRegionSize=0.3
    Debug=(bDrawDebug=false,bDraw2D=false,bDrawLocalAxes=false,DebugScale=30,DebugSegments=30,DebugThickness=0.2)
    DriverAxis=(X=-1,Y=0,Z=0)
    DriverItem=(Type=Bone,Name="calf_l")
    FalloffRegionScaleFactors=(PositiveWidth=1,NegativeWidth=0.1,PositiveHeight=1,NegativeHeight=1)
    FalloffSize=0.5
    FlipWidthScaling=false
    OptionalParentItem=(Type=Bone,Name="pelvis")
    RotationOffset=(X=-90,Y=0,Z=20)
   )
   > &RigUnit_ModifyTransforms(
    ItemToModify=((Item=(Type=Bone,Name="dyn_blouse_base_fr_l"),Transform=(Rotation=(X=0.0151,Y=0.0858,Z=-0.173,W=0.9811),Translation=(X=10,Y=0,Z=0),Scale3D=(X=1,Y=1,Z=1))))
    Weight=@n15.OutputParam
   )
   > &RigUnit_ModifyTransforms(
    ItemToModify=((Item=(Type=Bone,Name="dyn_blouse_base_side_l"),Transform=(Rotation=(X=-0.0449,Y=-0.1677,Z=-0.2549,W=0.9513),Translation=(X=0,Y=0,Z=0),Scale3D=(X=1,Y=1,Z=1))))
    Weight=@n15.OutputParam
   )
   > n16 = &RigUnit_SphericalPoseReader(
    ActiveRegionScaleFactors=(PositiveWidth=1,NegativeWidth=0,PositiveHeight=1,NegativeHeight=1)
    ActiveRegionSize=0.3
    Debug=(bDrawDebug=false,bDraw2D=false,bDrawLocalAxes=false,DebugScale=30,DebugSegments=30,DebugThickness=0.2)
    DriverAxis=(X=1,Y=0,Z=0)
    DriverItem=(Type=Bone,Name="calf_r")
    FalloffRegionScaleFactors=(PositiveWidth=1,NegativeWidth=0.1,PositiveHeight=1,NegativeHeight=1)
    FalloffSize=0.5
    FlipWidthScaling=false
    OptionalParentItem=(Type=Bone,Name="pelvis")
    RotationOffset=(X=50,Y=180,Z=25)
   )
   > n17 = &RigUnit_ModifyTransforms(
    ItemToModify=((Item=(Type=Bone,Name="dyn_blouse_base_fr_r"),Transform=(Rotation=(X=-0.4162,Y=-0.1574,Z=-0.0734,W=0.8925),Translation=(X=0,Y=0,Z=0),Scale3D=(X=1,Y=1,Z=1))))
    Weight=@n16.OutputParam
   )
   RigVMFunction_ControlFlowBranch = &RigVMFunction_ControlFlowBranch(ExecuteContext=@n17.ExecutePin, Condition=RigVMVariableNode() {variable=Is baseball sliding}.Value)
   n18 = &RigUnit_SphericalPoseReader(
    ActiveRegionScaleFactors=(PositiveWidth=3,NegativeWidth=0,PositiveHeight=2,NegativeHeight=1)
    ActiveRegionSize=0.07
    Debug=(bDrawDebug=false,bDraw2D=false,bDrawLocalAxes=false,DebugScale=30,DebugSegments=30,DebugThickness=0.2)
    DriverAxis=(X=1,Y=0,Z=0)
    DriverItem=(Type=Bone,Name="calf_r")
    FalloffRegionScaleFactors=(PositiveWidth=0.5,NegativeWidth=0.1,PositiveHeight=0.3,NegativeHeight=0.3)
    FalloffSize=0.3
    FlipWidthScaling=false
    OptionalParentItem=(Type=Bone,Name="pelvis")
    RotationOffset=(X=40,Y=-145,Z=-40)
    ExecutePin=@RigVMFunction_ControlFlowBranch.True
   )
   > &RigUnit_ModifyTransforms(
    ItemToModify=((Item=(Type=Bone,Name="dyn_blouse_base_fr_r"),Transform=(Rotation=(X=0.116,Y=-0.5016,Z=0.3512,W=0.782),Translation=(X=0,Y=0,Z=5),Scale3D=(X=1,Y=1,Z=1))))
    Weight=@n18.OutputParam
   )
   > &RigUnit_ModifyTransforms(
    ItemToModify=((Item=(Type=Bone,Name="dyn_blouse_base_side_r"),Transform=(Rotation=(X=0.2988,Y=-0.2988,Z=0.6409,W=0.6409),Translation=(X=0,Y=0,Z=5),Scale3D=(X=1,Y=1,Z=1))))
    Weight=@n18.OutputParam
   )
   > &RigUnit_ModifyTransforms(
    ItemToModify=((Item=(Type=Bone,Name="dyn_blouse_base_bk_r"),Transform=(Rotation=(X=0,Y=-0.5736,Z=0,W=0.8192),Translation=(X=0,Y=0,Z=5),Scale3D=(X=1,Y=1,Z=1))))
    Weight=@n18.OutputParam
   )
   > &RigUnit_ModifyTransforms(
    ItemToModify=((Item=(Type=Bone,Name="dyn_blouse_base_01_c"),Transform=(Rotation=(X=0,Y=-0.5,Z=0,W=0.866),Translation=(X=0,Y=0,Z=-5),Scale3D=(X=1,Y=1,Z=1))))
    Weight=@n18.OutputParam
   )
   n19 = &RigUnit_SphericalPoseReader(
    ActiveRegionScaleFactors=(PositiveWidth=1,NegativeWidth=0.1,PositiveHeight=1,NegativeHeight=1)
    ActiveRegionSize=0.3
    Debug=(bDrawDebug=true,bDraw2D=false,bDrawLocalAxes=false,DebugScale=30,DebugSegments=30,DebugThickness=0.2)
    DriverAxis=(X=-1,Y=0,Z=0)
    DriverItem=(Type=Bone,Name="thigh_l")
    FalloffRegionScaleFactors=(PositiveWidth=1,NegativeWidth=0.6,PositiveHeight=1,NegativeHeight=1)
    FalloffSize=0.5
    OptionalParentItem=(Type=Bone,Name="pelvis")
    RotationOffset=(X=90,Y=0,Z=0)
    ExecutePin=@SequenceAggregate_1_1_1.B
   )
   KalmanFilter = &RigVMFunction_KalmanFloat(BufferSize=8, Value=@n19.OutputParam) {template="KalmanFilter::Execute(in Value,in BufferSize,out Result)"}
   > &RigUnit_ModifyTransforms(
    ItemToModify=((Item=(Type=Bone,Name="dyn_blouse_base_fr_l"),Transform=(Rotation=(X=0.75,Y=-0.433,Z=-0.25,W=0.433),Translation=(X=0,Y=5,Z=0),Scale3D=(X=1,Y=1,Z=1))))
    Weight=@KalmanFilter.Result
    ItemToModify.0.Transform=&RigVMFunction_MathTransformFromSRT(Location=(X=0,Y=-3,Z=0), Rotation=(X=90,Y=-10,Z=-40), RotationOrder=XYZ, Scale=(X=1,Y=1,Z=1)).Transform
   )
   > &RigUnit_ModifyTransforms(
    ItemToModify=((Item=(Type=Bone,Name="dyn_blouse_base_side_l"),Transform=(Rotation=(X=0,Y=0,Z=0,W=1),Translation=(X=0,Y=0,Z=0),Scale3D=(X=1,Y=1,Z=1))))
    Weight=@KalmanFilter.Result
    ItemToModify.0.Transform=&RigVMFunction_MathTransformFromSRT(Location=(X=-1,Y=-3,Z=0), Rotation=(X=30,Y=0,Z=0), RotationOrder=XYZ, Scale=(X=1,Y=1,Z=1)).Transform
   )
   > n20 = &RigUnit_SphericalPoseReader(
    ActiveRegionScaleFactors=(PositiveWidth=1,NegativeWidth=1,PositiveHeight=1,NegativeHeight=4)
    ActiveRegionSize=0.05
    Debug=(bDrawDebug=false,bDrawLocalAxes=false,DebugScale=25,DebugSegments=20,DebugThickness=0.25)
    DriverAxis=(X=1,Y=0,Z=0)
    DriverItem=(Type=Bone,Name="upperarm_l")
    FalloffRegionScaleFactors=(PositiveWidth=0.5,NegativeWidth=0.5,PositiveHeight=0.5,NegativeHeight=2)
    FalloffSize=0.3
    OptionalParentItem=(Type=Bone)
    RotationOffset=(X=120,Y=10,Z=-60)
   )
   Multiply = &RigVMFunction_MathDoubleMul(A=@KalmanFilter.Result, B=@n20.OutputParam)
   > &RigUnit_ModifyTransforms(
    ItemToModify=((Item=(Type=Bone,Name="dyn_blouse_base_fr_l"),Transform=(Rotation=(X=0,Y=0,Z=0.0087,W=1),Translation=(X=0,Y=-2,Z=3),Scale3D=(X=1,Y=1,Z=1))))
    Weight=&RigVMFunction_KalmanFloat(BufferSize=8, Value=@Multiply.Result) {template="KalmanFilter::Execute(in Value,in BufferSize,out Result)"}.Result
   )
   n21 = &RigVMFunction_ControlFlowBranch(ExecuteContext=@SequenceAggregate_1_1_1.D, Condition=RigVMVariableNode() {variable=Is DBNO On}.Value)
   &RigUnit_ModifyTransforms(
    ItemToModify=((Item=(Type=Bone,Name="dyn_blouse_base_fr_l"),Transform=(Rotation=(X=0,Y=0,Z=0.2588,W=0.9659),Translation=(X=-5,Y=0,Z=0),Scale3D=(X=1,Y=1,Z=1))))
    Weight=1
    ExecutePin=@n21.True
   )
   > &RigUnit_ModifyTransforms(
    ItemToModify=((Item=(Type=Bone,Name="dyn_blouse_base_fr_r"),Transform=(Rotation=(X=0,Y=0,Z=0,W=1),Translation=(X=0,Y=5,Z=0),Scale3D=(X=1,Y=1,Z=1))))
    Weight=1
   )
   > &RigUnit_ModifyTransforms(
    ItemToModify=((Item=(Type=Bone,Name="dyn_blouse_fr_03_r"),Transform=(Rotation=(X=0,Y=0,Z=0,W=1),Translation=(X=0,Y=5,Z=0),Scale3D=(X=1,Y=1,Z=1))))
    Weight=1
   )
   > &RigUnit_ModifyTransforms(
    ItemToModify=((Item=(Type=Bone,Name="dyn_blouse_base_bk_l"),Transform=(Rotation=(X=-0.0076,Y=0.0868,Z=0.0868,W=0.9924),Translation=(X=3,Y=0,Z=0),Scale3D=(X=1,Y=1,Z=1))))
    Weight=1
   )
   > &RigUnit_ModifyTransforms(
    ItemToModify=((Item=(Type=Bone,Name="dyn_blouse_base_bk_r"),Transform=(Rotation=(X=0,Y=0,Z=0,W=1),Translation=(X=5,Y=0,Z=0),Scale3D=(X=1,Y=1,Z=1))))
    Weight=1
   )
   > &RigUnit_ModifyTransforms(
    ItemToModify=((Item=(Type=Bone,Name="dyn_blouse_base_01_c"),Transform=(Rotation=(X=0,Y=0,Z=0,W=1),Translation=(X=5,Y=0,Z=0),Scale3D=(X=1,Y=1,Z=1))))
    Weight=1
   )
   > &RigUnit_ModifyTransforms(
    ItemToModify=((Item=(Type=Bone,Name="dyn_blouse_base_side_l"),Transform=(Rotation=(X=0,Y=0,Z=0.1736,W=0.9848),Translation=(X=0,Y=0,Z=0),Scale3D=(X=1,Y=1,Z=1))))
    Weight=1
   )
   > &RigUnit_ModifyTransforms(
    ItemToModify=((Item=(Type=Bone,Name="dyn_blouse_base_side_r"),Transform=(Rotation=(X=-0.0449,Y=-0.1677,Z=-0.2549,W=0.9513),Translation=(X=0,Y=0,Z=0),Scale3D=(X=1,Y=1,Z=1))))
    Weight=1
   )
   RigVMFunction_MathQuaternionToEuler = &RigVMFunction_MathQuaternionToEuler(
    RotationOrder=ZYX
    Value=&RigUnit_GetRelativeTransformForItem(Child=(Type=Bone,Name="spine_04"), Parent=(Type=Bone,Name="spine_01"), bChildInitial=False, bParentInitial=false).RelativeTransform.Rotation
   )
   Remap = &RigVMFunction_MathDoubleRemap(SourceMaximum=20, SourceMinimum=10, TargetMaximum=1, TargetMinimum=0, Value=@RigVMFunction_MathQuaternionToEuler.Result.X)
   n22 = &RigVMFunction_MathDoubleMul(
    B=&RigVMFunction_KalmanFloat(BufferSize=16, Value=@Remap.Result) {template="KalmanFilter::Execute(in Value,in BufferSize,out Result)"}.Result
    A=@n19.OutputParam
   )
   &RigUnit_ModifyTransforms(
    ItemToModify=((Item=(Type=Bone,Name="dyn_blouse_base_fr_l"),Transform=(Rotation=(X=0,Y=0,Z=0,W=1),Translation=(X=0,Y=0,Z=0),Scale3D=(X=1,Y=1,Z=1))))
    WeightMinimum=0
    Weight=@n22.Result
    ItemToModify.0.Transform=&RigVMFunction_MathTransformFromSRT(Location=(X=3,Y=-2,Z=0), Rotation=(X=0,Y=-10,Z=0), RotationOrder=XYZ, Scale=(X=1,Y=1,Z=1)).Transform
    ExecutePin=@SequenceAggregate_1_1_1.C
   )
   > &RigUnit_ModifyTransforms(
    ItemToModify=((Item=(Type=Bone,Name="bag_belt_correction"),Transform=(Rotation=(X=0,Y=0,Z=0,W=1),Translation=(X=0,Y=0,Z=0),Scale3D=(X=1,Y=1,Z=1))))
    WeightMinimum=0
    Weight=@n22.Result
    ItemToModify.0.Transform=&RigVMFunction_MathTransformFromSRT(Location=(X=3,Y=-8,Z=15), Rotation=(X=0,Y=0,Z=0), RotationOrder=XYZ, Scale=(X=1,Y=1,Z=1)).Transform
   )
   n23 = &RigUnit_GetTransform(Item=(Type=Bone,Name="calf_l"))
   n24 = &RigUnit_GetTransform(Item=(Type=Bone,Name="calf_r"))
   > &RigUnit_ModifyTransforms(
    ItemToModify=((Item=(Type=Bone,Name="AltSkin_Corrective"),Transform=(Rotation=(X=0,Y=0,Z=0,W=1),Translation=(X=0,Y=0,Z=0),Scale3D=(X=1,Y=1,Z=1))))
    Weight=1
    WeightMinimum=0
    ItemToModify.0.Transform.Translation.Y=&RigVMFunction_MathDoubleMul(B=-0.15, A=&RigVMFunction_MathVectorDistance(A=@n23.Transform.Translation, B=@n24.Transform.Translation).Result).Result
   )
   FN_DeformRig_FNC_1 = RigVMFunctionReferenceNode(
    DeformRig_ Curve="(deformRig_l_upperarm_up_alpha_72_#1=False,deformRig_r_upperarm_up_alpha_73_#2=False,deformRig_l_upperarm_fr_alpha_74_#3=False,deformRig_r_upperarm_fr_alpha_75_CB51B8 ...(truncated)"
    DeformRig_Bool="(upperarm_1_#4=True,elbow_3_#5=True,wrist_7_#6=True,groin_8_#7=True,glute_17_#8=true,knee_9_40C408324CA261 ...(truncated)"
    DeformRig_Debug=(upperarm_debug_18_#4=False,wrist_debug_20_#6=False,lat_debug_26_#9=False,groinGlute_debug_27_#7=False)
    DeformRig_Value="(upperarm_up_18_#10=(Rotation=(X=0,Y=0,Z=0,W=1),Translation=(X=0,Y=0,Z=8),Scale3D=(X=0.7,Y=1,Z=1)),upperarm_fwd_fr_19_#11=(Rotation=(X=0,Y=0,Z=0,W=1),Translation=(X=0,Y=10,Z=0),Scale3D=(X=0.5,Y=1,Z= ...(truncated)"
    ExecuteContext=@RigUnit_SequenceAggregate.D
   )
   &RigVMFunction_KalmanFloat(BufferSize=16, Value=@n14.OutputParam) {template="KalmanFilter::Execute(in Value,in BufferSize,out Result)"}
  ]}
 SequenceAggregate_1_1_1_1_ContainedGraph {nodes [
   n1 = RigVMFunctionEntryNode()
   SequenceAggregate_1_1_1 = &RigVMFunction_Sequence(ExecuteContext=@n1.ExecuteContext)
   SequenceAggregate_1_1_1_1 = &RigVMFunction_Sequence(ExecuteContext=@SequenceAggregate_1_1_1.B)
   SequenceAggregate_1_1_1_2 = &RigVMFunction_Sequence(ExecuteContext=@SequenceAggregate_1_1_1_1.B)
   RigVMFunctionReturnNode(A=@SequenceAggregate_1_1_1.A, B=@SequenceAggregate_1_1_1_1.A, C=@SequenceAggregate_1_1_1_2.A, D=@SequenceAggregate_1_1_1_2.B)
  ]}
]
hierarchy [
 root {loc=<0 0 0>, rot=<0 0 0 1>}
 pelvis {parent=root, loc=<0 -0.5301 98.0571>, rot=<0.0127 0.707 -0.0127 -0.707>}
 spine_01 {parent=pelvis, loc=<6.1208 -0.9328 0>, rot=<0 0 0.034 -0.9994>}
 spine_02 {parent=spine_01, loc=<6.5996 -0.2095 0>, rot=<0 0 0.0238 -0.9997>}
 spine_03 {parent=spine_02, loc=<6.4914 -0.1237 0>, rot=<0 0 0.0149 -0.9999>}
 spine_04 {parent=spine_03, loc=<6.9598 0.0794 0>, rot=<0 0 -0.0132 -0.9999>}
 spine_05 {parent=spine_04, loc=<6.9987 0 0>, rot=<0 0 -0.0015 -1>}
 neck_01 {parent=spine_05, loc=<11.3138 0.1529 0>, rot=<0 0 -0.2003 -0.9797>}
 neck_02 {parent=neck_01, loc=<4.5627 0 0>, rot=<0 0 0.0504 -0.9987>}
 head {parent=neck_02, loc=<5.3914 0 0>, rot=<0 0 0.115 -0.9934>}
 attach_backpack {parent=spine_05, loc=<0 -8.35 0>, rot=<-0.0093 -0.707 0.0093 -0.707>}
 clavicle_l {parent=spine_05, loc=<5.4831 -0.1053 -3.2946>, rot=<0.0465 -0.7289 -0.009 -0.6829>}
 upperarm_l {parent=clavicle_l, loc=<9.8658 -1.2136 -0.4105>, rot=<0.0164 -0.3245 -0.0269 -0.9454>}
 lowerarm_l {parent=upperarm_l, loc=<26.465 -0.8345 0.259>, rot=<0.0036 0.0161 -0.2298 -0.9731>}
 hand_l {parent=lowerarm_l, loc=<22.7355 -0.3203 -0.6105>, rot=<-0.6784 -0.0223 0.0735 -0.7307>}
 thumb_01_l {parent=hand_l, loc=<1.3118 -0.3191 -2.459>, rot=<0.6546 -0.4042 0.0664 -0.6354>}
 thumb_02_l {parent=thumb_01_l, loc=<5.9336 -0.0448 0.0401>, rot=<-0.0037 0.0213 0.298 -0.9543>}
 thumb_03_l {parent=thumb_02_l, loc=<2.2223 0.3682 -0.0907>, rot=<0.0118 -0.013 0.0066 -0.9998>}
 index_metacarpal_l {parent=hand_l, loc=<1.6535 0.0183 -2.96>, rot=<0.0158 -0.0124 -0.0173 -0.9996>}
 index_01_l {parent=index_metacarpal_l, loc=<8.3008 0.0357 -0.0524>, rot=<-0.0044 0.0485 0.0924 -0.9945>}
 index_02_l {parent=index_01_l, loc=<3.3587 -0.3004 0.0452>, rot=<-0.002 0.0027 0.3225 -0.9466>}
 index_03_l {parent=index_02_l, loc=<2.5578 -0.123 -0.0252>, rot=<-0.002 0.0067 0.2558 -0.9667>}
 middle_metacarpal_l {parent=hand_l, loc=<2.1856 -0.1725 -0.5984>, rot=<0.006 0.0168 -0.0329 -0.9993>}
 middle_01_l {parent=middle_metacarpal_l, loc=<7.3275 -0.2447 0.0129>, rot=<-0.0208 0.0042 0.0858 -0.9961>}
 middle_02_l {parent=middle_01_l, loc=<4.0123 0.1401 0.1175>, rot=<-0.0218 0.0258 0.3512 -0.9357>}
 middle_03_l {parent=middle_02_l, loc=<2.6077 0.0534 0.1373>, rot=<0.0082 0.0188 0.1708 -0.9851>}
 ring_metacarpal_l {parent=hand_l, loc=<2.2074 -0.3625 1.3638>, rot=<-0.0284 0.0722 -0.0458 -0.9959>}
 ring_01_l {parent=ring_metacarpal_l, loc=<7.0295 -0.6197 -0.1446>, rot=<-0.0489 -0.0252 0.1892 -0.9804>}
 ring_02_l {parent=ring_01_l, loc=<3.6174 0.4647 -0.1027>, rot=<0 0 0.2806 -0.9598>}
 ring_03_l {parent=ring_02_l, loc=<2.4152 0.114 0.0199>, rot=<0 0 0.1189 -0.9929>}
 pinky_metacarpal_l {parent=hand_l, loc=<1.7616 -0.9753 3.3814>, rot=<-0.0397 0.0814 -0.069 -0.9935>}
 pinky_01_l {parent=pinky_metacarpal_l, loc=<7.1709 -0.4831 -0.1094>, rot=<-0.0654 -0.0035 0.2242 -0.9723>}
 pinky_02_l {parent=pinky_01_l, loc=<2.8274 0.0893 -0.1043>, rot=<0.0037 0.0011 0.2854 -0.9584>}
 pinky_03_l {parent=pinky_02_l, loc=<2.2046 0.0307 -0.0078>, rot=<-0.0002 0.0098 0.1445 -0.9895>}
 deform_wrist_dn_base_l {parent=hand_l, loc=<0 -1.5326 0>, rot=<0 0 0 1>}
 deform_wrist_dn_l {parent=deform_wrist_dn_base_l, loc=<0 0 0>, rot=<0 0 0 1>}
 deform_wrist_up_base_l {parent=hand_l, loc=<0 2.3009 0>, rot=<0 0 0 1>}
 deform_wrist_up_l {parent=deform_wrist_up_base_l, loc=<0 0 0>, rot=<0 0 0 1>}
 weapon_l {parent=hand_l, loc=<-0.1974 -3.4111 0.3811>, rot=<0 0 0 1>}
 hand_attach_l {parent=hand_l, loc=<8.5 -3 0>, rot=<0.7071 0.7071 0 0>}
 lowerarm_twist_02_l {parent=lowerarm_l, loc=<10.6707 0 0>, rot=<0 0 0 1>}
 lowerarm_twist_01_l {parent=lowerarm_l, loc=<18.316 0 0>, rot=<0 0 0 1>}
 _hand_l_Sleeves3 {parent=lowerarm_twist_01_l, loc=<0 0 0>, rot=<0.0285 -0.0335 0.0024 -0.999>}
 _hand_l_Sleeves_01 {parent=_hand_l_Sleeves3, loc=<6.1125 2.9938 0.0239>, rot=<0 0 0 1>}
 _hand_l_Sleeves_02 {parent=_hand_l_Sleeves3, loc=<-3.1998 0.052 -4.1704>, rot=<0 0 0 1>}
 _hand_l_Sleeves_02_01 {parent=_hand_l_Sleeves_02, loc=<0.0533 0.8912 -2.6501>, rot=<0 0 0 1>}
 _hand_l_Sleeves_03 {parent=_hand_l_Sleeves3, loc=<5.4627 -3.9744 -0.8673>, rot=<0 0 0 1>}
 _hand_l_Sleeves_04 {parent=_hand_l_Sleeves3, loc=<6.1997 -0.3873 2.7024>, rot=<0 0 0 1>}
 deform_elbow_in_base_l {parent=lowerarm_l, loc=<0.1834 3.6062 -0.2572>, rot=<0 0 0 1>}
 deform_elbow_in_l {parent=deform_elbow_in_base_l, loc=<0 0 0>, rot=<0 0 0 1>}
 deform_elbow_out_base_l {parent=lowerarm_l, loc=<0 -4 0>, rot=<0 0 0 1>}
 deform_elbow_out_l {parent=deform_elbow_out_base_l, loc=<0 0 0>, rot=<0 0 0 1>}
 upperarm_twist_01_l {parent=upperarm_l, loc=<7.2135 -0.1134 0>, rot=<0 0 0 1>}
 deform_upperarm_bk_base_l {parent=upperarm_twist_01_l, loc=<-6.4053 -4.3746 0.314>, rot=<0 0 0 1>}
 deform_upperarm_bk_l {parent=deform_upperarm_bk_base_l, loc=<0 0 0>, rot=<0 0 0 1>}
 deform_upperarm_fr_base_l {parent=upperarm_twist_01_l, loc=<-5.6312 5.092 0.3354>, rot=<0 0 0 1>}
 deform_upperarm_fr_l {parent=deform_upperarm_fr_base_l, loc=<0 0 0>, rot=<0 0 0 1>}
 deform_upperarm_up_base_l {parent=upperarm_twist_01_l, loc=<-8.891 0.4825 3.9517>, rot=<0 0 0 1>}
 deform_upperarm_up_l {parent=deform_upperarm_up_base_l, loc=<0 0 0>, rot=<0 0 0 1>}
 upperarm_twist_02_l {parent=upperarm_l, loc=<19.7998 -0.1134 0>, rot=<0 0 0 1>}
 scale_shoulder_l {parent=upperarm_l, loc=<0 0 0>, rot=<0 0 0 1>}
 clavicle_r {parent=spine_05, loc=<5.4831 -0.1053 3.2946>, rot=<0.7289 0.0465 -0.6829 0.009>}
 upperarm_r {parent=clavicle_r, loc=<-9.8658 1.2136 0.4105>, rot=<0.0164 -0.3245 -0.0269 -0.9454>}
 lowerarm_r {parent=upperarm_r, loc=<-26.465 0.8345 -0.259>, rot=<0.0036 0.0161 -0.2298 -0.9731>}
 hand_r {parent=lowerarm_r, loc=<-22.7355 0.3203 0.6106>, rot=<-0.6784 -0.0223 0.0735 -0.7307>}
 thumb_01_r {parent=hand_r, loc=<-1.3118 0.3191 2.459>, rot=<0.6546 -0.4042 0.0664 -0.6354>}
 thumb_02_r {parent=thumb_01_r, loc=<-5.9336 0.0448 -0.0401>, rot=<-0.0037 0.0213 0.298 -0.9543>}
 thumb_03_r {parent=thumb_02_r, loc=<-2.2223 -0.3682 0.0907>, rot=<0.0118 -0.013 0.0066 -0.9998>}
 index_metacarpal_r {parent=hand_r, loc=<-1.6535 -0.0183 2.96>, rot=<0.0158 -0.0124 -0.0173 -0.9996>}
 index_01_r {parent=index_metacarpal_r, loc=<-8.3008 -0.0357 0.0524>, rot=<-0.0044 0.0485 0.0924 -0.9945>}
 index_02_r {parent=index_01_r, loc=<-3.3587 0.3004 -0.0452>, rot=<-0.002 0.0027 0.3225 -0.9466>}
 index_03_r {parent=index_02_r, loc=<-2.5578 0.123 0.0252>, rot=<-0.002 0.0067 0.2558 -0.9667>}
 middle_metacarpal_r {parent=hand_r, loc=<-2.1856 0.1725 0.5984>, rot=<0.006 0.0168 -0.0329 -0.9993>}
 middle_01_r {parent=middle_metacarpal_r, loc=<-7.3275 0.2447 -0.0129>, rot=<-0.0208 0.0042 0.0858 -0.9961>}
 middle_02_r {parent=middle_01_r, loc=<-4.0123 -0.1401 -0.1175>, rot=<-0.0218 0.0258 0.3512 -0.9357>}
 middle_03_r {parent=middle_02_r, loc=<-2.6077 -0.0534 -0.1373>, rot=<0.0082 0.0188 0.1708 -0.9851>}
 ring_metacarpal_r {parent=hand_r, loc=<-2.2074 0.3625 -1.3638>, rot=<-0.0284 0.0722 -0.0458 -0.9959>}
 ring_01_r {parent=ring_metacarpal_r, loc=<-7.0295 0.6197 0.1446>, rot=<-0.0489 -0.0252 0.1892 -0.9804>}
 ring_02_r {parent=ring_01_r, loc=<-3.6174 -0.4647 0.1027>, rot=<0 0 0.2806 -0.9598>}
 ring_03_r {parent=ring_02_r, loc=<-2.4152 -0.114 -0.0199>, rot=<0 0 0.1189 -0.9929>}
 pinky_metacarpal_r {parent=hand_r, loc=<-1.7616 0.9753 -3.3814>, rot=<-0.0397 0.0814 -0.069 -0.9935>}
 pinky_01_r {parent=pinky_metacarpal_r, loc=<-7.1709 0.4831 0.1094>, rot=<-0.0654 -0.0035 0.2242 -0.9723>}
 pinky_02_r {parent=pinky_01_r, loc=<-2.8274 -0.0893 0.1043>, rot=<0.0037 0.0011 0.2854 -0.9584>}
 pinky_03_r {parent=pinky_02_r, loc=<-2.2046 -0.0307 0.0078>, rot=<-0.0002 0.0098 0.1445 -0.9895>}
 deform_wrist_dn_base_r {parent=hand_r, loc=<-0.0001 1.5328 0>, rot=<0 0 0 1>}
 deform_wrist_dn_r {parent=deform_wrist_dn_base_r, loc=<0 0 0>, rot=<0 0 0 1>}
 deform_wrist_up_base_r {parent=hand_r, loc=<-0.0002 -2.3007 -0.0001>, rot=<0 0 0 1>}
 deform_wrist_up_r {parent=deform_wrist_up_base_r, loc=<0 0 0>, rot=<0 0 0 1>}
 hand_attach_r {parent=hand_r, loc=<-8.5001 2.9999 0>, rot=<0 0 -0.7071 -0.7071>}
 weapon_r {parent=hand_r, loc=<0.1974 3.4111 -0.3811>, rot=<0 0 0 1>}
 lowerarm_twist_02_r {parent=lowerarm_r, loc=<-10.6707 0 0>, rot=<0 0 0 -1>}
 lowerarm_twist_01_r {parent=lowerarm_r, loc=<-18.316 0 0>, rot=<0 0 0 -1>}
 _hand_r_Sleeves3 {parent=lowerarm_twist_01_r, loc=<0 0 0>, rot=<0.0285 -0.0335 0.0024 -0.999>}
 _hand_r_Sreeves_01 {parent=_hand_r_Sleeves3, loc=<-4.6965 -2.9938 -0.0238>, rot=<0 0 0 1>}
 _hand_r_Sreeves_02 {parent=_hand_r_Sleeves3, loc=<4.6158 -0.052 4.1705>, rot=<0 0 0 1>}
 _hand_r_Sreeves_02_01 {parent=_hand_r_Sreeves_02, loc=<-0.0533 -0.8912 2.6501>, rot=<0 0 0 1>}
 _hand_r_Sreeves_03 {parent=_hand_r_Sleeves3, loc=<-4.0465 3.9743 0.8671>, rot=<0 0 0 1>}
 _hand_r_Sreeves_04 {parent=_hand_r_Sleeves3, loc=<-4.7835 0.3872 -2.7025>, rot=<0 0 0 1>}
 deform_elbow_in_base_r {parent=lowerarm_r, loc=<-0.1834 -3.6062 0.2572>, rot=<0 0 0 -1>}
 deform_elbow_in_r {parent=deform_elbow_in_base_r, loc=<0 0 0>, rot=<0 0 0 1>}
 deform_elbow_out_base_r {parent=lowerarm_r, loc=<0 4 -0.0001>, rot=<0 0 0 -1>}
 deform_elbow_out_r {parent=deform_elbow_out_base_r, loc=<0 0 0>, rot=<0 0 0 1>}
 upperarm_twist_01_r {parent=upperarm_r, loc=<-7.2135 0.1134 0>, rot=<0 0 0 -1>}
 deform_upperarm_bk_base_r {parent=upperarm_twist_01_r, loc=<6.4053 4.3746 -0.314>, rot=<0 0 0 -1>}
 deform_upperarm_bk_r {parent=deform_upperarm_bk_base_r, loc=<0 0 0>, rot=<0 0 0 1>}
 deform_upperarm_fr_base_r {parent=upperarm_twist_01_r, loc=<5.6312 -5.092 -0.3354>, rot=<0 0 0 -1>}
 deform_upperarm_fr_r {parent=deform_upperarm_fr_base_r, loc=<0 0 0>, rot=<0 0 0 1>}
 deform_upperarm_up_base_r {parent=upperarm_twist_01_r, loc=<8.891 -0.4825 -3.9517>, rot=<0 0 0 1>}
 deform_upperarm_up_r {parent=deform_upperarm_up_base_r, loc=<0 0 0>, rot=<0 0 0 1>}
 upperarm_twist_02_r {parent=upperarm_r, loc=<-19.7998 0.1134 0>, rot=<0 0 0 -1>}
 scale_shoulder_r {parent=upperarm_r, loc=<0 0 0>, rot=<0.9343 -0.0039 0.3563 -0.0065>}
 attach_cape {parent=spine_05, loc=<9.3216 -8.4501 0>, rot=<0.0616 -0.7044 -0.0616 -0.7044>}
 dyn_collar_01 {parent=spine_05, loc=<12.768 -10.0487 0>, rot=<0 0 0.9836 -0.1806>}
 dyn_collar_02 {parent=dyn_collar_01, loc=<-3.2287 1.802 0.0506>, rot=<0 0 0.9621 0.2729>}
 dyn_collar_03 {parent=dyn_collar_01, loc=<0 0 0>, rot=<0 0 -0.2994 -0.9541>}
 dyn_collar_04 {parent=dyn_collar_03, loc=<6.2097 -4.337 0.2152>, rot=<0 0 0.2937 -0.9559>}
 dyn_collar_01_l {parent=spine_05, loc=<9.3469 -0.3988 -14.3011>, rot=<0 -0.4154 0 -0.9096>}
 dyn_collar_02_l {parent=dyn_collar_01_l, loc=<2.1916 0.2862 2.8113>, rot=<-0.0628 0.4074 -0.0449 -0.91>}
 dyn_collar_01_r {parent=spine_05, loc=<9.327 0.32 13.059>, rot=<0 0.3461 0 -0.9382>}
 dyn_collar_02_r {parent=dyn_collar_01_r, loc=<2.6693 0.2537 -1.7341>, rot=<-0.0174 -0.331 -0.0832 -0.9398>}
 collar_r_base {parent=spine_05, loc=<2.0938 7.8453 11.6416>, rot=<0.2794 -0.6135 -0.0672 -0.7355>}
 dyn_collar_r {parent=collar_r_base, loc=<0 0 0>, rot=<0 0 0 1>}
 collar_l_base {parent=spine_05, loc=<2.2102 7.5574 -10.863>, rot=<-0.5391 -0.409 0.7294 0.1008>}
 dyn_collar_l {parent=collar_l_base, loc=<0 0 0>, rot=<0 0 -0.0938 -0.9956>}
 dyn_rope_r_base {parent=spine_05, loc=<0.7891 8.2233 12.8604>, rot=<0.8522 -0.0582 -0.1467 0.4988>}
 dyn_rope_r_01 {parent=dyn_rope_r_base, loc=<0 0 0>, rot=<0 0 0 1>}
 dyn_rope_r_02 {parent=dyn_rope_r_01, loc=<-4.5245 0 0>, rot=<-0.0362 -0.0741 0.0045 -0.9966>}
 dyn_rope_l_base {parent=spine_05, loc=<0.2594 9.8614 -10.5945>, rot=<-0.0844 -0.8472 -0.5006 -0.1566>}
 dyn_rope_l_01 {parent=dyn_rope_l_base, loc=<0 0 0>, rot=<0 0 0 1>}
 dyn_rope_l_02 {parent=dyn_rope_l_01, loc=<4.5246 0 0>, rot=<-0.0362 -0.0741 0.0045 -0.9966>}
 deform_pec_aim_base_l {parent=spine_04, loc=<3.2456 9.3309 -1.2237>, rot=<0.0603 -0.4033 0.3318 -0.8507>}
 deform_pec_aim_l {parent=deform_pec_aim_base_l, loc=<0 0 0>, rot=<0 0 0 1>}
 deform_pec_base_l {parent=deform_pec_aim_l, loc=<9.3665 0 0>, rot=<0 0 0 1>}
 deform_pec_l {parent=deform_pec_base_l, loc=<0 0 0>, rot=<0 0 0 1>}
 deform_pec_aim_base_r {parent=spine_04, loc=<3.2452 9.3309 1.2237>, rot=<-0.4033 -0.0603 0.8507 0.3318>}
 deform_pec_aim_r {parent=deform_pec_aim_base_r, loc=<0 0 0>, rot=<0 0 0 1>}
 deform_pec_base_r {parent=deform_pec_aim_r, loc=<-9.3667 -0.0003 -0.0004>, rot=<0 0 0 1>}
 deform_pec_r {parent=deform_pec_base_r, loc=<0 0 0>, rot=<0 0 0 1>}
 bag_belt_correction {parent=spine_04, loc=<1.2876 13.1421 -0.3066>, rot=<0.7398 0.5885 0.2099 0.2496>}
 shirt_scale {parent=spine_04, loc=<0 0 0>, rot=<0 0 0.0132 -0.9999>}
 dyn_blouse_base_01_c {parent=spine_02, loc=<6.4162 -10.2273 0>, rot=<0.0429 -0.7104 0.7012 -0.0423>}
 dyn_blouse_01_c {parent=dyn_blouse_base_01_c, loc=<0 0 0>, rot=<0 0 0 1>}
 dyn_blouse_02_c {parent=dyn_blouse_01_c, loc=<9.3728 0 0>, rot=<0 0.0314 0 -0.9995>}
 dyn_blouse_03_c {parent=dyn_blouse_02_c, loc=<10.3384 0 0>, rot=<0 0 0 1>}
 dyn_blouse_04_c {parent=dyn_blouse_03_c, loc=<10.7768 0 0>, rot=<0 0 0 1>}
 dyn_blouse_base_side_r {parent=spine_02, loc=<5.9505 -0.2256 13.47>, rot=<0 0 0.9997 -0.0245>}
 dyn_blouse_side_01_r {parent=dyn_blouse_base_side_r, loc=<0 0 0>, rot=<0 0.0436 0 -0.999>}
 dyn_blouse_side_02_r {parent=dyn_blouse_side_01_r, loc=<9.3728 0 0>, rot=<0 0.0489 0 -0.9988>}
 dyn_blouse_side_03_r {parent=dyn_blouse_side_02_r, loc=<11.8894 0 0>, rot=<0 0 0 1>}
 dyn_blouse_side_04_r {parent=dyn_blouse_side_03_r, loc=<11.0616 0 0>, rot=<0 0 0 1>}
 dyn_blouse_base_fr_r {parent=spine_02, loc=<6.4815 8.7726 10>, rot=<0.0639 0.4215 0.9041 -0.0279>}
 dyn_blouse_fr_01_r {parent=dyn_blouse_base_fr_r, loc=<0 0 0>, rot=<0 0 0 1>}
 dyn_blouse_fr_02_r {parent=dyn_blouse_fr_01_r, loc=<9.9866 0 0>, rot=<0 0 0 1>}
 dyn_blouse_fr_03_r {parent=dyn_blouse_fr_02_r, loc=<13.3437 0 0>, rot=<0 0 0 1>}
 dyn_blouse_fr_04_r {parent=dyn_blouse_fr_03_r, loc=<10.2287 0 0>, rot=<0 0 0 1>}
 dyn_blouse_base_fr_l {parent=spine_02, loc=<6.4815 8.7726 -10>, rot=<0.0316 -0.9036 -0.4212 -0.0718>}
 dyn_blouse_fr_01_l {parent=dyn_blouse_base_fr_l, loc=<0.0708 0 0>, rot=<0 0 0 1>}
 dyn_blouse_fr_02_l {parent=dyn_blouse_fr_01_l, loc=<10.1533 0 0>, rot=<0 -0.0204 0 -0.9998>}
 dyn_blouse_fr_03_l {parent=dyn_blouse_fr_02_l, loc=<11.3437 0 0>, rot=<0 0 0 1>}
 dyn_blouse_fr_04_l {parent=dyn_blouse_fr_03_l, loc=<11.8417 0 0>, rot=<0 0 0 1>}
 katball_base {parent=dyn_blouse_fr_01_l, loc=<6.7055 -0.0811 4.8791>, rot=<0 0 0 1>}
 dyn_katball_01 {parent=katball_base, loc=<0 0 0>, rot=<0 0 0 1>}
 dyn_katball_02 {parent=dyn_katball_01, loc=<4.0063 0.3789 -0.6922>, rot=<-0.0118 -0.0792 -0.0337 -0.9962>}
 dyn_katball_03 {parent=dyn_katball_02, loc=<4.4254 0 0>, rot=<0.4168 0.0154 0.0451 -0.9077>}
 dyn_blouse_base_bk_r {parent=spine_02, loc=<6.423 -8.2273 10>, rot=<0.0414 -0.4211 0.9057 -0.0269>}
 dyn_blouse_bk_01_r {parent=dyn_blouse_base_bk_r, loc=<0 0 0>, rot=<0 0 0 1>}
 dyn_blouse_bk_02_r {parent=dyn_blouse_bk_01_r, loc=<9.3728 0 0>, rot=<0 0.0336 0 -0.9994>}
 dyn_blouse_bk_03_r {parent=dyn_blouse_bk_02_r, loc=<10.9224 0 0>, rot=<0 0 0 1>}
 dyn_blouse_bk_04_r {parent=dyn_blouse_bk_03_r, loc=<10.7949 0 0>, rot=<0 0 0 1>}
 dyn_blouse_base_bk_l {parent=spine_02, loc=<6.423 -8.2273 -10>, rot=<0.0232 -0.8655 0.4984 -0.0436>}
 dyn_blouse_bk_01_l {parent=dyn_blouse_base_bk_l, loc=<0 0 0>, rot=<0 0 0 1>}
 dyn_blouse_bk_02_l {parent=dyn_blouse_bk_01_l, loc=<9.3728 0 0>, rot=<0 0.0305 0 -0.9995>}
 dyn_blouse_bk_03_l {parent=dyn_blouse_bk_02_l, loc=<10.3375 0 0>, rot=<0 0 0 1>}
 dyn_blouse_bk_04_l {parent=dyn_blouse_bk_03_l, loc=<11.1619 0 0>, rot=<0 0 0 1>}
 dyn_blouse_base_side_l {parent=spine_02, loc=<5.9507 -0.2256 -13.47>, rot=<0.0245 -0.9997 0 0>}
 dyn_blouse_side_01_l {parent=dyn_blouse_base_side_l, loc=<0 0 0>, rot=<0 0.0436 0 -0.999>}
 dyn_blouse_side_02_l {parent=dyn_blouse_side_01_l, loc=<9.3728 0 0>, rot=<0 0.044 0 -0.999>}
 dyn_blouse_side_03_l {parent=dyn_blouse_side_02_l, loc=<11.889 0 0>, rot=<0 0 0 1>}
 dyn_blouse_side_04_l {parent=dyn_blouse_side_03_l, loc=<10.6632 0 0>, rot=<0 0 0 1>}
 thigh_l {parent=pelvis, loc=<-1.0855 -0.2109 -8.7053>, rot=<0.0009 0.0355 0.0155 -0.9993>}
 calf_l {parent=thigh_l, loc=<-42.5284 -0.0019 -0.7297>, rot=<0.0006 -0.0161 -0.0391 -0.9991>}
 foot_l {parent=calf_l, loc=<-46.7401 -2.5014 -0.592>, rot=<0.0427 -0.0238 0.0353 -0.9982>}
 ball_l {parent=foot_l, loc=<-3.5658 10.8066 -0.1544>, rot=<0.0001 -0.0001 -0.7063 -0.7079>}
 dyn_laces_base_l {parent=foot_l, loc=<1.7229 4.1506 0.4182>, rot=<-0.1493 -0.6784 0.0986 -0.7126>}
 dyn_laces_01_l {parent=dyn_laces_base_l, loc=<2.3586 -0.5778 -0.8566>, rot=<0.8296 -0.1084 -0.5171 0.1807>}
 dyn_laces_02_l {parent=dyn_laces_base_l, loc=<-3.0086 0.0102 -0.7665>, rot=<-0.4275 0.2529 0.8678 -0.0133>}
 dyn_laces_02_l1 {parent=dyn_laces_base_l, loc=<-3.0086 0.0102 -0.7665>, rot=<-0.4275 0.2529 0.8678 -0.0133>}
 calf_twist_02_l {parent=calf_l, loc=<-19.1546 -1.7943 -0.2638>, rot=<0.0237 0.0095 -0.0035 -0.9997>}
 strap_l_base {parent=calf_twist_02_l, loc=<0.2767 -5.9526 5.3452>, rot=<-0.0382 -0.91 -0.3942 0.1227>}
 dyn_strap_l_01 {parent=strap_l_base, loc=<0 0 0>, rot=<0 0 0 1>}
 dyn_strap_l_02 {parent=dyn_strap_l_01, loc=<6.9182 0 0>, rot=<-0.0222 0.0853 0.0246 -0.9958>}
 dyn_strap_l_03 {parent=dyn_strap_l_02, loc=<9.3919 0 0>, rot=<-0.0001 0.0294 -0.0221 -0.9993>}
 dyn_strap_l_04 {parent=dyn_strap_l_03, loc=<8.1858 0 0>, rot=<0 -0.0003 0.0153 -0.9999>}
 calf_twist_01_l {parent=calf_l, loc=<-35.8483 -1.9197 -0.5794>, rot=<0.0237 0.0095 -0.0035 -0.9997>}
 deform_knee_in_base_l {parent=calf_l, loc=<0.8031 -5.0231 0.2456>, rot=<0 0 0 1>}
 deform_knee_in_l {parent=deform_knee_in_base_l, loc=<0 0 0>, rot=<0 0 0 1>}
 deform_knee_out_base_l {parent=calf_l, loc=<0 4 0>, rot=<0 0 0 -1>}
 deform_knee_out_l {parent=deform_knee_out_base_l, loc=<0 0 0>, rot=<0 0 0 1>}
 thigh_twist_01_l {parent=thigh_l, loc=<-27.952 0 0>, rot=<0.0013 0 0 -1>}
 deform_groin_base_l {parent=thigh_l, loc=<0.4801 5.6776 -0.2486>, rot=<0 0 0 -1>}
 deform_groin_l {parent=deform_groin_base_l, loc=<0 0 0>, rot=<0 0 0 1>}
 thigh_r {parent=pelvis, loc=<-1.0855 -0.2109 8.7053>, rot=<0.0355 -0.0009 0.9993 0.0155>}
 calf_r {parent=thigh_r, loc=<42.5284 0.0019 0.7297>, rot=<0.0006 -0.0161 -0.0391 -0.9991>}
 foot_r {parent=calf_r, loc=<46.7401 2.5014 0.592>, rot=<0.0427 -0.0238 0.0353 -0.9982>}
 ball_r {parent=foot_r, loc=<3.5658 -10.8066 0.1544>, rot=<0.0001 -0.0001 -0.7063 -0.7079>}
 dyn_laces_base_r {parent=foot_r, loc=<-1.7396 -4.3665 -0.4206>, rot=<0.7101 -0.1606 -0.6482 -0.2231>}
 dyn_laces_01_r {parent=dyn_laces_base_r, loc=<4.4327 -1.3748 -2.5114>, rot=<0 -0.0407 0 -0.9992>}
 dyn_laces_02_r {parent=dyn_laces_base_r, loc=<-3.3967 -2.198 -3.1254>, rot=<0 -0.9972 0 -0.0742>}
 calf_twist_02_r {parent=calf_r, loc=<19.1546 1.7943 0.2638>, rot=<0.0237 0.0095 -0.0035 -0.9997>}
 strap_r_base {parent=calf_twist_02_r, loc=<-1.3249 5.7948 -4.7499>, rot=<0.9013 -0.0486 -0.1512 -0.403>}
 dyn_strap_r_01 {parent=strap_r_base, loc=<0 0 0>, rot=<0 0 0 1>}
 dyn_strap_r_02 {parent=dyn_strap_r_01, loc=<7.606 0 0>, rot=<0.0177 -0.0952 0.0466 -0.9942>}
 dyn_strap_r_03 {parent=dyn_strap_r_02, loc=<9.4863 0 0>, rot=<0.001 -0.0517 -0.0237 -0.9984>}
 dyn_strap_r_04 {parent=dyn_strap_r_03, loc=<7.6366 0 0>, rot=<0.0001 0.0156 0.0075 -0.9999>}
 calf_twist_01_r {parent=calf_r, loc=<35.8483 1.9197 0.5794>, rot=<0.0237 0.0095 -0.0035 -0.9997>}
 deform_knee_in_base_r {parent=calf_r, loc=<-0.8031 5.0231 -0.2457>, rot=<0 0 0 -1>}
 deform_knee_in_r {parent=deform_knee_in_base_r, loc=<0 0 0>, rot=<0 0 0 1>}
 deform_knee_out_base_r {parent=calf_r, loc=<0 -4 0>, rot=<0 0 0 -1>}
 deform_knee_out_r {parent=deform_knee_out_base_r, loc=<0 0 0>, rot=<0 0 0 1>}
 thigh_twist_01_r {parent=thigh_r, loc=<27.952 0 0>, rot=<0.0013 0 0 -1>}
 deform_groin_base_r {parent=thigh_r, loc=<-0.4801 -5.6776 0.2486>, rot=<0 0 0 -1>}
 deform_groin_r {parent=deform_groin_base_r, loc=<0 0 0>, rot=<0 0 0 1>}
 deform_glute_base_l {parent=pelvis, loc=<-7.3466 -9.0213 -9.1245>, rot=<0.0172 0.031 -0.4477 -0.8935>}
 deform_glute_l {parent=deform_glute_base_l, loc=<0 0 0>, rot=<0 0 0 1>}
 deform_glute_base_r {parent=pelvis, loc=<-7.3464 -9.0213 9.1245>, rot=<0.031 -0.0172 0.8935 -0.4477>}
 deform_glute_r {parent=deform_glute_base_r, loc=<0 0 0>, rot=<0 0 0 1>}
 Alt_jacket_scale {parent=pelvis, loc=<0 0 0>, rot=<0 0 0 1>}
 AltSkin_Corrective {parent=pelvis, loc=<-11.8558 -13.8797 0>, rot=<0 0 0 1>}
 Claf_R {type=Null, loc=<0 0 0>, rot=<0 0 0 1>}
 Claf_L {type=Null, loc=<0 0 0>, rot=<0 0 0 1>}
]
variables [
 Debug {
  type=FSphericalPoseReaderDebugSettings
  default=(bDrawDebug=True,bDraw2D=True,bDrawLocalAxes=False,DebugScale=25,DebugSegments=20,DebugThickness=0.25)
 }
 Debug_bDrawDebug {type=bool, default=True}
 Is DBNO On {type=bool, default=False}
 Is baseball sliding {type=bool, default=False}
]
