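# Physics asset for the left leg strap on the SugarBubble medium female player body, defining capsule bodies and joints for a dangling strap chain.
# A small base capsule joins the left lower calf twist bone and anchors four chained dyn_strap capsules with tight swing and tighter twist limits.
# Larger capsules on both calves, calf twists and feet form the surrounding leg shapes, linked by looser constraints within each leg.
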PhysicsAsset /Game/Characters/Player/Female/Medium/Bodies/F_MED_SugarBubble/Meshes/F_MED_SugarBubble_LegStrap_L_Physics
bodies {
 calf_l [{shapeType=Capsule, length=27.1655, radius=5.2967, center=<-21.0099 -2.0195 0.0963>, rotation=<rot 0 -90.491 -88.9476>}]
 calf_r [{shapeType=Capsule, length=25.5801, radius=5.9518, center=<21.2816 2.7695 -0.5023>, rotation=<rot 0 91.6591 -89.0093>}]
 calf_twist_02_l [{shapeType=Capsule, length=19.0994, radius=7.2902, center=<-8.9678 -0.7793 -0.5166>, rotation=<rot 0 -83.4492 -89.7124>}]
 calf_twist_02_r [{shapeType=Capsule, length=19.9116, radius=6.8784, center=<11.5806 0.2245 0.2064>, rotation=<rot 0 -93.9469 -87.7164>}]
 dyn_strap_l_01 [{shapeType=Capsule, length=2.337, radius=1.666, shapeName=dyn_strap_l_01_capsule, center=<5.6591 0.0448 -0.1001>, rotation=<rot 90 0 0>}]
 dyn_strap_l_02 [{shapeType=Capsule, length=2.3629, radius=1.5353, shapeName=dyn_strap_l_02_capsule, center=<5.445 0.1729 -0.1185>, rotation=<rot 90 0 0>}]
 dyn_strap_l_03 [{shapeType=Capsule, length=3.694, radius=1.3798, shapeName=dyn_strap_l_03_capsule, center=<4.7183 0.0803 -0.008>, rotation=<rot 90 0 0>}]
 dyn_strap_l_04 [{shapeType=Capsule, length=1.9678, radius=1.6572, shapeName=dyn_strap_l_04_capsule, center=<3.4704 -0.0349 -0.0459>, rotation=<rot 90 0 0>}]
 foot_l [{shapeType=Capsule, length=13.0876, radius=7.7933, center=<-3.3736 7.6193 0.3001>, rotation=<rot 0 12.536 -88.5093>}]
 foot_r [{shapeType=Capsule, length=14.1236, radius=7.9936, center=<1.9019 -6.6545 -0.0017>, rotation=<rot 0 7.1228 -89.9437>}]
 strap_l_base [{shapeType=Capsule, length=0.505, radius=0.505, shapeName=strap_l_base_capsule, rotation=<rot 90 0 0>}]
}
constraints [
 {bone1Name=foot_l, bone2Name=calf_l, swing1LimitDegrees=45, swing2LimitDegrees=45, twistLimitDegrees=45}
 {bone1Name=calf_twist_02_l, bone2Name=calf_l, swing1LimitDegrees=45, swing2LimitDegrees=45, twistLimitDegrees=45}
 {bone1Name=foot_r, bone2Name=calf_r, swing1LimitDegrees=45, swing2LimitDegrees=45, twistLimitDegrees=45}
 {bone1Name=calf_twist_02_r, bone2Name=calf_r, swing1LimitDegrees=45, swing2LimitDegrees=45, twistLimitDegrees=45}
 {bone1Name=strap_l_base, bone2Name=calf_twist_02_l, swing1LimitDegrees=45, swing2LimitDegrees=45, twistLimitDegrees=45}
 {bone1Name=dyn_strap_l_01, bone2Name=strap_l_base, swing1LimitDegrees=15, swing2LimitDegrees=15, twistLimitDegrees=5}
 {bone1Name=dyn_strap_l_02, bone2Name=dyn_strap_l_01, swing1LimitDegrees=15, swing2LimitDegrees=15, twistLimitDegrees=5}
 {bone1Name=dyn_strap_l_03, bone2Name=dyn_strap_l_02, swing1LimitDegrees=15, swing2LimitDegrees=15, twistLimitDegrees=5}
 {bone1Name=dyn_strap_l_04, bone2Name=dyn_strap_l_03, swing1LimitDegrees=15, swing2LimitDegrees=15, twistLimitDegrees=5}
]
