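# Physics asset for the SugarBubble medium female body's leg strap mesh, covering dangling straps on both legs despite the R in its name.
# Each side chains four capsule strap bodies from a small base body constrained to the lower calf twist bone, with uniform tight swing and twist limits.
# Unconstrained foot capsules join the calf capsules as leg volumes.
# Left strap shapes keep right-side names, suggesting they were mirrored.
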PhysicsAsset /Game/Characters/Player/Female/Medium/Bodies/F_MED_SugarBubble/Meshes/F_MED_SugarBubble_LegStrap_R_Physics
bodies {
 calf_twist_02_l [{shapeType=Capsule, length=20.0994, radius=6.2902, center=<-8.9678 -0.7793 -0.5166>, rotation=<rot 0 -83.4492 -89.7124>}]
 calf_twist_02_r [{shapeType=Capsule, length=19.9116, radius=6.8784, center=<11.5806 0.2245 0.2064>, rotation=<rot 0 -93.9469 -87.7164>}]
 dyn_strap_l_01 [{shapeType=Capsule, length=4.4535, radius=1.5644, shapeName=dyn_strap_r_01_capsule, center=<4.9904 -0.0171 -0.1334>, rotation=<rot 90 180 0>}]
 dyn_strap_l_02 [{shapeType=Capsule, length=2.3633, radius=1.5396, shapeName=dyn_strap_r_02_capsule, center=<5.1815 -0.0613 0.0092>, rotation=<rot 90 180 0>}]
 dyn_strap_l_03 [{shapeType=Capsule, length=2.2349, radius=1.4735, shapeName=dyn_strap_r_03_capsule, center=<2.6569 0.0281 0.107>, rotation=<rot 90 180 0>}]
 dyn_strap_l_04 [{shapeType=Capsule, length=2.7734, radius=1.383, shapeName=dyn_strap_r_04_capsule, center=<2.3167 0.1825 0.04>, rotation=<rot 90 180 0>}]
 dyn_strap_r_01 [{shapeType=Capsule, length=4.4535, radius=1.5644, shapeName=dyn_strap_r_01_capsule, center=<4.4691 0.0171 0.1334>, rotation=<rot 90 0 0>}]
 dyn_strap_r_02 [{shapeType=Capsule, length=2.3633, radius=1.5396, shapeName=dyn_strap_r_02_capsule, center=<5.9529 0.0613 -0.0092>, rotation=<rot 90 0 0>}]
 dyn_strap_r_03 [{shapeType=Capsule, length=2.2349, radius=1.4735, shapeName=dyn_strap_r_03_capsule, center=<3.5753 -0.0281 -0.107>, rotation=<rot 90 0 0>}]
 dyn_strap_r_04 [{shapeType=Capsule, length=2.7734, radius=1.383, shapeName=dyn_strap_r_04_capsule, center=<3.3805 -0.1825 -0.04>, rotation=<rot 90 0 0>}]
 foot_l [{shapeType=Capsule, length=13.0876, radius=7.7933, center=<-3.3736 7.6193 0.3001>, rotation=<rot 0 12.536 -88.5093>}]
 foot_r [{shapeType=Capsule, length=14.1236, radius=7.9936, center=<1.9019 -6.6545 -0.0017>, rotation=<rot 0 7.1228 -89.9437>}]
 strap_l_base [{shapeType=Capsule, length=0.505, radius=0.505, shapeName=strap_r_base_capsule, rotation=<rot 90 180 0>}]
 strap_r_base [{shapeType=Capsule, length=0.505, radius=0.505, shapeName=strap_r_base_capsule, rotation=<rot 90 0 0>}]
}
constraints [
 {bone1Name=dyn_strap_r_04, bone2Name=dyn_strap_r_03, swing1LimitDegrees=15, swing2LimitDegrees=15, twistLimitDegrees=5}
 {bone1Name=strap_r_base, bone2Name=calf_twist_02_r, swing1LimitDegrees=15, swing2LimitDegrees=15, twistLimitDegrees=5}
 {bone1Name=dyn_strap_r_01, bone2Name=strap_r_base, swing1LimitDegrees=15, swing2LimitDegrees=15, twistLimitDegrees=5}
 {bone1Name=dyn_strap_r_02, bone2Name=dyn_strap_r_01, swing1LimitDegrees=15, swing2LimitDegrees=15, twistLimitDegrees=5}
 {bone1Name=dyn_strap_r_03, bone2Name=dyn_strap_r_02, swing1LimitDegrees=15, swing2LimitDegrees=15, twistLimitDegrees=5}
 {bone1Name=strap_l_base, bone2Name=calf_twist_02_l, swing1LimitDegrees=15, swing2LimitDegrees=15, twistLimitDegrees=5}
 {bone1Name=dyn_strap_l_01, bone2Name=strap_l_base, swing1LimitDegrees=15, swing2LimitDegrees=15, twistLimitDegrees=5}
 {bone1Name=dyn_strap_l_02, bone2Name=dyn_strap_l_01, swing1LimitDegrees=15, swing2LimitDegrees=15, twistLimitDegrees=5}
 {bone1Name=dyn_strap_l_03, bone2Name=dyn_strap_l_02, swing1LimitDegrees=15, swing2LimitDegrees=15, twistLimitDegrees=5}
 {bone1Name=dyn_strap_l_04, bone2Name=dyn_strap_l_03, swing1LimitDegrees=15, swing2LimitDegrees=15, twistLimitDegrees=5}
]
