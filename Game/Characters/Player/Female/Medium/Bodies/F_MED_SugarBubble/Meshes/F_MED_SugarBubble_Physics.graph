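# Physics asset for the SugarBubble medium female player body, rigging its long blouse for simulated secondary motion.
# Capsules on pelvis, spine, both thighs and right calf sit alongside seven four-link blouse chains (center, front, back, sides), each hung from a base bone jointed to spine_02.
# Chain links use tight swing and twist limits.
# Cross-links between neighboring chains, several with locked swing, tie the panels together around the hem.
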PhysicsAsset /Game/Characters/Player/Female/Medium/Bodies/F_MED_SugarBubble/Meshes/F_MED_SugarBubble_Physics
bodies {
 calf_twist_02_r [{shapeType=Capsule, length=22.3741, radius=8.5979, shapeName=calf_twist_02_r_capsule, center=<4.8853 1.0602 0.1403>, rotation=<rot 90 0 0>}]
 dyn_blouse_01_c [{shapeType=Capsule, length=10, radius=5, shapeName=dyn_blouse_01_c_capsule, center=<0 0 3>, rotation=<rot 0 0 90>}]
 dyn_blouse_02_c [{shapeType=Capsule, length=10, radius=5, shapeName=dyn_blouse_02_c_capsule, center=<0.9174 0 3>, rotation=<rot 0 0 90>}]
 dyn_blouse_03_c [{shapeType=Capsule, length=7.3818, radius=7.6182, shapeName=dyn_blouse_03_c_capsule, center=<0 0 5.233>, rotation=<rot 0 0 90>}]
 dyn_blouse_04_c [{shapeType=Capsule, length=10.8434, radius=9.8319, shapeName=dyn_blouse_04_c_capsule, center=<2.1305 0 5.8442>, rotation=<rot 0 0 90>}]
 dyn_blouse_base_01_c [{shapeType=Capsule, length=10, radius=5, shapeName=dyn_blouse_base_01_c_capsule, rotation=<rot 0 0 90>}]
 dyn_blouse_base_bk_l [{shapeType=Capsule, length=10, radius=5, shapeName=dyn_blouse_base_bk_l_capsule, rotation=<rot 0 0 90>}]
 dyn_blouse_base_bk_r [{shapeType=Capsule, length=10, radius=5, shapeName=dyn_blouse_base_bk_r_capsule, rotation=<rot 0 0 90>}]
 dyn_blouse_base_fr_l [{shapeType=Capsule, length=10, radius=5, shapeName=dyn_blouse_base_fr_l_capsule, rotation=<rot 0 0 90>}]
 dyn_blouse_base_fr_r [{shapeType=Capsule, length=10, radius=5, shapeName=dyn_blouse_base_fr_r_capsule, rotation=<rot 0 0 90>}]
 dyn_blouse_base_side_l [{shapeType=Capsule, length=10, radius=5, shapeName=dyn_blouse_base_side_l_capsule, rotation=<rot 0 0 90>}]
 dyn_blouse_base_side_r [{shapeType=Capsule, length=10, radius=5, shapeName=dyn_blouse_base_side_r_capsule, rotation=<rot 0 0 90>}]
 dyn_blouse_bk_01_l [{shapeType=Capsule, length=10, radius=5, shapeName=dyn_blouse_bk_01_l_capsule, center=<0 0 3>, rotation=<rot 0 0 90>}]
 dyn_blouse_bk_01_r [{shapeType=Capsule, length=10, radius=5, shapeName=dyn_blouse_bk_01_r_capsule, center=<0 0 4.9448>, rotation=<rot 0 0 90>}]
 dyn_blouse_bk_02_l [{shapeType=Capsule, length=8, radius=7, shapeName=dyn_blouse_bk_02_l_capsule, center=<0.9174 0 8.5761>, rotation=<rot 0 0 90>}]
 dyn_blouse_bk_02_r [{shapeType=Capsule, length=10, radius=5, shapeName=dyn_blouse_bk_02_r_capsule, center=<0.9174 0 4.9448>, rotation=<rot 0 0 90>}]
 dyn_blouse_bk_03_l [{shapeType=Capsule, length=7.8818, radius=7.1182, shapeName=dyn_blouse_bk_03_l_capsule, center=<0 0 5.8223>, rotation=<rot 0 0 90>}]
 dyn_blouse_bk_03_r [{shapeType=Capsule, length=7.3818, radius=7.6182, shapeName=dyn_blouse_bk_03_r_capsule, center=<0 0 5.8237>, rotation=<rot 0 0 90>}]
 dyn_blouse_bk_04_l [{shapeType=Capsule, length=13.6471, radius=8.9187, shapeName=dyn_blouse_bk_04_l_capsule, center=<2.2538 0.1859 5.1252>, rotation=<rot 0 0 95>}]
 dyn_blouse_bk_04_r [{shapeType=Capsule, length=12.5517, radius=10.1013, shapeName=dyn_blouse_bk_04_r_capsule, center=<2.322 0.3609 7.1246>, rotation=<rot 0 0 95>}]
 dyn_blouse_fr_01_l [{shapeType=Capsule, length=10, radius=5, shapeName=dyn_blouse_fr_01_l_capsule, center=<2.1161 0 3>, rotation=<rot 0 0 90>}]
 dyn_blouse_fr_01_r [{shapeType=Capsule, length=10, radius=5, shapeName=dyn_blouse_fr_01_r_capsule, center=<0 0 3>, rotation=<rot 0 0 90>}]
 dyn_blouse_fr_02_l [{shapeType=Capsule, length=7.5, radius=5, shapeName=dyn_blouse_fr_02_l_capsule, center=<0 0 3.5093>, rotation=<rot 0 0 90>}]
 dyn_blouse_fr_02_r [{shapeType=Capsule, length=9.5, radius=5.5, shapeName=dyn_blouse_fr_02_r_capsule, center=<2.0048 0 4.7932>, rotation=<rot 0 0 90>}]
 dyn_blouse_fr_03_l [{
   shapeType=Capsule
   length=10.3818
   radius=7.1182
   shapeName=dyn_blouse_fr_03_l_capsule
   center=<0.9613 0.0786 5.1115>
   rotation=<rot 0 0 92.5484>
  }]
 dyn_blouse_fr_03_r [{shapeType=Capsule, length=4.8818, radius=7.1182, shapeName=dyn_blouse_fr_03_r_capsule, center=<0 0 5.7016>, rotation=<rot 0 0 90>}]
 dyn_blouse_fr_04_l [{shapeType=Capsule, length=11.0469, radius=7.056, shapeName=dyn_blouse_fr_04_r_capsule, center=<3.0159 -0.1167 3.8885>, rotation=<rot 0 0 90>}]
 dyn_blouse_fr_04_r [{shapeType=Capsule, length=4.0469, radius=7.056, shapeName=dyn_blouse_fr_04_r_capsule, center=<3.0159 -0.1167 3.6011>, rotation=<rot 0 0 90>}]
 dyn_blouse_side_01_l [{shapeType=Capsule, length=10, radius=5, shapeName=dyn_blouse_side_01_l_capsule, center=<0 0 3>, rotation=<rot 0 0 90>}]
 dyn_blouse_side_01_r [{shapeType=Capsule, length=10, radius=5, shapeName=dyn_blouse_side_01_r_capsule, center=<0 0 3>, rotation=<rot 0 0 90>}]
 dyn_blouse_side_02_l [{shapeType=Capsule, length=10, radius=5, shapeName=dyn_blouse_side_02_l_capsule, center=<0 0 3>, rotation=<rot 0 0 90>}]
 dyn_blouse_side_02_r [{shapeType=Capsule, length=10, radius=5, shapeName=dyn_blouse_side_02_r_capsule, center=<0 0 3>, rotation=<rot 0 0 90>}]
 dyn_blouse_side_03_l [{shapeType=Capsule, length=7.8818, radius=7.1182, shapeName=dyn_blouse_side_03_l_capsule, center=<0 0 5.8428>, rotation=<rot 0 0 90>}]
 dyn_blouse_side_03_r [{shapeType=Capsule, length=7.8818, radius=7.1182, shapeName=dyn_blouse_side_03_r_capsule, center=<0 0 6.3724>, rotation=<rot 0 0 90>}]
 dyn_blouse_side_04_l [{
   shapeType=Capsule
   length=15.0469
   radius=8.7263
   shapeName=dyn_blouse_side_04_l_capsule
   center=<2.284 -0.1606 5.8358>
   rotation=<rot 0 0 85>
  }]
 dyn_blouse_side_04_r [{
   shapeType=Capsule
   length=9.7501
   radius=9.2947
   shapeName=dyn_blouse_side_04_r_capsule
   center=<1.9374 0.2068 6.3634>
   rotation=<rot 0 0 95>
  }]
 pelvis [{shapeType=Capsule, length=8.1652, radius=11.1647, shapeName=pelvis_capsule, center=<-1.4429 -0.4526 0>}]
 spine_01 [{shapeType=Capsule, length=8.7957, radius=8.995, shapeName=spine_01_capsule, center=<0.7103 -0.023 0>}]
 spine_02 [{shapeType=Capsule, length=9.9996, radius=7.9634, shapeName=spine_02_capsule, center=<3.2663 0 0>}]
 thigh_l [
  {
   shapeType=Capsule
   length=29.2565
   radius=9.1892
   shapeName=thigh_l_capsule
   center=<-19.929 -1.0447 0.2963>
   rotation=<rot 0.1824 87.2778 -88.9656>
  }
  {shapeType=Capsule, length=20, radius=10, shapeName=thigh_l_capsule, center=<-15.6156 0.3396 3.7373>, rotation=<rot 87.9532 0 0>}
 ]
 thigh_r [{
   shapeType=Capsule
   length=31.6725
   radius=10.4413
   shapeName=thigh_r_capsule
   center=<18.6612 0.1549 -1.9675>
   rotation=<rot 0.2971 -92.6036 -93.8913>
  }]
}
constraints [
 {
  bone1Name=spine_01
  bone2Name=pelvis
  swing1LimitDegrees=45
  swing1Motion=Free
  swing2LimitDegrees=45
  swing2Motion=Free
  twistLimitDegrees=45
  twistMotion=Free
 }
 {
  bone1Name=spine_02
  bone2Name=spine_01
  swing1LimitDegrees=45
  swing1Motion=Free
  swing2LimitDegrees=45
  swing2Motion=Free
  twistLimitDegrees=45
  twistMotion=Free
 }
 {
  bone1Name=thigh_l
  bone2Name=pelvis
  swing1LimitDegrees=45
  swing1Motion=Free
  swing2LimitDegrees=45
  swing2Motion=Free
  twistLimitDegrees=45
  twistMotion=Free
 }
 {
  bone1Name=thigh_r
  bone2Name=pelvis
  swing1LimitDegrees=45
  swing1Motion=Free
  swing2LimitDegrees=45
  swing2Motion=Free
  twistLimitDegrees=45
  twistMotion=Free
 }
 {
  bone1Name=dyn_blouse_base_bk_l
  bone2Name=spine_02
  swing1LimitDegrees=45
  swing1Motion=Free
  swing2LimitDegrees=45
  swing2Motion=Free
  twistLimitDegrees=45
  twistMotion=Free
 }
 {bone1Name=dyn_blouse_bk_01_l, bone2Name=dyn_blouse_base_bk_l, swing1LimitDegrees=13, swing2LimitDegrees=15, twistLimitDegrees=3}
 {bone1Name=dyn_blouse_bk_02_l, bone2Name=dyn_blouse_bk_01_l, swing1LimitDegrees=10, swing2LimitDegrees=15, twistLimitDegrees=1}
 {bone1Name=dyn_blouse_bk_03_l, bone2Name=dyn_blouse_bk_02_l, swing1LimitDegrees=5, swing2LimitDegrees=25, twistLimitDegrees=0}
 {bone1Name=dyn_blouse_bk_04_l, bone2Name=dyn_blouse_bk_03_l, swing1LimitDegrees=5, swing2LimitDegrees=35, twistLimitDegrees=5}
 {
  bone1Name=dyn_blouse_base_bk_r
  bone2Name=spine_02
  swing1LimitDegrees=45
  swing1Motion=Free
  swing2LimitDegrees=45
  swing2Motion=Free
  twistLimitDegrees=45
  twistMotion=Free
 }
 {bone1Name=dyn_blouse_bk_01_r, bone2Name=dyn_blouse_base_bk_r, swing1LimitDegrees=13, swing2LimitDegrees=15, twistLimitDegrees=0}
 {bone1Name=dyn_blouse_bk_02_r, bone2Name=dyn_blouse_bk_01_r, swing1LimitDegrees=10, swing2LimitDegrees=15, twistLimitDegrees=0}
 {bone1Name=dyn_blouse_bk_03_r, bone2Name=dyn_blouse_bk_02_r, swing1LimitDegrees=5, swing2LimitDegrees=25, twistLimitDegrees=2}
 {bone1Name=dyn_blouse_bk_04_r, bone2Name=dyn_blouse_bk_03_r, swing1LimitDegrees=10, swing2LimitDegrees=35, twistLimitDegrees=5}
 {
  bone1Name=dyn_blouse_base_fr_l
  bone2Name=spine_02
  swing1LimitDegrees=45
  swing1Motion=Free
  swing2LimitDegrees=45
  swing2Motion=Free
  twistLimitDegrees=45
  twistMotion=Free
 }
 {bone1Name=dyn_blouse_fr_01_l, bone2Name=dyn_blouse_base_fr_l, swing1LimitDegrees=30, swing2LimitDegrees=10, twistLimitDegrees=0}
 {bone1Name=dyn_blouse_fr_02_l, bone2Name=dyn_blouse_fr_01_l, swing1LimitDegrees=30, swing2LimitDegrees=40, twistLimitDegrees=10}
 {bone1Name=dyn_blouse_fr_03_l, bone2Name=dyn_blouse_fr_02_l, swing1LimitDegrees=20, swing2LimitDegrees=15, twistLimitDegrees=10}
 {bone1Name=dyn_blouse_fr_04_l, bone2Name=dyn_blouse_fr_03_l, swing1LimitDegrees=10, swing2LimitDegrees=10, twistLimitDegrees=10}
 {
  bone1Name=dyn_blouse_base_fr_r
  bone2Name=spine_02
  swing1LimitDegrees=45
  swing1Motion=Free
  swing2LimitDegrees=45
  swing2Motion=Free
  twistLimitDegrees=45
  twistMotion=Free
 }
 {bone1Name=dyn_blouse_fr_01_r, bone2Name=dyn_blouse_base_fr_r, swing1LimitDegrees=30, swing2LimitDegrees=10, twistLimitDegrees=0}
 {bone1Name=dyn_blouse_fr_03_r, bone2Name=dyn_blouse_fr_02_r, swing1LimitDegrees=30, swing2LimitDegrees=10, twistLimitDegrees=8}
 {bone1Name=dyn_blouse_fr_04_r, bone2Name=dyn_blouse_fr_03_r, swing1LimitDegrees=10, swing2LimitDegrees=10, twistLimitDegrees=8}
 {
  bone1Name=dyn_blouse_base_side_l
  bone2Name=spine_02
  swing1LimitDegrees=45
  swing1Motion=Free
  swing2LimitDegrees=45
  swing2Motion=Free
  twistLimitDegrees=45
  twistMotion=Free
 }
 {bone1Name=dyn_blouse_side_01_l, bone2Name=dyn_blouse_base_side_l, swing1LimitDegrees=15, swing2LimitDegrees=15, twistLimitDegrees=0}
 {bone1Name=dyn_blouse_side_02_l, bone2Name=dyn_blouse_side_01_l, swing1LimitDegrees=15, swing2LimitDegrees=15, twistLimitDegrees=0}
 {bone1Name=dyn_blouse_side_03_l, bone2Name=dyn_blouse_side_02_l, swing1LimitDegrees=15, swing2LimitDegrees=15, twistLimitDegrees=0}
 {bone1Name=dyn_blouse_side_04_l, bone2Name=dyn_blouse_side_03_l, swing1LimitDegrees=8, swing2LimitDegrees=15, twistLimitDegrees=0}
 {
  bone1Name=dyn_blouse_base_side_r
  bone2Name=spine_02
  swing1LimitDegrees=45
  swing1Motion=Free
  swing2LimitDegrees=45
  swing2Motion=Free
  twistLimitDegrees=45
  twistMotion=Free
 }
 {bone1Name=dyn_blouse_side_01_r, bone2Name=dyn_blouse_base_side_r, swing1LimitDegrees=15, swing2LimitDegrees=15, twistLimitDegrees=0}
 {bone1Name=dyn_blouse_side_02_r, bone2Name=dyn_blouse_side_01_r, swing1LimitDegrees=15, swing2LimitDegrees=15, twistLimitDegrees=0}
 {bone1Name=dyn_blouse_side_03_r, bone2Name=dyn_blouse_side_02_r, swing1LimitDegrees=15, swing2LimitDegrees=15, twistLimitDegrees=0}
 {bone1Name=dyn_blouse_side_04_r, bone2Name=dyn_blouse_side_03_r, swing1LimitDegrees=15, swing2LimitDegrees=15, twistLimitDegrees=0}
 {bone1Name=dyn_blouse_04_c, bone2Name=dyn_blouse_03_c, swing1LimitDegrees=5, swing2LimitDegrees=30, twistLimitDegrees=5}
 {
  bone1Name=dyn_blouse_base_01_c
  bone2Name=spine_02
  swing1LimitDegrees=45
  swing1Motion=Free
  swing2LimitDegrees=45
  swing2Motion=Free
  twistLimitDegrees=45
  twistMotion=Free
 }
 {bone1Name=dyn_blouse_01_c, bone2Name=dyn_blouse_base_01_c, swing1LimitDegrees=12, swing2LimitDegrees=5, twistLimitDegrees=0}
 {bone1Name=dyn_blouse_02_c, bone2Name=dyn_blouse_01_c, swing1LimitDegrees=0, swing2LimitDegrees=30, twistLimitDegrees=0}
 {bone1Name=dyn_blouse_03_c, bone2Name=dyn_blouse_02_c, swing1LimitDegrees=5, swing2LimitDegrees=25, twistLimitDegrees=0}
 {
  bone1Name=dyn_blouse_04_c
  bone2Name=dyn_blouse_bk_04_r
  swing1LimitDegrees=45
  swing1Motion=Locked
  swing2LimitDegrees=45
  swing2Motion=Free
  twistLimitDegrees=45
  twistMotion=Free
 }
 {
  bone1Name=dyn_blouse_04_c
  bone2Name=dyn_blouse_bk_04_l
  swing1LimitDegrees=45
  swing1Motion=Locked
  swing2LimitDegrees=45
  swing2Motion=Free
  twistLimitDegrees=45
  twistMotion=Free
 }
 {
  bone1Name=dyn_blouse_side_03_r
  bone2Name=dyn_blouse_fr_03_r
  swing1LimitDegrees=45
  swing1Motion=Locked
  swing2LimitDegrees=45
  swing2Motion=Locked
  twistLimitDegrees=45
  twistMotion=Free
 }
 {
  bone1Name=dyn_blouse_bk_04_r
  bone2Name=dyn_blouse_side_04_r
  swing1LimitDegrees=45
  swing1Motion=Locked
  swing2LimitDegrees=45
  swing2Motion=Locked
  twistLimitDegrees=45
  twistMotion=Locked
 }
 {
  bone1Name=dyn_blouse_bk_04_l
  bone2Name=dyn_blouse_side_04_l
  swing1LimitDegrees=45
  swing1Motion=Locked
  swing2LimitDegrees=45
  swing2Motion=Locked
  twistLimitDegrees=45
  twistMotion=Locked
 }
 {
  bone1Name=calf_twist_02_r
  bone2Name=thigh_r
  swing1LimitDegrees=45
  swing1Motion=Free
  swing2LimitDegrees=45
  swing2Motion=Free
  twistLimitDegrees=45
  twistMotion=Free
 }
 {bone1Name=dyn_blouse_fr_02_r, bone2Name=dyn_blouse_fr_01_r, swing1LimitDegrees=30, swing2LimitDegrees=10, twistLimitDegrees=8}
 {
  bone1Name=dyn_blouse_side_03_l
  bone2Name=dyn_blouse_fr_03_l
  swing1LimitDegrees=45
  swing1Motion=Free
  swing2LimitDegrees=45
  swing2Motion=Free
  twistLimitDegrees=45
  twistMotion=Free
 }
]
