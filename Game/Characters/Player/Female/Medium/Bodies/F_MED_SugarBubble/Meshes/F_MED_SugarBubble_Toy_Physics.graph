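# Physics asset for the F_MED_SugarBubble female player body, covering secondary motion of left-front blouse, katball dangle and two rope bones.
# Spine capsules form the anchor chain, with blouse base on spine_02 and both ropes on spine_05.
# Four blouse boxes and the katball chain branch from the blouse base.
# Katball and rope segments have tight angular limits, other joints swing and twist freely, and an unconstrained left-thigh capsule adds a collision shape.
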PhysicsAsset /Game/Characters/Player/Female/Medium/Bodies/F_MED_SugarBubble/Meshes/F_MED_SugarBubble_Toy_Physics
bodies {
 dyn_blouse_base_fr_l [{
   shapeType=Capsule
   length=2
   radius=9.5
   shapeName=dyn_blouse_fr_01_l_capsule
   center=<-0.1308 2.3201 -4.112>
   rotation=<rot 70.3319 97.9686 148.9038>
  }]
 dyn_blouse_fr_01_l [{shapeType=Box, extentX=13.5, extentY=14.5, extentZ=2.5, shapeName=dyn_blouse_fr_01_l_box, center=<0 0 -1.7446>}]
 dyn_blouse_fr_02_l [{shapeType=Box, extentX=13.5, extentY=14.5, extentZ=2.5, shapeName=dyn_blouse_fr_02_l_box, center=<0 0 -1.7446>}]
 dyn_blouse_fr_03_l [{shapeType=Box, extentX=10.5, extentY=14.5, extentZ=2.5, shapeName=dyn_blouse_fr_03_l_box, center=<0 0 -1.7446>, rotation=<rot 12.1379 0 0>}]
 dyn_blouse_fr_04_l [{shapeType=Box, extentX=13.5, extentY=14.5, extentZ=2.5, shapeName=dyn_blouse_fr_04_l_box, center=<0 0 -1.7446>}]
 dyn_katball_01 [{shapeType=Capsule, length=0.505, radius=0.505, shapeName=dyn_katball_01_capsule, center=<2.0781 0 -0.4038>, rotation=<rot 84.6355 0 0>}]
 dyn_katball_02 [{shapeType=Capsule, length=3.2602, radius=0.8207, shapeName=dyn_katball_02_capsule, center=<2.8128 0.0152 -0.009>, rotation=<rot 90 0 0>}]
 dyn_katball_03 [{shapeType=Capsule, length=0, radius=2.749, shapeName=dyn_katball_03_capsule, center=<5.2375 0.5312 -0.0048>, rotation=<rot 90 0 0>}]
 dyn_rope_l_01 [{
   shapeType=Capsule
   length=8.6854
   radius=0.5
   shapeName=dyn_rope_l_01_capsule
   center=<3.4142 -0.017 -0.0083>
   rotation=<rot 0 -87.5263 -85.6194>
  }]
 dyn_rope_l_02 [{
   shapeType=Capsule
   length=3.2774
   radius=1.0839
   shapeName=dyn_rope_l_02_capsule
   center=<6.5622 0.2917 -0.674>
   rotation=<rot 0 -90.9667 -76.83>
  }]
 dyn_rope_l_base [{shapeType=Capsule, length=0.505, radius=0.505, shapeName=dyn_rope_l_base_capsule, rotation=<rot 0 -90 0>}]
 dyn_rope_r_01 [{
   shapeType=Capsule
   length=6.3977
   radius=0.5853
   shapeName=dyn_rope_r_01_capsule
   center=<-3.3813 0.0182 -0.3771>
   rotation=<rot 0 -89.0867 -94.3625>
  }]
 dyn_rope_r_02 [{
   shapeType=Capsule
   length=6.0761
   radius=0.7
   shapeName=dyn_rope_r_02_capsule
   center=<-9.5407 -0.3473 0.9129>
   rotation=<rot 0 -89.1809 -78.5175>
  }]
 dyn_rope_r_base [{shapeType=Capsule, length=0.505, radius=0.505, shapeName=dyn_rope_r_base_capsule, rotation=<rot 0 -90 0>}]
 katball_base [{shapeType=Capsule, length=0.505, radius=0.505, shapeName=katball_base_capsule, rotation=<rot 0 -90 0>}]
 spine_02 [{shapeType=Capsule, length=0.5, radius=17.4634, shapeName=spine_02_capsule, center=<3.2663 -1.8052 0>}]
 spine_03 [{shapeType=Capsule, length=9.005, radius=12.505, shapeName=spine_03_capsule, rotation=<rot 0 -90 0>}]
 spine_04 [{shapeType=Capsule, length=8.3907, radius=14.1105, shapeName=spine_04_capsule, center=<0 -1.8282 0>, rotation=<rot 0 -90 0>}]
 spine_05 [{shapeType=Capsule, length=11.3907, radius=14.1105, shapeName=spine_04_capsule, center=<-7.016 -2.8481 -1.2072>, rotation=<rot 0 -90 0>}]
 thigh_l [{
   shapeType=Capsule
   length=29.2565
   radius=8.1892
   shapeName=thigh_l_capsule
   center=<-12.9585 -1.3766 0.4223>
   rotation=<rot 0.1824 87.2778 -88.9656>
  }]
}
constraints [
 {bone1Name=dyn_katball_03, bone2Name=dyn_katball_02, swing1LimitDegrees=15, swing2LimitDegrees=30, twistLimitDegrees=3}
 {bone1Name=dyn_katball_01, bone2Name=katball_base, swing1LimitDegrees=6, swing2LimitDegrees=20, twistLimitDegrees=3}
 {bone1Name=dyn_katball_02, bone2Name=dyn_katball_01, swing1LimitDegrees=15, swing2LimitDegrees=25, twistLimitDegrees=3}
 {bone1Name=dyn_rope_r_02, bone2Name=dyn_rope_r_01, swing1LimitDegrees=5, swing2LimitDegrees=20, twistLimitDegrees=2}
 {
  bone1Name=dyn_rope_l_base
  bone2Name=spine_05
  swing1LimitDegrees=45
  swing1Motion=Free
  swing2LimitDegrees=45
  swing2Motion=Free
  twistLimitDegrees=45
  twistMotion=Free
 }
 {bone1Name=dyn_rope_l_01, bone2Name=dyn_rope_l_base, swing1LimitDegrees=5, swing2LimitDegrees=10, twistLimitDegrees=2}
 {bone1Name=dyn_rope_l_02, bone2Name=dyn_rope_l_01, swing1LimitDegrees=5, swing2LimitDegrees=25, twistLimitDegrees=2}
 {
  bone1Name=dyn_rope_r_base
  bone2Name=spine_05
  swing1LimitDegrees=45
  swing1Motion=Free
  swing2LimitDegrees=45
  swing2Motion=Free
  twistLimitDegrees=45
  twistMotion=Free
 }
 {bone1Name=dyn_rope_r_01, bone2Name=dyn_rope_r_base, swing1LimitDegrees=5, swing2LimitDegrees=20, twistLimitDegrees=2}
 {
  bone1Name=spine_04
  bone2Name=spine_03
  swing1LimitDegrees=45
  swing1Motion=Free
  swing2LimitDegrees=45
  swing2Motion=Free
  twistLimitDegrees=45
  twistMotion=Free
 }
 {
  bone1Name=spine_03
  bone2Name=spine_02
  swing1LimitDegrees=45
  swing1Motion=Free
  swing2LimitDegrees=45
  swing2Motion=Free
  twistLimitDegrees=45
  twistMotion=Free
 }
 {
  bone1Name=spine_05
  bone2Name=spine_04
  swing1LimitDegrees=45
  swing1Motion=Free
  swing2LimitDegrees=45
  swing2Motion=Free
  twistLimitDegrees=45
  twistMotion=Free
 }
 {
  bone1Name=dyn_blouse_fr_01_l
  bone2Name=dyn_blouse_base_fr_l
  swing1LimitDegrees=45
  swing1Motion=Free
  swing2LimitDegrees=45
  swing2Motion=Free
  twistLimitDegrees=45
  twistMotion=Free
 }
 {
  bone1Name=dyn_blouse_fr_02_l
  bone2Name=dyn_blouse_base_fr_l
  swing1LimitDegrees=45
  swing1Motion=Free
  swing2LimitDegrees=45
  swing2Motion=Free
  twistLimitDegrees=45
  twistMotion=Free
 }
 {
  bone1Name=dyn_blouse_fr_03_l
  bone2Name=dyn_blouse_base_fr_l
  swing1LimitDegrees=45
  swing1Motion=Free
  swing2LimitDegrees=45
  swing2Motion=Free
  twistLimitDegrees=45
  twistMotion=Free
 }
 {
  bone1Name=dyn_blouse_fr_04_l
  bone2Name=dyn_blouse_base_fr_l
  swing1LimitDegrees=45
  swing1Motion=Free
  swing2LimitDegrees=45
  swing2Motion=Free
  twistLimitDegrees=45
  twistMotion=Free
 }
 {
  bone1Name=katball_base
  bone2Name=dyn_blouse_base_fr_l
  swing1LimitDegrees=45
  swing1Motion=Free
  swing2LimitDegrees=45
  swing2Motion=Free
  twistLimitDegrees=45
  twistMotion=Free
 }
 {
  bone1Name=dyn_blouse_base_fr_l
  bone2Name=spine_02
  swing1LimitDegrees=45
  swing1Motion=Free
  swing2LimitDegrees=45
  swing2Motion=Free
  twistLimitDegrees=45
  twistMotion=Free
 }
]
