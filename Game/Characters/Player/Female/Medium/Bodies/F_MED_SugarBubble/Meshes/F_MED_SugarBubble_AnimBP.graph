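# Post-process style AnimBlueprint for the F_MED_SugarBubble female body, adding secondary motion to its toy, straps and blouse.
# It copies the parent mesh's pose, then applies Control Rigs (fed slide/DBNO flags) and bone tweaks such as a backbling-driven collar.
# RigidBody passes switch physics assets for skydiving and front-end idle.
# Left leg-strap RigidBody, right-strap and blouse AnimDynamics follow, with rotation constraints blending back hem bones.
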AnimBlueprint /Game/Characters/Player/Female/Medium/Bodies/F_MED_SugarBubble/Meshes/F_MED_SugarBubble_AnimBP
target_skeleton=&/F_MED_SugarBubble_Skeleton
graphs [
 AnimGraph {nodes [
   Anim Dynamics - Bone: dyn_blouse_bk_01_l!(ComponentPose, Alpha=0.5, GravityOverride:Vector="0,0,-980", ExternalForce="0,0,0") -> (Pose) {
    angularDampingOverride=0.999
    angularSpringConstant=50
    bAngularSpring=true
    bUseGravityOverride=true
    gravityScale=0.35
    simulationSpace=RootRelative
    alphaBoolBlend {blendOption=Linear}
    boundBone {boneName=dyn_blouse_bk_01_l}
    componentLinearAccScale=<0.4 0.4 0.4>
    componentLinearVelScale=<0.4 0.4 0.4>
    gravityOverride {z=-980}
    physicsBodyDefinitions [{
      boundBone {boneName=dyn_blouse_bk_01_l}
      boxExtents=<16 16 16>
      constraintSetup {angularLimitsMax=<10 0 1>, angularLimitsMin=<-10 -15 -1>, angularTarget=<1 0 0>}
      localJointOffset=<2 0 0>
     }]
    relativeSpaceBone {boneName=spine_02}
   }
   Anim Dynamics - Bone: dyn_blouse_bk_01_r!(ComponentPose, Alpha=0.5, GravityOverride:Vector="0,0,-1000", ExternalForce="0,0,0") -> (Pose) {
    angularDampingOverride=0.999
    angularSpringConstant=50
    bAngularSpring=true
    bOverrideAngularDamping=true
    bUseGravityOverride=true
    gravityScale=0.35
    simulationSpace=RootRelative
    alphaBoolBlend {blendOption=Linear}
    boundBone {boneName=dyn_blouse_bk_01_r}
    componentLinearAccScale=<0.4 0.4 0.4>
    componentLinearVelScale=<0.4 0.4 0.4>
    gravityOverride {z=-1000}
    physicsBodyDefinitions [{
      boundBone {boneName=dyn_blouse_bk_01_r}
      boxExtents=<16 16 16>
      constraintSetup {angularLimitsMax=<10 0 1>, angularLimitsMin=<-10 -15 -1>, angularTarget=<1 0 0>}
      localJointOffset=<2 0 0>
     }]
    relativeSpaceBone {boneName=spine_02}
   }
   Anim Dynamics - Bone: dyn_blouse_01_c!(ComponentPose, Alpha=0.5, GravityOverride:Vector="0,0,-1000", ExternalForce="0,0,0") -> (Pose) {
    angularDampingOverride=0.9999
    angularSpringConstant=50
    bAngularSpring=true
    bOverrideAngularDamping=true
    bUseGravityOverride=true
    gravityScale=0.35
    simulationSpace=RootRelative
    alphaBoolBlend {blendOption=Linear}
    boundBone {boneName=dyn_blouse_01_c}
    componentLinearAccScale=<0.4 0.4 0.4>
    componentLinearVelScale=<0.4 0.4 0.4>
    gravityOverride {z=-1000}
    physicsBodyDefinitions [{
      boundBone {boneName=dyn_blouse_01_c}
      boxExtents=<16 16 16>
      constraintSetup {angularLimitsMax=<10 0 1>, angularLimitsMin=<-10 -15 -1>, angularTarget=<1 0 0>}
      localJointOffset=<2 0 0>
     }]
    relativeSpaceBone {boneName=spine_02}
   }
   n1 = Anim Dynamics - Bone: dyn_strap_l_01!(ComponentPose, Alpha=1, GravityOverride:Vector="0,0,-980", ExternalForce="0,0,0") -> (Pose) {
    angularDampingOverride=0.9
    angularSpringConstant=10
    bOverrideAngularDamping=true
    bUseGravityOverride=true
    bUseSphericalLimits=true
    simulationSpace=RootRelative
    alphaBoolBlend {blendOption=Linear}
    boundBone {boneName=dyn_strap_l_01}
    componentAppliedLinearAccClamp=<3000 3000 3000>
    componentLinearAccScale=<3 3 3>
    componentLinearVelScale=<3 3 3>
    gravityOverride {z=-980}
    physicsBodyDefinitions [{
      collisionType=OuterSphere
      boundBone {boneName=dyn_strap_l_01}
      boxExtents=<6 6 6>
      constraintSetup {angularLimitsMax=<5 5 10>, angularLimitsMin=<-5 0 -20>, angularTarget=<0.1 0 0>}
      localJointOffset=<0 0 -0.1>
     }]
    relativeSpaceBone {boneName=pelvis}
    sphericalLimits [{limitRadius=23, drivingBone {boneName=calf_r}, sphereLocalOffset=<39.2388 0 12.229>}]
   }
   n2 = Anim Dynamics - Bone: dyn_strap_l_02!(ComponentPose=@n1, Alpha=1, GravityOverride:Vector="0,0,-980", ExternalForce="0,0,0") -> (Pose) {
    angularDampingOverride=0.8
    angularSpringConstant=100
    bAngularSpring=true
    bOverrideAngularDamping=true
    bUseGravityOverride=true
    simulationSpace=BoneRelative
    alphaBoolBlend {blendOption=Linear}
    boundBone {boneName=dyn_strap_l_02}
    componentLinearAccScale=<20 20 20>
    componentLinearVelScale=<20 20 20>
    gravityOverride {z=-980}
    physicsBodyDefinitions [{
      boundBone {boneName=dyn_strap_l_02}
      boxExtents=<6 6 6>
      constraintSetup {angularLimitsMax=<0 5 10>, angularLimitsMin=<0 -10 -20>, angularTarget=<0.1 0 0>}
      localJointOffset=<0 0 0>
     }]
    relativeSpaceBone {boneName=dyn_strap_l_01}
   }
   n3 = Anim Dynamics - Bone: dyn_strap_l_03!(ComponentPose=@n2, Alpha=1, GravityOverride:Vector="0,0,-980", ExternalForce="0,0,0") -> (Pose) {
    angularDampingOverride=0.8
    angularSpringConstant=100
    bAngularSpring=true
    bOverrideAngularDamping=true
    bUseGravityOverride=true
    bUseSphericalLimits=true
    simulationSpace=BoneRelative
    alphaBoolBlend {blendOption=Linear}
    boundBone {boneName=dyn_strap_l_03}
    componentLinearAccScale=<20 20 20>
    componentLinearVelScale=<20 20 20>
    gravityOverride {z=-980}
    physicsBodyDefinitions [{
      collisionType=OuterSphere
      boundBone {boneName=dyn_strap_l_03}
      boxExtents=<6 6 6>
      constraintSetup {angularLimitsMax=<0 5 10>, angularLimitsMin=<0 -10 -20>, angularTarget=<0.1 0 0>}
      localJointOffset=<0 0 0>
     }]
    relativeSpaceBone {boneName=dyn_strap_l_01}
    sphericalLimits [
     {limitRadius=23, drivingBone {boneName=calf_r}, sphereLocalOffset=<39.2388 0 12.229>}
     {limitRadius=23, drivingBone {boneName=calf_l}, sphereLocalOffset=<-37.0445 0 -17.3015>}
    ]
   }
   Anim Dynamics - Bone: dyn_strap_l_04!(ComponentPose=@n3, Alpha=1, GravityOverride:Vector="0,0,-980", ExternalForce="0,0,0") -> (Pose) {
    angularDampingOverride=0.8
    angularSpringConstant=100
    bAngularSpring=true
    bOverrideAngularDamping=true
    bUseGravityOverride=true
    bUseSphericalLimits=true
    simulationSpace=BoneRelative
    alphaBoolBlend {blendOption=Linear}
    boundBone {boneName=dyn_strap_l_04}
    componentLinearAccScale=<20 20 20>
    componentLinearVelScale=<20 20 20>
    gravityOverride {z=-980}
    physicsBodyDefinitions [{
      collisionType=OuterSphere
      boundBone {boneName=dyn_strap_l_04}
      boxExtents=<6 6 6>
      constraintSetup {angularLimitsMax=<0 5 10>, angularLimitsMin=<0 -5 -20>, angularTarget=<0.1 0 0>}
      localJointOffset=<0 0 0>
     }]
    relativeSpaceBone {boneName=dyn_strap_l_01}
    sphericalLimits [
     {limitRadius=23, drivingBone {boneName=calf_r}, sphereLocalOffset=<39.2388 0 12.229>}
     {limitRadius=23, drivingBone {boneName=calf_l}, sphereLocalOffset=<-37.0445 0 -17.3015>}
    ]
   }
   "Transform (Modify) Bone - Bone: shirt_scale"!(ComponentPose, bAlphaBoolEnabled:bool=True, Translation="0,0,0", Rotation="0,0,0", Scale="0.1,0.1,0.1") -> (Pose) {
    alphaInputType=Bool
    scaleMode=BMM_Additive
    translationMode=BMM_Additive
    alphaBoolBlend {blendOption=Linear}
    boneToModify {boneName=shirt_scale}
    scale=<0.1 0.1 0.1>
   }
   "Transform (Modify) Bone - Bone: dyn_rope_l_02"!(ComponentPose, bAlphaBoolEnabled:bool=True, Translation="0,0,0", Rotation="-10.261,-3.078,8.623", Scale="1,1,1") -> (Pose) {
    alphaInputType=Bool
    rotationMode=BMM_Additive
    translationMode=BMM_Additive
    alphaBoolBlend {blendOption=Linear}
    boneToModify {boneName=dyn_rope_l_02}
    rotation=<rot -10.261 -3.078 8.623>
   }
   RigidBody!(ComponentPose, bAlphaBoolEnabled:bool, OverrideWorldGravity:Vector="0,0,0", ExternalForce) -> (Pose) {
    alphaInputType=Bool
    bClampLinearTranslationLimitToRefPose=true
    bOverrideWorldGravity=true
    overridePhysicsAsset=&/F_MED_SugarBubble_LegStrap_R_Physics
    simulationSpace=BaseBoneSpace
    alphaBoolBlend {blendOption=Linear}
    baseBoneRef {boneName=pelvis}
    componentAppliedLinearAccClamp=<1200 1200 1200>
    componentLinearAccScale=<0.25 0.25 0.4>
    componentLinearVelScale=<0.3 0.3 0.3>
    simSpaceSettings {
     maxAngularAcceleration=100
     maxAngularVelocity=2
     maxLinearAcceleration=100
     maxLinearVelocity=160
     velocityScaleZ=0
     worldAlpha=0.75
     externalLinearDragV=<1 1 1>
    }
   }
   n4 = RigidBody(ComponentPose=Local To Component(LocalPose=Use cached pose 'frontend pass'()), OverrideWorldGravity="0,0,0") {
    alphaInputType=Bool
    bClampLinearTranslationLimitToRefPose=true
    bOverrideWorldGravity=true
    overridePhysicsAsset=&/F_MED_SugarBubble_Toy_Physics
    simulationSpace=BaseBoneSpace
    alphaBoolBlend {blendOption=Linear}
    baseBoneRef {boneName=spine_02}
    componentAppliedLinearAccClamp=<1200 1200 1200>
    componentLinearAccScale=<0.25 0.25 0.4>
    componentLinearVelScale=<0.3 0.3 0.3>
    simSpaceSettings {
     maxAngularAcceleration=100
     maxAngularVelocity=2
     maxLinearAcceleration=100
     maxLinearVelocity=160
     velocityScaleZ=0
     worldAlpha=0.75
     externalLinearDragV=<1 1 1>
    }
   }
   n5 = Component To Local(ComponentPose=@n4)
   n6 = RigidBody(ComponentPose=Local To Component(LocalPose=Use cached pose 'frontend pass'()), OverrideWorldGravity="0,0,0") {
    alphaInputType=Bool
    bClampLinearTranslationLimitToRefPose=true
    bOverrideWorldGravity=true
    overridePhysicsAsset=&/F_MED_SugarBubble_Toy_FrontEnd_Physics
    simulationSpace=BaseBoneSpace
    alphaBoolBlend {blendOption=Linear}
    baseBoneRef {boneName=spine_02}
    componentAppliedLinearAccClamp=<1200 1200 1200>
    componentLinearAccScale=<0.25 0.25 0.4>
    componentLinearVelScale=<0.3 0.3 0.3>
    simSpaceSettings {
     maxAngularAcceleration=100
     maxAngularVelocity=2
     maxLinearAcceleration=100
     maxLinearVelocity=160
     velocityScaleZ=0
     worldAlpha=0.75
     externalLinearDragV=<1 1 1>
    }
   }
   n7 = Component To Local(ComponentPose=@n6)
   n8 = Get FrontEnd Idle Bypass() -> (FrontEnd Idle Bypass:bool)
   n9 = RigidBody(
    ComponentPose=Local To Component(LocalPose=Blend Poses by bool(BlendPose_0=@n7, BlendPose_1=@n5, bActiveValue=@n8) {blendType=Linear})
    OverrideWorldGravity="0,0,0"
   ) {
    alphaInputType=Bool
    bClampLinearTranslationLimitToRefPose=true
    bOverrideWorldGravity=true
    overridePhysicsAsset=&/F_MED_SugarBubble_LegStrap_L_Physics
    simulationSpace=BaseBoneSpace
    alphaBoolBlend {blendOption=Linear}
    baseBoneRef {boneName=calf_l}
    componentAppliedLinearAccClamp=<1200 1200 1200>
    componentLinearAccScale=<0.25 0.25 0.4>
    componentLinearVelScale=<0.3 0.3 0.3>
    simSpaceSettings {
     maxAngularAcceleration=100
     maxAngularVelocity=2
     maxLinearAcceleration=100
     maxLinearVelocity=160
     velocityScaleZ=0
     worldAlpha=0.75
     externalLinearDragV=<1 1 1>
    }
   }
   n10 = Anim Dynamics - Bone: dyn_strap_r_01!(ComponentPose=@n9, Alpha=1, GravityOverride:Vector="0,0,-980", ExternalForce="0,0,0") -> (Pose) {
    angularDampingOverride=0.6
    angularSpringConstant=10
    bGravityOverrideInSimSpace=true
    bOverrideAngularDamping=true
    bUseGravityOverride=true
    simulationSpace=RootRelative
    alphaBoolBlend {blendOption=Linear}
    boundBone {boneName=dyn_strap_r_01}
    componentAppliedLinearAccClamp=<3000 3000 3000>
    componentLinearAccScale=<10 10 10>
    componentLinearVelScale=<10 10 10>
    gravityOverride {z=-980}
    physicsBodyDefinitions [{
      boundBone {boneName=dyn_strap_r_01}
      boxExtents=<6 6 6>
      constraintSetup {angularLimitsMax=<0 5 10>, angularLimitsMin=<0 -5 -20>, angularTarget=<1 0 0>}
      localJointOffset=<0.1 0 0>
     }]
    relativeSpaceBone {boneName=calf_r}
   }
   n11 = Anim Dynamics - Bone: dyn_strap_r_02!(ComponentPose=@n10, Alpha=1, GravityOverride:Vector="0,0,-980", ExternalForce="0,0,0") -> (Pose) {
    angularDampingOverride=0.6
    angularSpringConstant=10
    bGravityOverrideInSimSpace=true
    bOverrideAngularDamping=true
    bUseGravityOverride=true
    simulationSpace=RootRelative
    alphaBoolBlend {blendOption=Linear}
    boundBone {boneName=dyn_strap_r_02}
    componentAppliedLinearAccClamp=<3000 3000 3000>
    componentLinearAccScale=<10 10 10>
    componentLinearVelScale=<10 10 10>
    gravityOverride {z=-980}
    physicsBodyDefinitions [{
      boundBone {boneName=dyn_strap_r_02}
      boxExtents=<6 6 6>
      constraintSetup {angularLimitsMax=<0 5 10>, angularLimitsMin=<0 -5 -20>, angularTarget=<1 0 0>}
      localJointOffset=<0.1 0 0>
     }]
    relativeSpaceBone {boneName=calf_r}
   }
   n12 = Anim Dynamics - Bone: dyn_strap_r_03!(ComponentPose=@n11, Alpha=1, GravityOverride:Vector="0,0,-980", ExternalForce="0,0,0") -> (Pose) {
    angularDampingOverride=0.9
    angularSpringConstant=10
    bGravityOverrideInSimSpace=true
    bOverrideAngularDamping=true
    bUseGravityOverride=true
    bUseSphericalLimits=true
    simulationSpace=RootRelative
    alphaBoolBlend {blendOption=Linear}
    boundBone {boneName=dyn_strap_r_03}
    componentAppliedLinearAccClamp=<3000 3000 3000>
    componentLinearAccScale=<10 10 10>
    componentLinearVelScale=<10 10 10>
    gravityOverride {z=-980}
    physicsBodyDefinitions [{
      collisionType=OuterSphere
      boundBone {boneName=dyn_strap_r_03}
      boxExtents=<6 6 6>
      constraintSetup {angularLimitsMax=<0 8 10>, angularLimitsMin=<0 -5 -20>, angularTarget=<1 0 0>}
      localJointOffset=<0.1 -0.2 0>
     }]
    relativeSpaceBone {boneName=calf_r}
    sphericalLimits [
     {limitRadius=23, drivingBone {boneName=calf_l}, sphereLocalOffset=<-40.5511 0 -12.9337>}
     {limitRadius=23, drivingBone {boneName=calf_r}, sphereLocalOffset=<37.6393 -4.7986 14.3324>}
    ]
   }
   n13 = Anim Dynamics - Bone: dyn_strap_r_04!(ComponentPose=@n12, Alpha=1, GravityOverride:Vector="0,0,-980", ExternalForce="0,0,0") -> (Pose) {
    angularDampingOverride=0.99
    angularSpringConstant=10
    bGravityOverrideInSimSpace=true
    bOverrideAngularDamping=true
    bUseGravityOverride=true
    bUseSphericalLimits=true
    simulationSpace=RootRelative
    alphaBoolBlend {blendOption=Linear}
    boundBone {boneName=dyn_strap_r_04}
    componentAppliedLinearAccClamp=<3000 3000 3000>
    componentLinearAccScale=<10 10 10>
    componentLinearVelScale=<10 10 10>
    gravityOverride {z=-980}
    physicsBodyDefinitions [{
      collisionType=OuterSphere
      boundBone {boneName=dyn_strap_r_04}
      boxExtents=<6 6 6>
      constraintSetup {angularLimitsMax=<0 10 10>, angularLimitsMin=<0 -5 -20>, angularTarget=<1 0 0>}
      localJointOffset=<0.1 0 0>
     }]
    relativeSpaceBone {boneName=calf_r}
    sphericalLimits [
     {limitRadius=23, drivingBone {boneName=calf_l}, sphereLocalOffset=<-40.5511 0 -12.9337>}
     {limitRadius=23, drivingBone {boneName=calf_r}, sphereLocalOffset=<37.8377 -4.7986 14.3324>}
    ]
   }
   n14 = "Transform (Modify) Bone - Bone: dyn_blouse_bk_03_l"(ComponentPose=@n13, Alpha=1, Translation="0,0,0", Rotation="0.000,0.000,-8.548", Scale="1,1,1") {rotationMode=BMM_Additive, alphaBoolBlend {blendOption=Linear}, boneToModify {boneName=dyn_blouse_bk_03_l}, rotation {roll=-8.548}}
   n15 = Anim Dynamics - Bone: dyn_blouse_bk_02_l!(ComponentPose=@n14, Alpha=0.5, GravityOverride:Vector="0,0,-980", ExternalForce="0,0,0") -> (Pose) {
    angularDampingOverride=0.999
    angularSpringConstant=50
    bAngularSpring=true
    bUseGravityOverride=true
    gravityScale=0.35
    simulationSpace=RootRelative
    alphaBoolBlend {blendOption=Linear}
    boundBone {boneName=dyn_blouse_bk_02_l}
    componentLinearAccScale=<0.4 0.4 0.4>
    componentLinearVelScale=<0.4 0.4 0.4>
    gravityOverride {z=-980}
    physicsBodyDefinitions [{
      boundBone {boneName=dyn_blouse_bk_02_l}
      boxExtents=<16 16 16>
      constraintSetup {angularLimitsMax=<10 0 1>, angularLimitsMin=<-10 -15 -1>, angularTarget=<1 0 0>}
      localJointOffset=<2 0 0>
     }]
    relativeSpaceBone {boneName=spine_02}
   }
   n16 = Anim Dynamics - Bone: dyn_blouse_bk_03_l!(ComponentPose=@n15, Alpha=0.5, GravityOverride:Vector="0,0,-1000", ExternalForce="0,0,0") -> (Pose) {
    angularDampingOverride=0.999
    angularSpringConstant=50
    bAngularSpring=true
    bOverrideAngularDamping=true
    bUseGravityOverride=true
    gravityScale=0.35
    simulationSpace=RootRelative
    alphaBoolBlend {blendOption=Linear}
    boundBone {boneName=dyn_blouse_bk_03_l}
    componentLinearAccScale=<0.4 0.4 0.4>
    componentLinearVelScale=<0.4 0.4 0.4>
    gravityOverride {z=-1000}
    physicsBodyDefinitions [{
      boundBone {boneName=dyn_blouse_bk_03_l}
      boxExtents=<16 16 16>
      constraintSetup {angularLimitsMax=<10 0 1>, angularLimitsMin=<-10 -15 -1>, angularTarget=<1 0 0>}
      localJointOffset=<2 0 0>
     }]
    relativeSpaceBone {boneName=spine_02}
   }
   n17 = Anim Dynamics - Bone: dyn_blouse_bk_04_l!(ComponentPose=@n16, Alpha=0.5, GravityOverride:Vector="0,0,-1000", ExternalForce="0,0,0") -> (Pose) {
    angularDampingOverride=0.999
    angularSpringConstant=50
    bAngularSpring=true
    bOverrideAngularDamping=true
    bUseGravityOverride=true
    gravityScale=0.35
    simulationSpace=RootRelative
    alphaBoolBlend {blendOption=Linear}
    boundBone {boneName=dyn_blouse_bk_04_l}
    componentLinearAccScale=<0.4 0.4 0.4>
    componentLinearVelScale=<0.4 0.4 0.4>
    gravityOverride {z=-1000}
    physicsBodyDefinitions [{
      boundBone {boneName=dyn_blouse_bk_04_l}
      boxExtents=<16 16 16>
      constraintSetup {angularLimitsMax=<10 0 1>, angularLimitsMin=<-10 -15 -1>, angularTarget=<1 0 0>}
      localJointOffset=<2 0 0>
     }]
    relativeSpaceBone {boneName=spine_02}
   }
   n18 = Anim Dynamics - Bone: dyn_blouse_02_c!(ComponentPose=@n17, Alpha=0.5, GravityOverride:Vector="0,0,-1000", ExternalForce="0,0,0") -> (Pose) {
    angularDampingOverride=0.9999
    angularSpringConstant=50
    bAngularSpring=true
    bOverrideAngularDamping=true
    bUseGravityOverride=true
    gravityScale=0.35
    simulationSpace=RootRelative
    alphaBoolBlend {blendOption=Linear}
    boundBone {boneName=dyn_blouse_02_c}
    componentLinearAccScale=<0.4 0.4 0.4>
    componentLinearVelScale=<0.4 0.4 0.4>
    gravityOverride {z=-1000}
    physicsBodyDefinitions [{
      boundBone {boneName=dyn_blouse_02_c}
      boxExtents=<16 16 16>
      constraintSetup {angularLimitsMax=<10 0 1>, angularLimitsMin=<-10 -15 -1>, angularTarget=<1 0 0>}
      localJointOffset=<2 0 0>
     }]
    relativeSpaceBone {boneName=spine_02}
   }
   n19 = Anim Dynamics - Bone: dyn_blouse_03_c!(ComponentPose=@n18, Alpha=0.5, GravityOverride:Vector="0,0,-1000", ExternalForce="0,0,0") -> (Pose) {
    angularDampingOverride=0.9999
    angularSpringConstant=50
    bAngularSpring=true
    bOverrideAngularDamping=true
    bUseGravityOverride=true
    gravityScale=0.35
    simulationSpace=RootRelative
    alphaBoolBlend {blendOption=Linear}
    boundBone {boneName=dyn_blouse_03_c}
    componentLinearAccScale=<0.4 0.4 0.4>
    componentLinearVelScale=<0.4 0.4 0.4>
    gravityOverride {z=-1000}
    physicsBodyDefinitions [{
      boundBone {boneName=dyn_blouse_03_c}
      boxExtents=<16 16 16>
      constraintSetup {angularLimitsMax=<10 0 1>, angularLimitsMin=<-10 -15 -1>, angularTarget=<1 0 0>}
      localJointOffset=<2 0 0>
     }]
    relativeSpaceBone {boneName=spine_02}
   }
   n20 = Anim Dynamics - Bone: dyn_blouse_04_c!(ComponentPose=@n19, Alpha=0.5, GravityOverride:Vector="0,0,-1000", ExternalForce="0,0,0") -> (Pose) {
    angularDampingOverride=0.9999
    angularSpringConstant=50
    bAngularSpring=true
    bOverrideAngularDamping=true
    bUseGravityOverride=true
    gravityScale=0.35
    simulationSpace=RootRelative
    alphaBoolBlend {blendOption=Linear}
    boundBone {boneName=dyn_blouse_04_c}
    componentLinearAccScale=<0.4 0.4 0.4>
    componentLinearVelScale=<0.4 0.4 0.4>
    gravityOverride {z=-1000}
    physicsBodyDefinitions [{
      boundBone {boneName=dyn_blouse_04_c}
      boxExtents=<16 16 16>
      constraintSetup {angularLimitsMax=<10 0 1>, angularLimitsMin=<-10 -15 -1>, angularTarget=<1 0 0>}
      localJointOffset=<2 0 0>
     }]
    relativeSpaceBone {boneName=spine_02}
   }
   n21 = Anim Dynamics - Bone: dyn_blouse_bk_02_r!(ComponentPose=@n20, Alpha=0.5, GravityOverride:Vector="0,0,-980", ExternalForce="0,0,0") -> (Pose) {
    angularDampingOverride=0.999
    angularSpringConstant=50
    bAngularSpring=true
    bUseGravityOverride=true
    gravityScale=0.35
    simulationSpace=RootRelative
    alphaBoolBlend {blendOption=Linear}
    boundBone {boneName=dyn_blouse_bk_02_r}
    componentLinearAccScale=<0.4 0.4 0.4>
    componentLinearVelScale=<0.4 0.4 0.4>
    gravityOverride {z=-980}
    physicsBodyDefinitions [{
      boundBone {boneName=dyn_blouse_bk_02_r}
      boxExtents=<16 16 16>
      constraintSetup {angularLimitsMax=<10 0 1>, angularLimitsMin=<-10 -15 -1>, angularTarget=<1 0 0>}
      localJointOffset=<2 0 0>
     }]
    relativeSpaceBone {boneName=spine_02}
   }
   n22 = Anim Dynamics - Bone: dyn_blouse_bk_03_r!(ComponentPose=@n21, Alpha=0.5, GravityOverride:Vector="0,0,-1000", ExternalForce="0,0,0") -> (Pose) {
    angularDampingOverride=0.999
    angularSpringConstant=50
    bAngularSpring=true
    bOverrideAngularDamping=true
    bUseGravityOverride=true
    gravityScale=0.35
    simulationSpace=RootRelative
    alphaBoolBlend {blendOption=Linear}
    boundBone {boneName=dyn_blouse_bk_03_r}
    componentLinearAccScale=<0.4 0.4 0.4>
    componentLinearVelScale=<0.4 0.4 0.4>
    gravityOverride {z=-1000}
    physicsBodyDefinitions [{
      boundBone {boneName=dyn_blouse_bk_03_r}
      boxExtents=<16 16 16>
      constraintSetup {angularLimitsMax=<10 0 1>, angularLimitsMin=<-10 -15 -1>, angularTarget=<1 0 0>}
      localJointOffset=<2 0 0>
     }]
    relativeSpaceBone {boneName=spine_02}
   }
   n23 = Anim Dynamics - Bone: dyn_blouse_bk_04_r!(ComponentPose=@n22, Alpha=0.5, GravityOverride:Vector="0,0,-1000", ExternalForce="0,0,0") -> (Pose) {
    angularDampingOverride=0.999
    angularSpringConstant=50
    bAngularSpring=true
    bOverrideAngularDamping=true
    bUseGravityOverride=true
    gravityScale=0.35
    simulationSpace=RootRelative
    alphaBoolBlend {blendOption=Linear}
    boundBone {boneName=dyn_blouse_bk_04_r}
    componentLinearAccScale=<0.4 0.4 0.4>
    componentLinearVelScale=<0.4 0.4 0.4>
    gravityOverride {z=-1000}
    physicsBodyDefinitions [{
      boundBone {boneName=dyn_blouse_bk_04_r}
      boxExtents=<16 16 16>
      constraintSetup {angularLimitsMax=<10 0 1>, angularLimitsMin=<-10 -15 -1>, angularTarget=<1 0 0>}
      localJointOffset=<2 0 0>
     }]
    relativeSpaceBone {boneName=spine_02}
   }
   n24 = Anim Dynamics - Bone: dyn_blouse_fr_01_r!(ComponentPose=@n23, Alpha=0.5, GravityOverride:Vector="0,0,-980", ExternalForce="0,0,0") -> (Pose) {
    angularDampingOverride=0.99
    angularSpringConstant=50
    bAngularSpring=true
    bOverrideAngularDamping=true
    bUseGravityOverride=true
    gravityScale=0.35
    simulationSpace=RootRelative
    alphaBoolBlend {blendOption=Linear}
    boundBone {boneName=dyn_blouse_fr_01_r}
    componentLinearAccScale=<0.4 0.4 0.4>
    componentLinearVelScale=<0.4 0.4 0.4>
    gravityOverride {z=-980}
    physicsBodyDefinitions [{
      boundBone {boneName=dyn_blouse_fr_01_r}
      boxExtents=<16 16 16>
      constraintSetup {angularLimitsMax=<0 0 1>, angularLimitsMin=<0 -15 -1>, angularTarget=<1 0 0>}
      localJointOffset=<2 0 0>
     }]
    relativeSpaceBone {boneName=spine_02}
   }
   n25 = Anim Dynamics - Bone: dyn_blouse_fr_02_r!(ComponentPose=@n24, Alpha=0.5, GravityOverride:Vector="0,0,-1000", ExternalForce="0,0,0") -> (Pose) {
    angularDampingOverride=0.999
    angularSpringConstant=50
    bAngularSpring=true
    bOverrideAngularDamping=true
    bUseGravityOverride=true
    gravityScale=0.35
    simulationSpace=RootRelative
    alphaBoolBlend {blendOption=Linear}
    boundBone {boneName=dyn_blouse_fr_02_r}
    componentLinearAccScale=<0.4 0.4 0.4>
    componentLinearVelScale=<0.4 0.4 0.4>
    gravityOverride {z=-1000}
    physicsBodyDefinitions [{
      boundBone {boneName=dyn_blouse_fr_02_r}
      boxExtents=<16 16 16>
      constraintSetup {angularLimitsMax=<0 0 1>, angularLimitsMin=<0 -15 -1>, angularTarget=<1 0 0>}
      localJointOffset=<2 0 0>
     }]
    relativeSpaceBone {boneName=dyn_blouse_fr_01_r}
   }
   n26 = Anim Dynamics - Bone: dyn_blouse_fr_03_r!(ComponentPose=@n25, Alpha=0.5, GravityOverride:Vector="0,0,-1000", ExternalForce="0,0,0") -> (Pose) {
    angularDampingOverride=0.999
    angularSpringConstant=50
    bAngularSpring=true
    bOverrideAngularDamping=true
    bUseGravityOverride=true
    gravityScale=0.35
    simulationSpace=RootRelative
    alphaBoolBlend {blendOption=Linear}
    boundBone {boneName=dyn_blouse_fr_03_r}
    componentLinearAccScale=<0.4 0.4 0.4>
    componentLinearVelScale=<0.4 0.4 0.4>
    gravityOverride {z=-1000}
    physicsBodyDefinitions [{
      boundBone {boneName=dyn_blouse_fr_03_r}
      boxExtents=<16 16 16>
      constraintSetup {angularLimitsMax=<0 0 1>, angularLimitsMin=<0 -15 -1>, angularTarget=<1 0 0>}
      localJointOffset=<2 0 0>
     }]
    relativeSpaceBone {boneName=dyn_blouse_fr_02_r}
   }
   n27 = Anim Dynamics - Bone: dyn_blouse_fr_04_r!(ComponentPose=@n26, Alpha=0.5, GravityOverride:Vector="0,0,-1000", ExternalForce="0,0,0") -> (Pose) {
    angularDampingOverride=0.999
    angularSpringConstant=50
    bAngularSpring=true
    bOverrideAngularDamping=true
    bUseGravityOverride=true
    gravityScale=0.35
    simulationSpace=RootRelative
    alphaBoolBlend {blendOption=Linear}
    boundBone {boneName=dyn_blouse_fr_04_r}
    componentLinearAccScale=<0.4 0.4 0.4>
    componentLinearVelScale=<0.4 0.4 0.4>
    gravityOverride {z=-1000}
    physicsBodyDefinitions [{
      boundBone {boneName=dyn_blouse_fr_04_r}
      boxExtents=<16 16 16>
      constraintSetup {angularLimitsMax=<0 0 1>, angularLimitsMin=<0 -15 -1>, angularTarget=<1 0 0>}
      localJointOffset=<2 0 0>
     }]
    relativeSpaceBone {boneName=dyn_blouse_fr_03_r}
   }
   n28 = Anim Dynamics - Bone: dyn_blouse_side_01_r!(ComponentPose=@n27, Alpha=0.5, GravityOverride:Vector="0,0,-980", ExternalForce="0,0,0") -> (Pose) {
    angularDampingOverride=0.999
    angularSpringConstant=50
    bAngularSpring=true
    bUseGravityOverride=true
    gravityScale=0.35
    simulationSpace=RootRelative
    alphaBoolBlend {blendOption=Linear}
    boundBone {boneName=dyn_blouse_side_01_r}
    componentLinearAccScale=<0.4 0.4 0.4>
    componentLinearVelScale=<0.4 0.4 0.4>
    gravityOverride {z=-980}
    physicsBodyDefinitions [{
      boundBone {boneName=dyn_blouse_side_01_r}
      boxExtents=<16 16 16>
      constraintSetup {angularLimitsMax=<5 0 1>, angularLimitsMin=<-5 -15 -1>, angularTarget=<1 0 0>}
      localJointOffset=<2 0 0>
     }]
    relativeSpaceBone {boneName=spine_02}
   }
   n29 = Anim Dynamics - Bone: dyn_blouse_side_02_r!(ComponentPose=@n28, Alpha=0.5, GravityOverride:Vector="0,0,-1000", ExternalForce="0,0,0") -> (Pose) {
    angularDampingOverride=0.999
    angularSpringConstant=50
    bAngularSpring=true
    bOverrideAngularDamping=true
    bUseGravityOverride=true
    gravityScale=0.35
    simulationSpace=RootRelative
    alphaBoolBlend {blendOption=Linear}
    boundBone {boneName=dyn_blouse_side_02_r}
    componentLinearAccScale=<0.4 0.4 0.4>
    componentLinearVelScale=<0.4 0.4 0.4>
    gravityOverride {z=-1000}
    physicsBodyDefinitions [{
      boundBone {boneName=dyn_blouse_side_02_r}
      boxExtents=<16 16 16>
      constraintSetup {angularLimitsMax=<5 0 1>, angularLimitsMin=<-5 -15 -1>, angularTarget=<1 0 0>}
      localJointOffset=<2 0 0>
     }]
    relativeSpaceBone {boneName=dyn_blouse_side_01_r}
   }
   n30 = Anim Dynamics - Bone: dyn_blouse_side_03_r!(ComponentPose=@n29, Alpha=0.5, GravityOverride:Vector="0,0,-1000", ExternalForce="0,0,0") -> (Pose) {
    angularDampingOverride=0.999
    angularSpringConstant=50
    bAngularSpring=true
    bOverrideAngularDamping=true
    bUseGravityOverride=true
    gravityScale=0.35
    simulationSpace=RootRelative
    alphaBoolBlend {blendOption=Linear}
    boundBone {boneName=dyn_blouse_side_03_r}
    componentLinearAccScale=<0.4 0.4 0.4>
    componentLinearVelScale=<0.4 0.4 0.4>
    gravityOverride {z=-1000}
    physicsBodyDefinitions [{
      boundBone {boneName=dyn_blouse_side_03_r}
      boxExtents=<16 16 16>
      constraintSetup {angularLimitsMax=<5 0 1>, angularLimitsMin=<-5 -15 -1>, angularTarget=<1 0 0>}
      localJointOffset=<2 0 0>
     }]
    relativeSpaceBone {boneName=dyn_blouse_side_02_r}
   }
   n31 = Anim Dynamics - Bone: dyn_blouse_side_04_r!(ComponentPose=@n30, Alpha=0.5, GravityOverride:Vector="0,0,-1000", ExternalForce="0,0,0") -> (Pose) {
    angularDampingOverride=0.999
    angularSpringConstant=50
    bAngularSpring=true
    bOverrideAngularDamping=true
    bUseGravityOverride=true
    gravityScale=0.35
    simulationSpace=RootRelative
    alphaBoolBlend {blendOption=Linear}
    boundBone {boneName=dyn_blouse_side_04_r}
    componentLinearAccScale=<0.4 0.4 0.4>
    componentLinearVelScale=<0.4 0.4 0.4>
    gravityOverride {z=-1000}
    physicsBodyDefinitions [{
      boundBone {boneName=dyn_blouse_side_04_r}
      boxExtents=<16 16 16>
      constraintSetup {angularLimitsMax=<5 0 1>, angularLimitsMin=<-5 -15 -1>, angularTarget=<1 0 0>}
      localJointOffset=<2 0 0>
     }]
    relativeSpaceBone {boneName=dyn_blouse_side_03_r}
   }
   n32 = Anim Dynamics - Bone: dyn_blouse_fr_01_l!(ComponentPose=@n31, Alpha=0.5, GravityOverride:Vector="0,0,-980", ExternalForce="0,0,0") -> (Pose) {
    angularDampingOverride=0.999
    angularSpringConstant=50
    bAngularSpring=true
    bUseGravityOverride=true
    gravityScale=0.35
    simulationSpace=RootRelative
    alphaBoolBlend {blendOption=Linear}
    boundBone {boneName=dyn_blouse_fr_01_l}
    componentLinearAccScale=<0.4 0.4 0.4>
    componentLinearVelScale=<0.4 0.4 0.4>
    gravityOverride {z=-980}
    physicsBodyDefinitions [{
      boundBone {boneName=dyn_blouse_fr_01_l}
      boxExtents=<16 16 16>
      constraintSetup {angularLimitsMax=<0 0 1>, angularLimitsMin=<0 -15 -1>, angularTarget=<1 0 0>}
      localJointOffset=<2 0 0>
     }]
    relativeSpaceBone {boneName=spine_02}
   }
   n33 = Anim Dynamics - Bone: dyn_blouse_fr_02_l!(ComponentPose=@n32, Alpha=0.5, GravityOverride:Vector="0,0,-1000", ExternalForce="0,0,0") -> (Pose) {
    angularDampingOverride=0.999
    angularSpringConstant=50
    bAngularSpring=true
    bOverrideAngularDamping=true
    bUseGravityOverride=true
    gravityScale=0.35
    simulationSpace=RootRelative
    alphaBoolBlend {blendOption=Linear}
    boundBone {boneName=dyn_blouse_fr_02_l}
    componentLinearAccScale=<0.4 0.4 0.4>
    componentLinearVelScale=<0.4 0.4 0.4>
    gravityOverride {z=-1000}
    physicsBodyDefinitions [{
      boundBone {boneName=dyn_blouse_fr_02_l}
      boxExtents=<16 16 16>
      constraintSetup {angularLimitsMax=<0 0 1>, angularLimitsMin=<0 -15 -1>, angularTarget=<1 0 0>}
      localJointOffset=<2 0 0>
     }]
    relativeSpaceBone {boneName=dyn_blouse_fr_01_l}
   }
   n34 = Anim Dynamics - Bone: dyn_blouse_fr_03_l!(ComponentPose=@n33, Alpha=0.5, GravityOverride:Vector="0,0,-1000", ExternalForce="0,0,0") -> (Pose) {
    angularDampingOverride=0.999
    angularSpringConstant=50
    bAngularSpring=true
    bOverrideAngularDamping=true
    bUseGravityOverride=true
    gravityScale=0.35
    simulationSpace=RootRelative
    alphaBoolBlend {blendOption=Linear}
    boundBone {boneName=dyn_blouse_fr_03_l}
    componentLinearAccScale=<0.4 0.4 0.4>
    componentLinearVelScale=<0.4 0.4 0.4>
    gravityOverride {z=-1000}
    physicsBodyDefinitions [{
      boundBone {boneName=dyn_blouse_fr_03_l}
      boxExtents=<16 16 16>
      constraintSetup {angularLimitsMax=<0 0 1>, angularLimitsMin=<0 -15 -1>, angularTarget=<1 0 0>}
      localJointOffset=<2 0 0>
     }]
    relativeSpaceBone {boneName=dyn_blouse_fr_02_l}
   }
   n35 = Anim Dynamics - Bone: dyn_blouse_fr_04_l!(ComponentPose=@n34, Alpha=0.5, GravityOverride:Vector="0,0,-1000", ExternalForce="0,0,0") -> (Pose) {
    angularDampingOverride=0.999
    angularSpringConstant=50
    bAngularSpring=true
    bOverrideAngularDamping=true
    bUseGravityOverride=true
    gravityScale=0.35
    simulationSpace=RootRelative
    alphaBoolBlend {blendOption=Linear}
    boundBone {boneName=dyn_blouse_fr_04_l}
    componentLinearAccScale=<0.4 0.4 0.4>
    componentLinearVelScale=<0.4 0.4 0.4>
    gravityOverride {z=-1000}
    physicsBodyDefinitions [{
      boundBone {boneName=dyn_blouse_fr_04_l}
      boxExtents=<16 16 16>
      constraintSetup {angularLimitsMax=<0 0 1>, angularLimitsMin=<0 -15 -1>, angularTarget=<1 0 0>}
      localJointOffset=<2 0 0>
     }]
    relativeSpaceBone {boneName=dyn_blouse_fr_03_l}
   }
   n36 = Anim Dynamics - Bone: dyn_blouse_side_01_l!(ComponentPose=@n35, Alpha=0.5, GravityOverride:Vector="0,0,-980", ExternalForce="0,0,0") -> (Pose) {
    angularDampingOverride=0.999
    angularSpringConstant=50
    bAngularSpring=true
    bUseGravityOverride=true
    gravityScale=0.35
    simulationSpace=RootRelative
    alphaBoolBlend {blendOption=Linear}
    boundBone {boneName=dyn_blouse_side_01_l}
    componentLinearAccScale=<0.4 0.4 0.4>
    componentLinearVelScale=<0.4 0.4 0.4>
    gravityOverride {z=-980}
    physicsBodyDefinitions [{
      boundBone {boneName=dyn_blouse_side_01_l}
      boxExtents=<16 16 16>
      constraintSetup {angularLimitsMax=<5 0 1>, angularLimitsMin=<-5 -15 -1>, angularTarget=<1 0 0>}
      localJointOffset=<2 0 0>
     }]
    relativeSpaceBone {boneName=spine_02}
   }
   n37 = Anim Dynamics - Bone: dyn_blouse_side_02_l!(ComponentPose=@n36, Alpha=0.5, GravityOverride:Vector="0,0,-1000", ExternalForce="0,0,0") -> (Pose) {
    angularDampingOverride=0.999
    angularSpringConstant=50
    bAngularSpring=true
    bOverrideAngularDamping=true
    bUseGravityOverride=true
    gravityScale=0.35
    simulationSpace=RootRelative
    alphaBoolBlend {blendOption=Linear}
    boundBone {boneName=dyn_blouse_side_02_l}
    componentLinearAccScale=<0.4 0.4 0.4>
    componentLinearVelScale=<0.4 0.4 0.4>
    gravityOverride {z=-1000}
    physicsBodyDefinitions [{
      boundBone {boneName=dyn_blouse_side_02_l}
      boxExtents=<16 16 16>
      constraintSetup {angularLimitsMax=<5 0 1>, angularLimitsMin=<-5 -15 -1>, angularTarget=<1 0 0>}
      localJointOffset=<2 0 0>
     }]
    relativeSpaceBone {boneName=dyn_blouse_side_01_l}
   }
   n38 = Anim Dynamics - Bone: dyn_blouse_side_03_l!(ComponentPose=@n37, Alpha=0.5, GravityOverride:Vector="0,0,-1000", ExternalForce="0,0,0") -> (Pose) {
    angularDampingOverride=0.999
    angularSpringConstant=50
    bAngularSpring=true
    bOverrideAngularDamping=true
    bUseGravityOverride=true
    gravityScale=0.35
    simulationSpace=RootRelative
    alphaBoolBlend {blendOption=Linear}
    boundBone {boneName=dyn_blouse_side_03_l}
    componentLinearAccScale=<0.4 0.4 0.4>
    componentLinearVelScale=<0.4 0.4 0.4>
    gravityOverride {z=-1000}
    physicsBodyDefinitions [{
      boundBone {boneName=dyn_blouse_side_03_l}
      boxExtents=<16 16 16>
      constraintSetup {angularLimitsMax=<5 0 1>, angularLimitsMin=<-5 -15 -1>, angularTarget=<1 0 0>}
      localJointOffset=<2 0 0>
     }]
    relativeSpaceBone {boneName=dyn_blouse_side_02_l}
   }
   n39 = Anim Dynamics - Bone: dyn_blouse_side_04_l!(ComponentPose=@n38, Alpha=0.5, GravityOverride:Vector="0,0,-1000", ExternalForce="0,0,0") -> (Pose) {
    angularDampingOverride=0.999
    angularSpringConstant=50
    bAngularSpring=true
    bOverrideAngularDamping=true
    bUseGravityOverride=true
    gravityScale=0.35
    simulationSpace=RootRelative
    alphaBoolBlend {blendOption=Linear}
    boundBone {boneName=dyn_blouse_side_04_l}
    componentLinearAccScale=<0.4 0.4 0.4>
    componentLinearVelScale=<0.4 0.4 0.4>
    gravityOverride {z=-1000}
    physicsBodyDefinitions [{
      boundBone {boneName=dyn_blouse_side_04_l}
      boxExtents=<16 16 16>
      constraintSetup {angularLimitsMax=<5 0 1>, angularLimitsMin=<-5 -15 -1>, angularTarget=<1 0 0>}
      localJointOffset=<2 0 0>
     }]
    relativeSpaceBone {boneName=dyn_blouse_side_03_l}
   }
   n40 = "Constraint - dyn_blouse_bk_04_l (RR)"!(ComponentPose=@n39, Alpha=1, ConstraintWeights_0=0.5, ConstraintWeights_1:float=0.5) -> (Pose) {
    alphaBoolBlend {blendOption=Linear}
    boneToModify {boneName=dyn_blouse_bk_04_l}
    constraintSetup [{transformType=Rotation, targetBone {boneName=dyn_blouse_04_c}}, {transformType=Rotation, targetBone {boneName=dyn_blouse_side_04_l}}]
    constraintWeights [1,1]
   }
   n41 = "Constraint - dyn_blouse_bk_04_r (RR)"!(ComponentPose=@n40, Alpha=1, ConstraintWeights_0=0.5, ConstraintWeights_1:float=0.5) -> (Pose) {
    alphaBoolBlend {blendOption=Linear}
    boneToModify {boneName=dyn_blouse_bk_04_r}
    constraintSetup [{transformType=Rotation, targetBone {boneName=dyn_blouse_04_c}}, {transformType=Rotation, targetBone {boneName=dyn_blouse_side_04_r}}]
    constraintWeights [1,1]
   }
   Output Pose(Result=Component To Local(ComponentPose=@n41)) {layerGroup=None}
   n42 = RigidBody(ComponentPose=Local To Component(LocalPose=Use cached pose 'Pre sim'()), OverrideWorldGravity="0,0,0") {
    alphaInputType=Bool
    bClampLinearTranslationLimitToRefPose=true
    bOverrideWorldGravity=true
    overridePhysicsAsset=&/F_MED_SugarBubble_Physics
    simulationSpace=BaseBoneSpace
    alphaBoolBlend {blendOption=Linear}
    baseBoneRef {boneName=spine_02}
    componentAppliedLinearAccClamp=<1200 1200 1200>
    componentLinearAccScale=<0.25 0.25 0.25>
    componentLinearVelScale=<0.3 0.3 0.3>
    simSpaceSettings {
     maxAngularAcceleration=100
     maxAngularVelocity=2
     maxLinearAcceleration=100
     maxLinearVelocity=160
     velocityScaleZ=0
     worldAlpha=0.75
     externalLinearDragV=<1 1 1>
    }
   }
   n43 = RigidBody(ComponentPose=Local To Component(LocalPose=Use cached pose 'Pre sim'()), OverrideWorldGravity="0,0,0") {
    alphaInputType=Bool
    bClampLinearTranslationLimitToRefPose=true
    bOverrideWorldGravity=true
    overridePhysicsAsset=&/F_MED_SugarBubble_Skydive_Physics
    simulationSpace=BaseBoneSpace
    alphaBoolBlend {blendOption=Linear}
    baseBoneRef {boneName=spine_02}
    componentAppliedLinearAccClamp=<1200 1200 1200>
    componentLinearAccScale=<0.25 0.25 0.25>
    componentLinearVelScale=<0.3 0.3 0.3>
    simSpaceSettings {
     maxAngularAcceleration=100
     maxAngularVelocity=2
     maxLinearAcceleration=100
     maxLinearVelocity=160
     velocityScaleZ=0
     worldAlpha=0.75
     externalLinearDragV=<1 1 1>
    }
   }
   n44 = Control Rig!(
    Source=Blend Poses by bool(BlendPose_0=Component To Local(ComponentPose=@n43), BlendPose_1=Component To Local(ComponentPose=@n42), bActiveValue=Get bIsSkyDiving()) {blendType=Linear}
    Alpha
    Is baseball sliding:bool=Get bIsBaseballSliding()
    Is DBNO On:bool=Get bIsDBNO()
   ) -> (Pose) {
    alphaBoolBlend {blendOption=Linear}
    controlRigAssetReference {blueprintRigClass=&/F_MED_SugarBubble_AS_CtrlRig.F_MED_SugarBubble_AS_CtrlRig_C}
   }
   Save cached pose 'frontend pass'(Pose=Component To Local(ComponentPose=Local To Component(LocalPose=@n44))) {cacheName=frontend pass}
   n45 = Control Rig(Source=Copy Pose From Mesh() {bUseAttachedParent=true}, Is baseball sliding=Get bIsBaseballSliding(), Is DBNO On=Get bIsDBNO()) {
    alphaBoolBlend {blendOption=Linear}
    controlRigAssetReference {blueprintRigClass=&/F_MED_SugarBubble_CtrlRig.F_MED_SugarBubble_CtrlRig_C}
   }
   n46 = "Transform (Modify) Bone - Bone: Alt_jacket_scale"!(ComponentPose=Local To Component(LocalPose=@n45), bAlphaBoolEnabled:bool=True, Translation="0,0,0", Rotation="0,0,0", Scale="0.1,0.1,0.1") -> (Pose) {
    alphaInputType=Bool
    scaleMode=BMM_Additive
    translationMode=BMM_Additive
    alphaBoolBlend {blendOption=Linear}
    boneToModify {boneName=Alt_jacket_scale}
    scale=<0.1 0.1 0.1>
   }
   n47 = "Transform (Modify) Bone - Bone: dyn_collar_01"!(
    ComponentPose=@n46
    bAlphaBoolEnabled:bool=Get IsWearingBackbling() -> (IsWearingBackbling:bool)
    Translation="-3,3,0"
    Rotation="0,0,0"
    Scale="1,1,1"
   ) -> (Pose) {
    alphaInputType=Bool
    translationMode=BMM_Additive
    alphaBoolBlend {blendOption=Linear}
    boneToModify {boneName=dyn_collar_01}
    translation=<-3 3>
   }
   n48 = Anim Dynamics - Bone: dyn_laces_01_l!(ComponentPose=@n47, Alpha=1, GravityOverride:Vector="0,0,0", ExternalForce="0,0,0") -> (Pose) {
    angularDampingOverride=0.9
    angularSpringConstant=1000
    bAngularSpring=true
    bGravityOverrideInSimSpace=true
    bOverrideAngularDamping=true
    bOverrideLinearDamping=true
    bUseGravityOverride=true
    simulationSpace=RootRelative
    alphaBoolBlend {blendOption=Linear}
    boundBone {boneName=dyn_laces_01_l}
    componentLinearAccScale=<1 1 1>
    componentLinearVelScale=<1 1 1>
    physicsBodyDefinitions [{
      boundBone {boneName=dyn_laces_01_l}
      boxExtents=<8 8 8>
      constraintSetup {angularLimitsMax=<0 10 20>, angularLimitsMin=<0 5 -20>, angularTarget=<0 0 0>}
      localJointOffset=<0.3 0 0>
     }]
    relativeSpaceBone {boneName=calf_l}
   }
   n49 = Anim Dynamics - Bone: dyn_laces_02_l!(ComponentPose=@n48, Alpha=1, GravityOverride:Vector="0,0,0", ExternalForce="0,0,0") -> (Pose) {
    angularDampingOverride=0.9
    angularSpringConstant=1000
    bAngularSpring=true
    bGravityOverrideInSimSpace=true
    bOverrideAngularDamping=true
    bOverrideLinearDamping=true
    bUseGravityOverride=true
    simulationSpace=RootRelative
    alphaBoolBlend {blendOption=Linear}
    boundBone {boneName=dyn_laces_02_l}
    componentLinearAccScale=<1 1 1>
    componentLinearVelScale=<1 1 1>
    physicsBodyDefinitions [{
      boundBone {boneName=dyn_laces_02_l}
      boxExtents=<8 8 8>
      constraintSetup {angularLimitsMax=<30 -5 20>, angularLimitsMin=<-20 -15 -20>, angularTarget=<0 0 0>}
      localJointOffset=<0.3 0 0>
     }]
    relativeSpaceBone {boneName=calf_l}
   }
   n50 = Anim Dynamics - Bone: dyn_laces_01_r(ComponentPose=@n49, Alpha=1, ExternalForce="0,0,0") {
    angularDampingOverride=0.9995
    angularSpringConstant=2500
    bAngularSpring=true
    bGravityOverrideInSimSpace=true
    bOverrideAngularDamping=true
    bUseGravityOverride=true
    alphaBoolBlend {blendOption=Linear}
    boundBone {boneName=dyn_laces_01_r}
    componentAppliedLinearAccClamp=<3000 3000 3000>
    componentLinearAccScale=<0.3 0.6 0.3>
    componentLinearVelScale=<0.3 0.3 0.3>
    gravityOverride {z=-980}
    physicsBodyDefinitions [{
      boundBone {boneName=dyn_laces_01_r}
      boxExtents=<10 10 10>
      constraintSetup {angularTargetAxis=AxisZ, angularLimitsMax=<0 0 0>, angularLimitsMin=<0 -40 0>, angularTarget=<0 0 1>}
      localJointOffset=<0 0 0>
     }]
   }
   n51 = Anim Dynamics - Bone: dyn_laces_02_r(ComponentPose=@n50, Alpha=1, ExternalForce="0,0,0") {
    angularDampingOverride=0.9995
    angularSpringConstant=2500
    bAngularSpring=true
    bGravityOverrideInSimSpace=true
    bOverrideAngularDamping=true
    bUseGravityOverride=true
    alphaBoolBlend {blendOption=Linear}
    boundBone {boneName=dyn_laces_02_r}
    componentAppliedLinearAccClamp=<3000 3000 3000>
    componentLinearAccScale=<0.3 0.6 0.3>
    componentLinearVelScale=<0.3 0.3 0.3>
    gravityOverride {z=-980}
    physicsBodyDefinitions [{
      boundBone {boneName=dyn_laces_02_r}
      boxExtents=<10 10 10>
      constraintSetup {angularTargetAxis=AxisZ, angularLimitsMax=<0 50 0>, angularLimitsMin=<0 0 0>, angularTarget=<0 0 1>}
      localJointOffset=<0 0 0>
     }]
   }
   n52 = Anim Dynamics - Bone: dyn_collar_01(ComponentPose=@n51, Alpha=0.6, ExternalForce="0,0,0") {
    angularDampingOverride=0.9995
    angularSpringConstant=2000
    bAngularSpring=true
    bGravityOverrideInSimSpace=true
    bLinearSpring=true
    bOverrideAngularDamping=true
    bOverrideLinearDamping=true
    bUseGravityOverride=true
    linearDampingOverride=0.9995
    linearSpringConstant=50
    simulationSpace=RootRelative
    alphaBoolBlend {blendOption=Linear}
    boundBone {boneName=dyn_collar_01}
    componentAppliedLinearAccClamp=<1000 1000 1000>
    gravityOverride {z=-600}
    physicsBodyDefinitions [{
      boundBone {boneName=dyn_collar_01}
      boxExtents=<10 10 10>
      constraintSetup {angularLimitsMax=<0 0 0>, angularLimitsMin=<0 0 -20>, angularTarget=<1 0 0>, linearAxesMax=<4 0 4>, linearAxesMin=<-3 0 -3>}
      localJointOffset=<0 0 0>
     }]
   }
   n53 = Anim Dynamics - Bone: dyn_collar_02(ComponentPose=@n52, Alpha=0.6, ExternalForce="0,0,0") {
    angularDampingOverride=0.9995
    angularSpringConstant=2000
    bAngularSpring=true
    bGravityOverrideInSimSpace=true
    bLinearSpring=true
    bOverrideAngularDamping=true
    bOverrideLinearDamping=true
    bUseGravityOverride=true
    linearDampingOverride=0.9995
    linearSpringConstant=50
    simulationSpace=RootRelative
    alphaBoolBlend {blendOption=Linear}
    boundBone {boneName=dyn_collar_02}
    componentAppliedLinearAccClamp=<1000 1000 1000>
    gravityOverride {z=-600}
    physicsBodyDefinitions [{
      boundBone {boneName=dyn_collar_02}
      boxExtents=<10 10 10>
      constraintSetup {angularLimitsMax=<0 0 0>, angularLimitsMin=<0 0 -20>, angularTarget=<1 0 0>, linearAxesMax=<4 0 4>, linearAxesMin=<-3 0 -3>}
      localJointOffset=<0 0 0>
     }]
   }
   Save cached pose 'Pre sim'(Pose=Component To Local(ComponentPose=@n53)) {cacheName=Pre sim}
  ]}
 EventGraph {nodes [
   n1 = Gravity Override State Function DEPRECATED(
    CharacterPartAnimInstance=Self-Reference()
    GameTimeInSeconds=GetGameTimeInSeconds()
    GravityOverrideParams=Get Gravity Override Params()
    DebugJointName=Pelvis
   )
   > Set Gravity Override(Gravity Override=@n1)
   Event BlueprintUpdateAnimation()
   > n2 = Sequence!(execute) -> (then_0, then_1, then_2:Exec, then_3:Exec)
   > n3 = Gravity Override State Function DEPRECATED(CharacterPartAnimInstance=Self-Reference(), GameTimeInSeconds=GetGameTimeInSeconds(), GravityOverrideParams=Get Gravity Override Params())
   > Set Gravity Override(Gravity Override=@n3)
   n4 = GetOwnerSkeletalMeshComponent()
   Set Spine_3 rotation(
    execute=@n2.then_1
    Spine_3 rotation:double=MapRangeClamped(
     Value=GetDeltaTransformFromRefPose!(self=@n4, BoneName=spine_04, BaseName=spine_02) -> (
      ReturnValue_Location:Vector
      ReturnValue_Rotation_Roll:float
      ReturnValue_Rotation_Pitch:float
      ReturnValue_Rotation_Yaw:float
      ReturnValue_Scale:Vector
     ).ReturnValue_Rotation_Yaw
     InRangeA=-1
     InRangeB=0
     OutRangeA=-0.2
    )
   ) -> (Output_Get:double)
   > Set is skydive(is skydive:double=SelectFloat(A=0.1, B=0.5, bPickA=Get bIsSkyDiving())) -> (Output_Get:double)
   Branch(execute=@n2.then_3, Condition=Get bIsInFrontEnd())
   > Set FrontEnd Idle Bypass(FrontEnd Idle Bypass:bool=NOT Boolean(A=Get bIsPlayingEmote())) -> (Output_Get:bool)
   n5 = CheckForBackbling()
   n6 = Set IsWearingBackbling(execute=@n2.then_2, IsWearingBackbling:bool=@n5) -> (Output_Get:bool)
   > Set Is Backbling A Cape(Is Backbling A Cape:bool=@n5.bIsBackblingACape) -> (Output_Get:bool)
   > Set NotBackbling(NotBackbling:double="To Float (Boolean)"(InBool=NOT Boolean(A=@n5))) -> (Output_Get:double)
   > Set is Wear a Cape(is Wear a Cape:bool=AND Boolean(A=@n5, B=@n5.bIsBackblingLong)) -> (Output_Get:bool)
   > n7 = Branch(Condition=@n6)
   > n8 = Branch(Condition=@n5.bIsBackblingLong)
   Set IsShortBackblingOnAlpha(execute=[@n8.then, @n7.else], IsShortBackblingOnAlpha:double=0) -> (Output_Get:double)
   Set IsShortBackblingOnAlpha(execute=@n8.else, IsShortBackblingOnAlpha=1)
  ]}
]
variables [
 FrontEnd Idle Bypass {type=boolean}
 Gravity Override {type=Vector}
 Gravity Override Params {type=GravityOverrideParamsStruct}
 Is Backbling A Cape {type=boolean}
 Is Rigid Body {type=boolean}
 is skydive {type=number}
 is Wear a Cape {type=boolean}
 Is Wearing Backbling {type=boolean}
 isPlayingFrontendIdle {type=number}
 IsShortBackblingOnAlpha {type=number}
 IsWearingBackbling {type=boolean}
 NewVar {type=number}
 NotBackbling {type=number}
 Override World Gravity {type=Vector}
 Spine_3 rotation {type=number}
]
