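# Physics asset for the F_MED_SugarBubble player body, providing capsule colliders for its cape, as the name indicates.
# Capsules cover pelvis, spine, clavicles, head and feet, while neck, arm, thigh and calf bodies have no shapes.
# Seven four-bone blouse chains hanging from spine_04 and two strap chains on the calves also carry capsules.
# Constraints follow the bone hierarchy with all swing and twist motion free.
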PhysicsAsset /Game/Characters/Player/Female/Medium/Bodies/F_MED_SugarBubble/Meshes/F_MED_SugarBubble_CapeColliders
bodies {
 calf_l []
 calf_r []
 clavicle_l [{shapeType=Capsule, length=6.5623, radius=7.5194, center=<3.9566 -1.0862 -2.1127>, rotation=<rot -77.5566 177.3915 7.7314>}]
 clavicle_r [{shapeType=Capsule, length=6.5623, radius=7.5194, center=<-3.9566 1.0862 2.1127>, rotation=<rot -77.5566 177.3915 -172.2686>}]
 dyn_blouse_01_c [{shapeType=Capsule, length=7, radius=6, shapeName=dyn_blouse_01_c_capsule, center=<0 0 -2.7243>, rotation=<rot 0 0 90>}]
 dyn_blouse_02_c [{shapeType=Capsule, length=6, radius=8.6985, shapeName=dyn_blouse_02_c_capsule, center=<4.1077 0 -2.9673>, rotation=<rot 0 0 90>}]
 dyn_blouse_03_c [{shapeType=Capsule, length=6, radius=8.2186, shapeName=dyn_blouse_03_c_capsule, center=<3.9221 0 -2.7243>, rotation=<rot 0 0 90>}]
 dyn_blouse_04_c [{shapeType=Capsule, length=15.8204, radius=7, shapeName=dyn_blouse_04_c_capsule, center=<3.4081 0 -3.6572>, rotation=<rot 0 0 90>}]
 dyn_blouse_bk_01_l [{shapeType=Capsule, length=7, radius=6, shapeName=dyn_blouse_bk_01_l_capsule, center=<0 0 -2.7243>, rotation=<rot 0 0 90>}]
 dyn_blouse_bk_01_r [{shapeType=Capsule, length=7, radius=6, shapeName=dyn_blouse_bk_01_r_capsule, center=<0 0 -2.7243>, rotation=<rot 0 0 90>}]
 dyn_blouse_bk_02_l [{shapeType=Capsule, length=6, radius=8.6985, shapeName=dyn_blouse_bk_02_l_capsule, center=<5.302 0 -4.9431>, rotation=<rot 0 0 90>}]
 dyn_blouse_bk_02_r [{shapeType=Capsule, length=6, radius=7, shapeName=dyn_blouse_bk_02_r_capsule, center=<3.4912 0 -2.7243>, rotation=<rot 0 0 90>}]
 dyn_blouse_bk_03_l [{shapeType=Capsule, length=6, radius=7.9241, shapeName=dyn_blouse_bk_03_l_capsule, center=<4.6454 0 -2.7243>, rotation=<rot 0 0 90>}]
 dyn_blouse_bk_03_r [{shapeType=Capsule, length=6, radius=9.6439, shapeName=dyn_blouse_bk_03_r_capsule, center=<3.6394 0 -5.0701>, rotation=<rot 0 0 90>}]
 dyn_blouse_bk_04_l [{
   shapeType=Capsule
   length=9.0988
   radius=7
   shapeName=dyn_blouse_bk_04_l_capsule
   center=<3.4319 0.1079 -2.7816>
   rotation=<rot -0.0001 0 86.6554>
  }]
 dyn_blouse_bk_04_r [{
   shapeType=Capsule
   length=9.6582
   radius=7
   shapeName=dyn_blouse_bk_04_r_capsule
   center=<2.7239 -0.2477 -3.8836>
   rotation=<rot 0 0 102.0599>
  }]
 dyn_blouse_fr_01_l [{shapeType=Capsule, length=7, radius=6, shapeName=dyn_blouse_fr_01_l_capsule, center=<0 0 -2.7243>, rotation=<rot 0 0 90>}]
 dyn_blouse_fr_01_r [{shapeType=Capsule, length=7, radius=6, shapeName=dyn_blouse_fr_01_r_capsule, center=<0 0 -2.7243>, rotation=<rot 0 0 90>}]
 dyn_blouse_fr_02_l [{shapeType=Capsule, length=7, radius=7.6985, shapeName=dyn_blouse_fr_02_l_capsule, center=<4.4499 0 -3.7315>, rotation=<rot 0 0 90>}]
 dyn_blouse_fr_02_r [{shapeType=Capsule, length=7, radius=7.6985, shapeName=dyn_blouse_fr_02_r_capsule, center=<2.8527 0 -4.429>, rotation=<rot 0 0 90>}]
 dyn_blouse_fr_03_l [{shapeType=Capsule, length=7, radius=7.1538, shapeName=dyn_blouse_fr_03_l_capsule, center=<4.7914 0 -2.7243>, rotation=<rot 0 0 90>}]
 dyn_blouse_fr_03_r [{shapeType=Capsule, length=7, radius=6, shapeName=dyn_blouse_fr_03_r_capsule, center=<1.6029 0 -2.7243>, rotation=<rot 0 0 90>}]
 dyn_blouse_fr_04_l [{
   shapeType=Capsule
   length=11.2665
   radius=6
   shapeName=dyn_blouse_fr_04_l_capsule
   center=<2.2464 0 -2.7243>
   rotation=<rot 0.0001 0.0001 78.6142>
  }]
 dyn_blouse_fr_04_r [{shapeType=Capsule, length=8.1744, radius=6, shapeName=dyn_blouse_fr_04_r_capsule, center=<2.2464 0 -4.9445>, rotation=<rot 0 0 90>}]
 dyn_blouse_side_01_l [{shapeType=Capsule, length=7, radius=6, shapeName=dyn_blouse_side_01_l_capsule, center=<0 0 -2.7243>, rotation=<rot 0 0 90>}]
 dyn_blouse_side_01_r [{shapeType=Capsule, length=7, radius=6, shapeName=dyn_blouse_side_01_r_capsule, center=<0 0 -2.7243>, rotation=<rot 0 0 90>}]
 dyn_blouse_side_02_l [{shapeType=Capsule, length=5, radius=9.6985, shapeName=dyn_blouse_side_02_l_capsule, center=<2.8527 0 -5.4515>, rotation=<rot 0 0 90>}]
 dyn_blouse_side_02_r [{shapeType=Capsule, length=5, radius=9.6985, shapeName=dyn_blouse_side_02_r_capsule, center=<2.8527 0 -4.429>, rotation=<rot 0 0 90>}]
 dyn_blouse_side_03_l [{
   shapeType=Capsule
   length=5
   radius=8
   shapeName=dyn_blouse_side_03_l_capsule
   center=<2.4664 -2.4032 -2.2485>
   rotation=<rot 0 0.0002 79.1102>
  }]
 dyn_blouse_side_03_r [{shapeType=Capsule, length=5, radius=9.5437, shapeName=dyn_blouse_side_03_r_capsule, center=<3.2025 0 -2.2379>, rotation=<rot 0 0 90>}]
 dyn_blouse_side_04_l [{
   shapeType=Capsule
   length=7.3873
   radius=8
   shapeName=dyn_blouse_side_04_l_capsule
   center=<2.2464 0 -2.7243>
   rotation=<rot 0 0.0001 79.0224>
  }]
 dyn_blouse_side_04_r [{
   shapeType=Capsule
   length=8.3556
   radius=8
   shapeName=dyn_blouse_side_04_r_capsule
   center=<2.2464 -0.4356 -3.5365>
   rotation=<rot 0 -0.0002 96.6264>
  }]
 dyn_strap_l_01 [{shapeType=Capsule, length=7.5, radius=6, shapeName=dyn_strap_l_01_capsule, center=<5.6589 0.0502 -0.1001>, rotation=<rot 90 0 0>}]
 dyn_strap_l_02 [{shapeType=Capsule, length=7.5, radius=6, shapeName=dyn_strap_l_02_capsule, center=<5.4447 0.1792 -0.1188>, rotation=<rot 90 0 0>}]
 dyn_strap_l_03 [{shapeType=Capsule, length=7.5, radius=6, shapeName=dyn_strap_l_03_capsule, center=<4.7181 0.0843 -0.0082>, rotation=<rot 90 0 0>}]
 dyn_strap_l_04 [{shapeType=Capsule, length=7.5, radius=6, shapeName=dyn_strap_l_04_capsule, center=<3.4703 -0.0343 -0.046>, rotation=<rot 90 0 0>}]
 dyn_strap_r_01 [{shapeType=Capsule, length=7.5, radius=6, shapeName=dyn_strap_r_01_capsule, center=<4.4691 0.0171 0.1334>, rotation=<rot 90 0 0>}]
 dyn_strap_r_02 [{shapeType=Capsule, length=7.5, radius=6, shapeName=dyn_strap_r_02_capsule, center=<5.9529 0.0613 -0.0092>, rotation=<rot 90 0 0>}]
 dyn_strap_r_03 [{shapeType=Capsule, length=7.5, radius=6, shapeName=dyn_strap_r_03_capsule, center=<3.5753 -0.0281 -0.107>, rotation=<rot 90 0 0>}]
 dyn_strap_r_04 [{shapeType=Capsule, length=7.5, radius=6, shapeName=dyn_strap_r_04_capsule, center=<3.3805 -0.1825 -0.04>, rotation=<rot 90 0 0>}]
 foot_l [{shapeType=Capsule, length=11.0765, radius=8.0344, center=<-4.1811 6.0027 -0.338>, rotation=<rot 0.0388 1.4701 -89.3982>}]
 foot_r [{shapeType=Capsule, length=11.0765, radius=8.0344, center=<4.1811 -6.0027 0.338>, rotation=<rot 0.0388 1.4701 90.6018>}]
 head [{shapeType=Capsule, length=0, radius=18, center=<9.0697 0.0746 0>, rotation=<rot 0 -90 0>}]
 lowerarm_l []
 lowerarm_r []
 neck_02 []
 pelvis [{shapeType=Capsule, length=9.6042, radius=11.1441, center=<-2.4471 -2.5289 -0.1981>}]
 spine_01 [{shapeType=Capsule, length=4.3888, radius=14.3638, center=<-8.9021 -0.8973 0>}]
 spine_02 [
  {shapeType=Capsule, length=2.5219, radius=9.7317, center=<-2.6885 -1.2641 0>}
  {shapeType=Capsule, length=20.6875, radius=9.3125, center=<-1.8317 1.9362 0>, rotation=<rot 90 0 0>}
 ]
 spine_03 [
  {shapeType=Capsule, length=22.6863, radius=8.4854, center=<0.0574 -4.5026 0>}
  {shapeType=Capsule, length=20.7188, radius=9.2812, center=<-1.2163 0.549 0>, rotation=<rot 90 180 -175.0732>}
 ]
 spine_04 [
  {shapeType=Capsule, length=22.7003, radius=10.3916, center=<5.6234 -2.5877 0>}
  {shapeType=Capsule, length=10.0312, radius=9.2812, center=<-1.2326 1.308 0>, rotation=<rot 90 0 0>}
 ]
 spine_05 [{shapeType=Capsule, length=15.9252, radius=15.3571, center=<0.6969 -0.5282 0>}]
 thigh_l []
 thigh_r []
 upperarm_l []
 upperarm_r []
}
constraints [
 {
  bone1Name=spine_01
  bone2Name=pelvis
  swing1LimitDegrees=45
  swing1Motion=Free
  swing2LimitDegrees=45
  swing2Motion=Free
  twistLimitDegrees=45
  twistMotion=Free
 }
 {
  bone1Name=spine_02
  bone2Name=spine_01
  swing1LimitDegrees=45
  swing1Motion=Free
  swing2LimitDegrees=45
  swing2Motion=Free
  twistLimitDegrees=45
  twistMotion=Free
 }
 {
  bone1Name=spine_03
  bone2Name=spine_02
  swing1LimitDegrees=45
  swing1Motion=Free
  swing2LimitDegrees=45
  swing2Motion=Free
  twistLimitDegrees=45
  twistMotion=Free
 }
 {
  bone1Name=spine_04
  bone2Name=spine_03
  swing1LimitDegrees=45
  swing1Motion=Free
  swing2LimitDegrees=45
  swing2Motion=Free
  twistLimitDegrees=45
  twistMotion=Free
 }
 {
  bone1Name=spine_05
  bone2Name=spine_04
  swing1LimitDegrees=45
  swing1Motion=Free
  swing2LimitDegrees=45
  swing2Motion=Free
  twistLimitDegrees=45
  twistMotion=Free
 }
 {
  bone1Name=clavicle_l
  bone2Name=spine_05
  swing1LimitDegrees=45
  swing1Motion=Free
  swing2LimitDegrees=45
  swing2Motion=Free
  twistLimitDegrees=45
  twistMotion=Free
 }
 {
  bone1Name=upperarm_l
  bone2Name=clavicle_l
  swing1LimitDegrees=45
  swing1Motion=Free
  swing2LimitDegrees=45
  swing2Motion=Free
  twistLimitDegrees=45
  twistMotion=Free
 }
 {
  bone1Name=lowerarm_l
  bone2Name=upperarm_l
  swing1LimitDegrees=45
  swing1Motion=Free
  swing2LimitDegrees=45
  swing2Motion=Free
  twistLimitDegrees=45
  twistMotion=Free
 }
 {
  bone1Name=neck_02
  bone2Name=spine_05
  swing1LimitDegrees=45
  swing1Motion=Free
  swing2LimitDegrees=45
  swing2Motion=Free
  twistLimitDegrees=45
  twistMotion=Free
 }
 {
  bone1Name=head
  bone2Name=neck_02
  swing1LimitDegrees=45
  swing1Motion=Free
  swing2LimitDegrees=45
  swing2Motion=Free
  twistLimitDegrees=45
  twistMotion=Free
 }
 {
  bone1Name=calf_l
  bone2Name=thigh_l
  swing1LimitDegrees=45
  swing1Motion=Free
  swing2LimitDegrees=45
  swing2Motion=Free
  twistLimitDegrees=45
  twistMotion=Free
 }
 {
  bone1Name=thigh_l
  bone2Name=pelvis
  swing1LimitDegrees=45
  swing1Motion=Free
  swing2LimitDegrees=45
  swing2Motion=Free
  twistLimitDegrees=45
  twistMotion=Free
 }
 {
  bone1Name=foot_l
  bone2Name=calf_l
  swing1LimitDegrees=45
  swing1Motion=Free
  swing2LimitDegrees=45
  swing2Motion=Free
  twistLimitDegrees=45
  twistMotion=Free
 }
 {
  bone1Name=lowerarm_r
  bone2Name=upperarm_r
  swing1LimitDegrees=45
  swing1Motion=Free
  swing2LimitDegrees=45
  swing2Motion=Free
  twistLimitDegrees=45
  twistMotion=Free
 }
 {
  bone1Name=upperarm_r
  bone2Name=clavicle_r
  swing1LimitDegrees=45
  swing1Motion=Free
  swing2LimitDegrees=45
  swing2Motion=Free
  twistLimitDegrees=45
  twistMotion=Free
 }
 {
  bone1Name=clavicle_r
  bone2Name=spine_05
  swing1LimitDegrees=45
  swing1Motion=Free
  swing2LimitDegrees=45
  swing2Motion=Free
  twistLimitDegrees=45
  twistMotion=Free
 }
 {
  bone1Name=thigh_r
  bone2Name=pelvis
  swing1LimitDegrees=45
  swing1Motion=Free
  swing2LimitDegrees=45
  swing2Motion=Free
  twistLimitDegrees=45
  twistMotion=Free
 }
 {
  bone1Name=foot_r
  bone2Name=calf_r
  swing1LimitDegrees=45
  swing1Motion=Free
  swing2LimitDegrees=45
  swing2Motion=Free
  twistLimitDegrees=45
  twistMotion=Free
 }
 {
  bone1Name=calf_r
  bone2Name=thigh_r
  swing1LimitDegrees=45
  swing1Motion=Free
  swing2LimitDegrees=45
  swing2Motion=Free
  twistLimitDegrees=45
  twistMotion=Free
 }
 {
  bone1Name=dyn_blouse_bk_01_l
  bone2Name=spine_04
  swing1LimitDegrees=45
  swing1Motion=Free
  swing2LimitDegrees=45
  swing2Motion=Free
  twistLimitDegrees=45
  twistMotion=Free
 }
 {
  bone1Name=dyn_blouse_bk_04_l
  bone2Name=dyn_blouse_bk_03_l
  swing1LimitDegrees=45
  swing1Motion=Free
  swing2LimitDegrees=45
  swing2Motion=Free
  twistLimitDegrees=45
  twistMotion=Free
 }
 {
  bone1Name=dyn_blouse_bk_03_l
  bone2Name=dyn_blouse_bk_02_l
  swing1LimitDegrees=45
  swing1Motion=Free
  swing2LimitDegrees=45
  swing2Motion=Free
  twistLimitDegrees=45
  twistMotion=Free
 }
 {
  bone1Name=dyn_blouse_bk_02_l
  bone2Name=dyn_blouse_bk_01_l
  swing1LimitDegrees=45
  swing1Motion=Free
  swing2LimitDegrees=45
  swing2Motion=Free
  twistLimitDegrees=45
  twistMotion=Free
 }
 {
  bone1Name=dyn_blouse_bk_01_r
  bone2Name=spine_04
  swing1LimitDegrees=45
  swing1Motion=Free
  swing2LimitDegrees=45
  swing2Motion=Free
  twistLimitDegrees=45
  twistMotion=Free
 }
 {
  bone1Name=dyn_blouse_bk_03_r
  bone2Name=dyn_blouse_bk_02_r
  swing1LimitDegrees=45
  swing1Motion=Free
  swing2LimitDegrees=45
  swing2Motion=Free
  twistLimitDegrees=45
  twistMotion=Free
 }
 {
  bone1Name=dyn_blouse_bk_02_r
  bone2Name=dyn_blouse_bk_01_r
  swing1LimitDegrees=45
  swing1Motion=Free
  swing2LimitDegrees=45
  swing2Motion=Free
  twistLimitDegrees=45
  twistMotion=Free
 }
 {
  bone1Name=dyn_blouse_bk_04_r
  bone2Name=dyn_blouse_bk_03_r
  swing1LimitDegrees=45
  swing1Motion=Free
  swing2LimitDegrees=45
  swing2Motion=Free
  twistLimitDegrees=45
  twistMotion=Free
 }
 {
  bone1Name=dyn_blouse_fr_01_l
  bone2Name=spine_04
  swing1LimitDegrees=45
  swing1Motion=Free
  swing2LimitDegrees=45
  swing2Motion=Free
  twistLimitDegrees=45
  twistMotion=Free
 }
 {
  bone1Name=dyn_blouse_fr_02_l
  bone2Name=dyn_blouse_fr_01_l
  swing1LimitDegrees=45
  swing1Motion=Free
  swing2LimitDegrees=45
  swing2Motion=Free
  twistLimitDegrees=45
  twistMotion=Free
 }
 {
  bone1Name=dyn_blouse_fr_03_l
  bone2Name=dyn_blouse_fr_02_l
  swing1LimitDegrees=45
  swing1Motion=Free
  swing2LimitDegrees=45
  swing2Motion=Free
  twistLimitDegrees=45
  twistMotion=Free
 }
 {
  bone1Name=dyn_blouse_fr_04_l
  bone2Name=dyn_blouse_fr_03_l
  swing1LimitDegrees=45
  swing1Motion=Free
  swing2LimitDegrees=45
  swing2Motion=Free
  twistLimitDegrees=45
  twistMotion=Free
 }
 {
  bone1Name=dyn_blouse_fr_01_r
  bone2Name=spine_04
  swing1LimitDegrees=45
  swing1Motion=Free
  swing2LimitDegrees=45
  swing2Motion=Free
  twistLimitDegrees=45
  twistMotion=Free
 }
 {
  bone1Name=dyn_blouse_fr_02_r
  bone2Name=dyn_blouse_fr_01_r
  swing1LimitDegrees=45
  swing1Motion=Free
  swing2LimitDegrees=45
  swing2Motion=Free
  twistLimitDegrees=45
  twistMotion=Free
 }
 {
  bone1Name=dyn_blouse_fr_03_r
  bone2Name=dyn_blouse_fr_02_r
  swing1LimitDegrees=45
  swing1Motion=Free
  swing2LimitDegrees=45
  swing2Motion=Free
  twistLimitDegrees=45
  twistMotion=Free
 }
 {
  bone1Name=dyn_blouse_fr_04_r
  bone2Name=dyn_blouse_fr_03_r
  swing1LimitDegrees=45
  swing1Motion=Free
  swing2LimitDegrees=45
  swing2Motion=Free
  twistLimitDegrees=45
  twistMotion=Free
 }
 {
  bone1Name=dyn_blouse_side_01_l
  bone2Name=spine_04
  swing1LimitDegrees=45
  swing1Motion=Free
  swing2LimitDegrees=45
  swing2Motion=Free
  twistLimitDegrees=45
  twistMotion=Free
 }
 {
  bone1Name=dyn_blouse_side_02_l
  bone2Name=dyn_blouse_side_01_l
  swing1LimitDegrees=45
  swing1Motion=Free
  swing2LimitDegrees=45
  swing2Motion=Free
  twistLimitDegrees=45
  twistMotion=Free
 }
 {
  bone1Name=dyn_blouse_side_03_l
  bone2Name=dyn_blouse_side_02_l
  swing1LimitDegrees=45
  swing1Motion=Free
  swing2LimitDegrees=45
  swing2Motion=Free
  twistLimitDegrees=45
  twistMotion=Free
 }
 {
  bone1Name=dyn_blouse_side_04_l
  bone2Name=dyn_blouse_side_03_l
  swing1LimitDegrees=45
  swing1Motion=Free
  swing2LimitDegrees=45
  swing2Motion=Free
  twistLimitDegrees=45
  twistMotion=Free
 }
 {
  bone1Name=dyn_blouse_side_01_r
  bone2Name=spine_04
  swing1LimitDegrees=45
  swing1Motion=Free
  swing2LimitDegrees=45
  swing2Motion=Free
  twistLimitDegrees=45
  twistMotion=Free
 }
 {
  bone1Name=dyn_blouse_side_02_r
  bone2Name=dyn_blouse_side_01_r
  swing1LimitDegrees=45
  swing1Motion=Free
  swing2LimitDegrees=45
  swing2Motion=Free
  twistLimitDegrees=45
  twistMotion=Free
 }
 {
  bone1Name=dyn_blouse_side_04_r
  bone2Name=dyn_blouse_side_03_r
  swing1LimitDegrees=45
  swing1Motion=Free
  swing2LimitDegrees=45
  swing2Motion=Free
  twistLimitDegrees=45
  twistMotion=Free
 }
 {
  bone1Name=dyn_blouse_side_03_r
  bone2Name=dyn_blouse_side_02_r
  swing1LimitDegrees=45
  swing1Motion=Free
  swing2LimitDegrees=45
  swing2Motion=Free
  twistLimitDegrees=45
  twistMotion=Free
 }
 {
  bone1Name=dyn_blouse_01_c
  bone2Name=spine_04
  swing1LimitDegrees=45
  swing1Motion=Free
  swing2LimitDegrees=45
  swing2Motion=Free
  twistLimitDegrees=45
  twistMotion=Free
 }
 {
  bone1Name=dyn_blouse_02_c
  bone2Name=dyn_blouse_01_c
  swing1LimitDegrees=45
  swing1Motion=Free
  swing2LimitDegrees=45
  swing2Motion=Free
  twistLimitDegrees=45
  twistMotion=Free
 }
 {
  bone1Name=dyn_blouse_03_c
  bone2Name=dyn_blouse_02_c
  swing1LimitDegrees=45
  swing1Motion=Free
  swing2LimitDegrees=45
  swing2Motion=Free
  twistLimitDegrees=45
  twistMotion=Free
 }
 {
  bone1Name=dyn_blouse_04_c
  bone2Name=dyn_blouse_03_c
  swing1LimitDegrees=45
  swing1Motion=Free
  swing2LimitDegrees=45
  swing2Motion=Free
  twistLimitDegrees=45
  twistMotion=Free
 }
 {
  bone1Name=dyn_strap_r_04
  bone2Name=dyn_strap_r_03
  swing1LimitDegrees=45
  swing1Motion=Free
  swing2LimitDegrees=45
  swing2Motion=Free
  twistLimitDegrees=45
  twistMotion=Free
 }
 {
  bone1Name=dyn_strap_r_01
  bone2Name=calf_r
  swing1LimitDegrees=45
  swing1Motion=Free
  swing2LimitDegrees=45
  swing2Motion=Free
  twistLimitDegrees=45
  twistMotion=Free
 }
 {
  bone1Name=dyn_strap_r_02
  bone2Name=dyn_strap_r_01
  swing1LimitDegrees=45
  swing1Motion=Free
  swing2LimitDegrees=45
  swing2Motion=Free
  twistLimitDegrees=45
  twistMotion=Free
 }
 {
  bone1Name=dyn_strap_r_03
  bone2Name=dyn_strap_r_02
  swing1LimitDegrees=45
  swing1Motion=Free
  swing2LimitDegrees=45
  swing2Motion=Free
  twistLimitDegrees=45
  twistMotion=Free
 }
 {
  bone1Name=dyn_strap_l_04
  bone2Name=dyn_strap_l_03
  swing1LimitDegrees=45
  swing1Motion=Free
  swing2LimitDegrees=45
  swing2Motion=Free
  twistLimitDegrees=45
  twistMotion=Free
 }
 {
  bone1Name=dyn_strap_l_01
  bone2Name=calf_l
  swing1LimitDegrees=45
  swing1Motion=Free
  swing2LimitDegrees=45
  swing2Motion=Free
  twistLimitDegrees=45
  twistMotion=Free
 }
 {
  bone1Name=dyn_strap_l_02
  bone2Name=dyn_strap_l_01
  swing1LimitDegrees=45
  swing1Motion=Free
  swing2LimitDegrees=45
  swing2Motion=Free
  twistLimitDegrees=45
  twistMotion=Free
 }
 {
  bone1Name=dyn_strap_l_03
  bone2Name=dyn_strap_l_02
  swing1LimitDegrees=45
  swing1Motion=Free
  swing2LimitDegrees=45
  swing2Motion=Free
  twistLimitDegrees=45
  twistMotion=Free
 }
]
